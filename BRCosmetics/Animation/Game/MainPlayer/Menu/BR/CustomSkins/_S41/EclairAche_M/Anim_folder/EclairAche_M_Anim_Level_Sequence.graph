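# Level Sequence that animates the EclairAche_M custom skin on a player mannequin, filed under Battle Royale main-player menu animations.
# A spawnable FortPlayerMannequin sits at the origin with a Body component binding and an empty skeletal animation track.
# Motion comes from a keyed Fortnite_Biped_CtrlRig Control Rig track with FK/IK switches and foot-roll controls.
# Held poses and brief foot lifts return to or near the starting leg pose by the end.
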LevelSequence /BRCosmetics/Animation/Game/MainPlayer/Menu/BR/CustomSkins/_S41/EclairAche_M/Anim_folder/EclairAche_M_Anim_Level_Sequence
bindings [
 Body {id=#1, parent_id=#2, possessed_class=SkeletalMeshComponentBudgeted, tracks []}
 FortPlayerMannequin {
  id=#2
  possessed_class=FortPlayerMannequin
  tracks [
   {type=MovieSceneSpawnTrack, sections [{type=MovieSceneSpawnSection, channels [None {type=MovieSceneScriptingBoolChannel, default=true}]}]}
   {
    type=MovieScene3DTransformTrack
    sections [{
      type=MovieScene3DTransformSection
      channels [
       Location.X {type=MovieSceneScriptingDoubleChannel, default=0}
       Location.Y {type=MovieSceneScriptingDoubleChannel, default=0}
       Location.Z {type=MovieSceneScriptingDoubleChannel, default=0}
       Rotation.X {type=MovieSceneScriptingDoubleChannel, default=0}
       Rotation.Y {type=MovieSceneScriptingDoubleChannel, default=0}
       Rotation.Z {type=MovieSceneScriptingDoubleChannel, default=0}
       Scale.X {type=MovieSceneScriptingDoubleChannel, default=1}
       Scale.Y {type=MovieSceneScriptingDoubleChannel, default=1}
       Scale.Z {type=MovieSceneScriptingDoubleChannel, default=1}
      ]
     }]
   }
   {type=MovieSceneSkeletalAnimationTrack, sections []}
   {
    type=MovieSceneControlRigParameterTrack
    display_name=Fortnite_Biped_CtrlRig
    sections [
     {
      type=MovieSceneControlRigParameterSection
      channels [
       leg_l_stretch_switch {type=MovieSceneScriptingBoolChannel, default=false, const [false,1]}
       leg_r_stretch_switch {type=MovieSceneScriptingBoolChannel, default=false, const [false,1]}
       arm_l_fk_local {type=MovieSceneScriptingBoolChannel, default=true, const [true,21]}
       Auto Interpolate_3 {type=MovieSceneScriptingBoolChannel, default=false, const [false,20]}
       Auto Interpolate {type=MovieSceneScriptingBoolChannel, default=false, const [false,20]}
       arm_l_stretch_switch {type=MovieSceneScriptingBoolChannel, default=false, const [false,29]}
       Keep Aligned {type=MovieSceneScriptingBoolChannel, default=false, const [false,29]}
       arm_r_fk_local {type=MovieSceneScriptingBoolChannel, default=true, const [true,21]}
       Auto Interpolate_4 {type=MovieSceneScriptingBoolChannel, default=false, const [false,25]}
       Auto Interpolate_2 {type=MovieSceneScriptingBoolChannel, default=false, const [false,1]}
       arm_r_stretch_switch {type=MovieSceneScriptingBoolChannel, default=false, const [false,25]}
       Keep Aligned_2 {type=MovieSceneScriptingBoolChannel, default=false, const [false,25]}
       head_local {type=MovieSceneScriptingBoolChannel, default=true, const [true,21]}
       Stretch {type=MovieSceneScriptingBoolChannel, default=false, const [false,21]}
       head_stretch {type=MovieSceneScriptingBoolChannel, default=false, const [false,21]}
       spine_stretch_switch {type=MovieSceneScriptingBoolChannel, default=false, const [false,21]}
       arm_l_fk_ik_switch {type=MovieSceneScriptingBoolChannel, default=true, const [true,21]}
       leg_l_fk_ik_switch {type=MovieSceneScriptingBoolChannel, default=true, const [true,21]}
       arm_r_fk_ik_switch {type=MovieSceneScriptingBoolChannel, default=true, const [true,21]}
       leg_r_fk_ik_switch {type=MovieSceneScriptingBoolChannel, default=true, const [true,21]}
       ShowBodyControls {type=MovieSceneScriptingBoolChannel, default=true, const [true,21]}
       spine_fk_ik_switch {type=MovieSceneScriptingBoolChannel, default=false, const [false,21]}
       neck_fk_ik_switch {type=MovieSceneScriptingBoolChannel, default=false, const [false,21]}
       space {type=MovieSceneScriptingByteChannel, default=1}
       global_ctrl.Location.X {type=MovieSceneScriptingFloatChannel, default=0, const [0,21]}
       global_ctrl.Location.Y {type=MovieSceneScriptingFloatChannel, default=0, const [0,21]}
       global_ctrl.Location.Z {type=MovieSceneScriptingFloatChannel, default=0, const [0,21]}
       global_ctrl.Rotation.X {type=MovieSceneScriptingFloatChannel, default=0, const [0,21]}
       global_ctrl.Rotation.Y {type=MovieSceneScriptingFloatChannel, default=0, const [0,21]}
       global_ctrl.Rotation.Z {type=MovieSceneScriptingFloatChannel, default=0, const [0,21]}
       root_ctrl.Location.X {type=MovieSceneScriptingFloatChannel, default=0, const [0,21]}
       root_ctrl.Location.Y {type=MovieSceneScriptingFloatChannel, default=0, const [0,21]}
       root_ctrl.Location.Z {type=MovieSceneScriptingFloatChannel, default=0, const [0,21]}
       root_ctrl.Rotation.X {type=MovieSceneScriptingFloatChannel, default=0, const [0,21]}
       root_ctrl.Rotation.Y {type=MovieSceneScriptingFloatChannel, default=0, const [0,21]}
       root_ctrl.Rotation.Z {type=MovieSceneScriptingFloatChannel, default=0, const [0,21]}
       root_ctrl.Scale.X {type=MovieSceneScriptingFloatChannel, default=1, const [1,21]}
       root_ctrl.Scale.Y {type=MovieSceneScriptingFloatChannel, default=1, const [1,21]}
       root_ctrl.Scale.Z {type=MovieSceneScriptingFloatChannel, default=1, const [1,21]}
       body_offset_ctrl.Location.X {type=MovieSceneScriptingFloatChannel, default=0, const [0,21]}
       body_offset_ctrl.Location.Y {type=MovieSceneScriptingFloatChannel, default=0, const [0,21]}
       body_offset_ctrl.Location.Z {type=MovieSceneScriptingFloatChannel, default=0, const [0,21]}
       body_offset_ctrl.Rotation.X {type=MovieSceneScriptingFloatChannel, default=0, const [0,21]}
       body_offset_ctrl.Rotation.Y {type=MovieSceneScriptingFloatChannel, default=0, const [0,21]}
       body_offset_ctrl.Rotation.Z {type=MovieSceneScriptingFloatChannel, default=0, const [0,21]}
       body_offset_ctrl.Scale.X {type=MovieSceneScriptingFloatChannel, default=1, const [1,21]}
       body_offset_ctrl.Scale.Y {type=MovieSceneScriptingFloatChannel, default=1, const [1,21]}
       body_offset_ctrl.Scale.Z {type=MovieSceneScriptingFloatChannel, default=1, const [1,21]}
       thigh_l_fk_ctrl.Location.X {type=MovieSceneScriptingFloatChannel, default=0, thinned_from=21, keys [[0,0],[354,0]]}
       thigh_l_fk_ctrl.Location.Y {type=MovieSceneScriptingFloatChannel, default=0, thinned_from=21, keys [[0,0],[354,0]]}
       thigh_l_fk_ctrl.Location.Z {type=MovieSceneScriptingFloatChannel, default=0, thinned_from=21, keys [[0,0],[354,0]]}
       thigh_l_fk_ctrl.Rotation.X {
        type=MovieSceneScriptingFloatChannel
        default=20.7279
        thinned_from=21
        keys [[0,20.7279],[40,20.7279],[71,16.5506],[90,16.5506],[121,20.7279],[150,20.7279],[164,22.6902],[205,22.6902],[216,19.881],[240,22.6902],[260,32.8281],[301,32.8281],[322,20.7279],[354,20.7279]]
       }
       thigh_l_fk_ctrl.Rotation.Y {
        type=MovieSceneScriptingFloatChannel
        default=-0.0135
        thinned_from=21
        keys [[0,-0.0135],[40,-0.0135],[71,0.8028],[90,0.8028],[121,-0.0135],[150,-0.0135],[164,2.0782],[205,2.0782],[216,3.7045],[240,2.0782],[260,1.0809],[301,1.0809],[322,-0.0135],[354,-0.0135]]
       }
       thigh_l_fk_ctrl.Rotation.Z {
        type=MovieSceneScriptingFloatChannel
        default=3.9688
        thinned_from=21
        keys [[0,3.9688],[40,3.9688],[71,3.1454],[90,3.1454],[121,3.9688],[150,3.9688],[164,-7.0761],[205,-7.0761],[216,4.9007],[240,-7.0761],[260,6.1801],[301,6.1801],[322,3.9688],[354,3.9688]]
       }
       thigh_l_fk_ctrl.Scale.X {type=MovieSceneScriptingFloatChannel, default=1, const [1,21]}
       thigh_l_fk_ctrl.Scale.Y {type=MovieSceneScriptingFloatChannel, default=1, const [1,21]}
       thigh_l_fk_ctrl.Scale.Z {type=MovieSceneScriptingFloatChannel, default=1, const [1,21]}
       calf_l_fk_ctrl.Location.X {type=MovieSceneScriptingFloatChannel, default=0, thinned_from=17, keys [[0,0],[174,0],[216,-0.819],[260,0],[354,0]]}
       calf_l_fk_ctrl.Location.Y {type=MovieSceneScriptingFloatChannel, default=0, thinned_from=17, keys [[0,0],[174,0],[216,0.0613],[260,0],[354,0]]}
       calf_l_fk_ctrl.Location.Z {type=MovieSceneScriptingFloatChannel, default=0, thinned_from=17, keys [[0,0],[174,0],[216,-0.0176],[260,0],[354,0]]}
       calf_l_fk_ctrl.Rotation.X {
        type=MovieSceneScriptingFloatChannel
        default=0.0935
        thinned_from=17
        keys [[0,0.0935],[40,0.0935],[71,0.0882],[90,0.0882],[121,0.0935],[150,0.0935],[164,0.4752],[174,0.0822],[216,0.0341],[260,0.0804],[301,0.0804],[322,0.0935],[354,0.0935]]
       }
       calf_l_fk_ctrl.Rotation.Y {
        type=MovieSceneScriptingFloatChannel
        default=1.2214
        thinned_from=17
        keys [[0,1.2214],[150,1.2214],[164,1.1291],[174,1.222],[216,1.2245],[354,1.2214]]
       }
       calf_l_fk_ctrl.Rotation.Z {
        type=MovieSceneScriptingFloatChannel
        default=2.2093
        thinned_from=17
        keys [[0,2.2093],[40,2.2093],[71,1.9642],[90,1.9642],[121,2.2093],[150,2.2093],[164,20.6607],[174,1.683],[216,-0.57],[260,1.5994],[301,1.5994],[322,2.2093],[354,2.2093]]
       }
       calf_l_fk_ctrl.Scale.X {type=MovieSceneScriptingFloatChannel, default=1, const [1,17]}
       calf_l_fk_ctrl.Scale.Y {type=MovieSceneScriptingFloatChannel, default=1, const [1,17]}
       calf_l_fk_ctrl.Scale.Z {type=MovieSceneScriptingFloatChannel, default=1, const [1,17]}
       foot_l_fk_ctrl.Location.X {type=MovieSceneScriptingFloatChannel, default=0, thinned_from=17, keys [[0,0],[164,0],[174,-0.0039],[217,-0.0039],[260,0],[354,0]]}
       foot_l_fk_ctrl.Location.Y {type=MovieSceneScriptingFloatChannel, default=0.0398, thinned_from=17, keys [[0,0.0398],[354,0.0398]]}
       foot_l_fk_ctrl.Location.Z {type=MovieSceneScriptingFloatChannel, default=-0.8885, thinned_from=17, keys [[0,-0.8885],[354,-0.8885]]}
       foot_l_fk_ctrl.Rotation.X {
        type=MovieSceneScriptingFloatChannel
        default=-1.2985
        thinned_from=17
        keys [[0,-1.2985],[40,-1.2985],[71,-1.3798],[90,-1.3798],[121,-1.2985],[153,-1.2985],[164,9.691],[174,-1.2002],[217,-1.2482],[260,-1.244],[301,-1.244],[322,-1.2985],[354,-1.2985]]
       }
       foot_l_fk_ctrl.Rotation.Y {
        type=MovieSceneScriptingFloatChannel
        default=-0.8497
        thinned_from=17
        keys [[0,-0.8497],[40,-0.8497],[71,-1.2499],[90,-1.2499],[121,-0.8497],[153,-0.8497],[164,-2.9174],[174,-1.802],[217,-1.7971],[260,-1.7733],[301,-1.7733],[322,-0.8497],[354,-0.8497]]
       }
       foot_l_fk_ctrl.Rotation.Z {
        type=MovieSceneScriptingFloatChannel
        default=-0.1994
        thinned_from=17
        keys [[0,-0.1994],[40,-0.1994],[71,0.5105],[90,0.5105],[121,-0.1994],[153,-0.1994],[164,6.4],[174,-0.5887],[217,-0.6194],[260,-0.6393],[301,-0.6393],[322,-0.1994],[354,-0.1994]]
       }
       foot_l_fk_ctrl.Scale.X {type=MovieSceneScriptingFloatChannel, default=1, const [1,17]}
       foot_l_fk_ctrl.Scale.Y {type=MovieSceneScriptingFloatChannel, default=1, const [1,17]}
       foot_l_fk_ctrl.Scale.Z {type=MovieSceneScriptingFloatChannel, default=1, const [1,17]}
       ball_l_fk_ctrl.Location.X {type=MovieSceneScriptingFloatChannel, default=0, thinned_from=17, keys [[0,0],[354,0]]}
       ball_l_fk_ctrl.Location.Y {type=MovieSceneScriptingFloatChannel, default=0, thinned_from=17, keys [[0,0],[354,0]]}
       ball_l_fk_ctrl.Location.Z {type=MovieSceneScriptingFloatChannel, default=0, thinned_from=17, keys [[0,0],[354,0]]}
       ball_l_fk_ctrl.Rotation.X {type=MovieSceneScriptingFloatChannel, default=0, thinned_from=17, keys [[0,0],[153,0],[164,-14.6633],[174,-0.0643],[217,0],[354,0]]}
       ball_l_fk_ctrl.Rotation.Y {type=MovieSceneScriptingFloatChannel, default=0, thinned_from=17, keys [[0,0],[153,0],[164,-2.8606],[174,-0.0125],[217,0],[354,0]]}
       ball_l_fk_ctrl.Rotation.Z {type=MovieSceneScriptingFloatChannel, default=0, thinned_from=17, keys [[0,0],[153,0],[164,-10.8786],[174,-0.0477],[217,0],[354,0]]}
       ball_l_fk_ctrl.Scale.X {type=MovieSceneScriptingFloatChannel, default=1, const [1,17]}
       ball_l_fk_ctrl.Scale.Y {type=MovieSceneScriptingFloatChannel, default=1, const [1,17]}
       ball_l_fk_ctrl.Scale.Z {type=MovieSceneScriptingFloatChannel, default=1, const [1,17]}
       foot_l_ik_ctrl.Location.X {type=MovieSceneScriptingFloatChannel, default=0, const [4.2828,17]}
       foot_l_ik_ctrl.Location.Y {type=MovieSceneScriptingFloatChannel, default=0, const [0,17]}
       foot_l_ik_ctrl.Location.Z {type=MovieSceneScriptingFloatChannel, default=0, thinned_from=16, keys [[0,0],[354,0]]}
       foot_l_ik_ctrl.Rotation.X {type=MovieSceneScriptingFloatChannel, default=0, thinned_from=17, keys [[0,0],[354,0]]}
       foot_l_ik_ctrl.Rotation.Y {type=MovieSceneScriptingFloatChannel, default=0, thinned_from=16, keys [[0,0],[354,0]]}
       foot_l_ik_ctrl.Rotation.Z {type=MovieSceneScriptingFloatChannel, default=-20, const [-20,17]}
       foot_l_ik_ctrl.Scale.X {type=MovieSceneScriptingFloatChannel, default=1, const [1,18]}
       foot_l_ik_ctrl.Scale.Y {type=MovieSceneScriptingFloatChannel, default=1, const [1,18]}
       foot_l_ik_ctrl.Scale.Z {type=MovieSceneScriptingFloatChannel, default=1, const [1,18]}
       ball_l_ik_ctrl.Location.X {type=MovieSceneScriptingFloatChannel, default=0, const [0,17]}
       ball_l_ik_ctrl.Location.Y {type=MovieSceneScriptingFloatChannel, default=0, const [0,17]}
       ball_l_ik_ctrl.Location.Z {type=MovieSceneScriptingFloatChannel, default=0, const [0,17]}
       ball_l_ik_ctrl.Rotation.X {type=MovieSceneScriptingFloatChannel, default=0, const [0,17]}
       ball_l_ik_ctrl.Rotation.Y {type=MovieSceneScriptingFloatChannel, default=0, const [0,17]}
       ball_l_ik_ctrl.Rotation.Z {type=MovieSceneScriptingFloatChannel, default=0, const [0,17]}
       ball_l_ik_ctrl.Scale.X {type=MovieSceneScriptingFloatChannel, default=1, const [1,17]}
       ball_l_ik_ctrl.Scale.Y {type=MovieSceneScriptingFloatChannel, default=1, const [1,17]}
       ball_l_ik_ctrl.Scale.Z {type=MovieSceneScriptingFloatChannel, default=1, const [1,17]}
       foot_bk1_l_ctrl.Location.X {type=MovieSceneScriptingFloatChannel, default=0, const [0,19]}
       foot_bk1_l_ctrl.Location.Y {type=MovieSceneScriptingFloatChannel, default=0, const [0,19]}
       foot_bk1_l_ctrl.Location.Z {type=MovieSceneScriptingFloatChannel, default=0, const [0,19]}
       foot_bk1_l_ctrl.Rotation.X {type=MovieSceneScriptingFloatChannel, default=0, const [0,19]}
       foot_bk1_l_ctrl.Rotation.Y {type=MovieSceneScriptingFloatChannel, default=0, const [0,19]}
       foot_bk1_l_ctrl.Rotation.Z {
        type=MovieSceneScriptingFloatChannel
        default=0
        thinned_from=20
        keys [[0,0],[153,0],[159,-2.5121],[164,0],[205,0],[217,-2.8896],[233,0],[354,0]]
       }
       foot_bk1_l_ctrl.Scale.X {type=MovieSceneScriptingFloatChannel, default=1, const [1,19]}
       foot_bk1_l_ctrl.Scale.Y {type=MovieSceneScriptingFloatChannel, default=1, const [1,19]}
       foot_bk1_l_ctrl.Scale.Z {type=MovieSceneScriptingFloatChannel, default=1, const [1,19]}
       foot_roll_l_ctrl.Location.X {type=MovieSceneScriptingFloatChannel, default=0, const [0,17]}
       foot_roll_l_ctrl.Location.Y {type=MovieSceneScriptingFloatChannel, default=0, const [0,17]}
       foot_roll_l_ctrl.Location.Z {type=MovieSceneScriptingFloatChannel, default=0, const [0,17]}
       foot_roll_l_ctrl.Rotation.X {type=MovieSceneScriptingFloatChannel, default=0, const [0,17]}
       foot_roll_l_ctrl.Rotation.Y {type=MovieSceneScriptingFloatChannel, default=0, const [0,17]}
       foot_roll_l_ctrl.Rotation.Z {type=MovieSceneScriptingFloatChannel, default=0, const [0,17]}
       foot_roll_l_ctrl.Scale.X {type=MovieSceneScriptingFloatChannel, default=1, const [1,17]}
       foot_roll_l_ctrl.Scale.Y {type=MovieSceneScriptingFloatChannel, default=1, const [1,17]}
       foot_roll_l_ctrl.Scale.Z {type=MovieSceneScriptingFloatChannel, default=1, const [1,17]}
       roll_blend_l_ctrl {type=MovieSceneScriptingFloatChannel, default=1, const [1,17]}
       tip_l_ctrl.Location.X {type=MovieSceneScriptingFloatChannel, default=0, const [0,19]}
       tip_l_ctrl.Location.Y {type=MovieSceneScriptingFloatChannel, default=0, const [0,19]}
       tip_l_ctrl.Location.Z {type=MovieSceneScriptingFloatChannel, default=0, const [0,19]}
       tip_l_ctrl.Rotation.X {type=MovieSceneScriptingFloatChannel, default=0, const [0,19]}
       tip_l_ctrl.Rotation.Y {type=MovieSceneScriptingFloatChannel, default=0, const [0,19]}
       tip_l_ctrl.Rotation.Z {
        type=MovieSceneScriptingFloatChannel
        default=0
        thinned_from=20
        keys [[0,0],[159,0],[164,-3.3698],[170,-1.1972],[174,-0.0149],[196,-0.0002],[354,0]]
       }
       tip_l_ctrl.Scale.X {type=MovieSceneScriptingFloatChannel, default=1, const [1,19]}
       tip_l_ctrl.Scale.Y {type=MovieSceneScriptingFloatChannel, default=1, const [1,19]}
       tip_l_ctrl.Scale.Z {type=MovieSceneScriptingFloatChannel, default=1, const [1,19]}
       heel_l_ctrl.Location.X {type=MovieSceneScriptingFloatChannel, default=0, const [0,22]}
       heel_l_ctrl.Location.Y {type=MovieSceneScriptingFloatChannel, default=0, const [0,22]}
       heel_l_ctrl.Location.Z {type=MovieSceneScriptingFloatChannel, default=0, const [0,22]}
       heel_l_ctrl.Rotation.X {type=MovieSceneScriptingFloatChannel, default=0, const [0,22]}
       heel_l_ctrl.Rotation.Y {type=MovieSceneScriptingFloatChannel, default=0, const [0,22]}
       heel_l_ctrl.Rotation.Z {type=MovieSceneScriptingFloatChannel, default=0, thinned_from=23, keys [[0,0],[159,0],[164,3.6505],[166,3.0018],[176,0],[354,0]]}
       heel_l_ctrl.Scale.X {type=MovieSceneScriptingFloatChannel, default=1, const [1,22]}
       heel_l_ctrl.Scale.Y {type=MovieSceneScriptingFloatChannel, default=1, const [1,22]}
       heel_l_ctrl.Scale.Z {type=MovieSceneScriptingFloatChannel, default=1, const [1,22]}
       leg_l_softness {type=MovieSceneScriptingFloatChannel, default=0, const [0.5165,15]}
       leg_l_pv_ik_ctrl.Location.X {type=MovieSceneScriptingFloatChannel, default=0, const [0,21]}
       leg_l_pv_ik_ctrl.Location.Y {type=MovieSceneScriptingFloatChannel, default=0, const [0,21]}
       leg_l_pv_ik_ctrl.Location.Z {type=MovieSceneScriptingFloatChannel, default=0, const [0,21]}
       leg_l_pv_ik_ctrl.Rotation.X {type=MovieSceneScriptingFloatChannel, default=0, const [0,21]}
       leg_l_pv_ik_ctrl.Rotation.Y {type=MovieSceneScriptingFloatChannel, default=0, const [0,21]}
       leg_l_pv_ik_ctrl.Rotation.Z {type=MovieSceneScriptingFloatChannel, default=0, const [0,21]}
       leg_l_pv_ik_ctrl.Scale.X {type=MovieSceneScriptingFloatChannel, default=1, const [1,21]}
       leg_l_pv_ik_ctrl.Scale.Y {type=MovieSceneScriptingFloatChannel, default=1, const [1,21]}
       leg_l_pv_ik_ctrl.Scale.Z {type=MovieSceneScriptingFloatChannel, default=1, const [1,21]}
       thigh_r_fk_ctrl.Location.X {type=MovieSceneScriptingFloatChannel, default=0, thinned_from=21, keys [[0,0],[354,0]]}
       thigh_r_fk_ctrl.Location.Y {type=MovieSceneScriptingFloatChannel, default=0, thinned_from=21, keys [[0,0],[354,0]]}
       thigh_r_fk_ctrl.Location.Z {type=MovieSceneScriptingFloatChannel, default=0, thinned_from=21, keys [[0,0],[354,0]]}
       thigh_r_fk_ctrl.Rotation.X {
        type=MovieSceneScriptingFloatChannel
        default=20.7191
        thinned_from=21
        keys [[0,20.7191],[40,20.7191],[71,24.8829],[90,24.8829],[121,20.7191],[150,20.7191],[164,20.9934],[205,20.9934],[216,21.5795],[240,20.9934],[260,8.8569],[301,8.8569],[322,20.7191],[354,20.7191]]
       }
       thigh_r_fk_ctrl.Rotation.Y {
        type=MovieSceneScriptingFloatChannel
        default=-0.0159
        thinned_from=21
        keys [[0,-0.0159],[40,-0.0159],[71,-0.9694],[90,-0.9694],[121,-0.0159],[150,-0.0159],[164,2.1847],[205,2.1847],[216,2.7589],[240,2.1847],[260,-1.8351],[301,-1.8351],[322,-0.0159],[354,-0.0159]]
       }
       thigh_r_fk_ctrl.Rotation.Z {
        type=MovieSceneScriptingFloatChannel
        default=3.9734
        thinned_from=21
        keys [[0,3.9734],[40,3.9734],[71,4.8839],[90,4.8839],[121,3.9734],[150,3.9734],[164,-1.2155],[205,-1.2155],[216,-0.4967],[240,-1.2155],[260,0.9574],[301,0.9574],[322,3.9734],[354,3.9734]]
       }
       thigh_r_fk_ctrl.Scale.X {type=MovieSceneScriptingFloatChannel, default=1, const [1,21]}
       thigh_r_fk_ctrl.Scale.Y {type=MovieSceneScriptingFloatChannel, default=1, const [1,21]}
       thigh_r_fk_ctrl.Scale.Z {type=MovieSceneScriptingFloatChannel, default=1, const [1,21]}
       calf_r_fk_ctrl.Location.X {type=MovieSceneScriptingFloatChannel, default=0, thinned_from=15, keys [[0,0],[354,0]]}
       calf_r_fk_ctrl.Location.Y {type=MovieSceneScriptingFloatChannel, default=0, thinned_from=15, keys [[0,0],[354,0]]}
       calf_r_fk_ctrl.Location.Z {type=MovieSceneScriptingFloatChannel, default=0, thinned_from=15, keys [[0,0],[354,0]]}
       calf_r_fk_ctrl.Rotation.X {
        type=MovieSceneScriptingFloatChannel
        default=0.0945
        thinned_from=15
        keys [[0,0.0945],[40,0.0945],[71,0.0912],[90,0.0912],[121,0.0945],[150,0.0945],[169,0.3174],[190,0.3174],[196,0.0955],[322,0.0945],[354,0.0945]]
       }
       calf_r_fk_ctrl.Rotation.Y {
        type=MovieSceneScriptingFloatChannel
        default=1.1958
        thinned_from=15
        keys [[0,1.1958],[150,1.1958],[169,1.1568],[190,1.1568],[196,1.1956],[354,1.1958]]
       }
       calf_r_fk_ctrl.Rotation.Z {
        type=MovieSceneScriptingFloatChannel
        default=2.1016
        thinned_from=15
        keys [[0,2.1016],[40,2.1016],[71,1.9442],[90,1.9442],[121,2.1016],[150,2.1016],[169,12.9259],[190,12.9259],[196,2.1484],[322,2.101],[354,2.1016]]
       }
       calf_r_fk_ctrl.Scale.X {type=MovieSceneScriptingFloatChannel, default=1, const [1,15]}
       calf_r_fk_ctrl.Scale.Y {type=MovieSceneScriptingFloatChannel, default=1, const [1,15]}
       calf_r_fk_ctrl.Scale.Z {type=MovieSceneScriptingFloatChannel, default=1, const [1,15]}
       foot_r_fk_ctrl.Location.X {type=MovieSceneScriptingFloatChannel, default=0, thinned_from=15, keys [[0,0],[354,0]]}
       foot_r_fk_ctrl.Location.Y {type=MovieSceneScriptingFloatChannel, default=-0.0329, thinned_from=15, keys [[0,-0.0329],[190,-0.0329],[196,-0.0344],[354,-0.0344]]}
       foot_r_fk_ctrl.Location.Z {type=MovieSceneScriptingFloatChannel, default=-0.9075, thinned_from=15, keys [[0,-0.9075],[354,-0.9075]]}
       foot_r_fk_ctrl.Rotation.X {
        type=MovieSceneScriptingFloatChannel
        default=-1.2933
        thinned_from=15
        keys [[0,-1.2933],[40,-1.2933],[71,-1.2153],[90,-1.2153],[121,-1.2933],[150,-1.2933],[169,-1.0552],[190,-1.0552],[196,-1.2926],[354,-1.2933]]
       }
       foot_r_fk_ctrl.Rotation.Y {
        type=MovieSceneScriptingFloatChannel
        default=-0.8223
        thinned_from=15
        keys [[0,-0.8223],[40,-0.8223],[71,-0.4478],[90,-0.4478],[121,-0.8223],[150,-0.8223],[169,2.5357],[190,2.5357],[196,-0.8041],[354,-0.8223]]
       }
       foot_r_fk_ctrl.Rotation.Z {
        type=MovieSceneScriptingFloatChannel
        default=-0.0975
        thinned_from=15
        keys [[0,-0.0975],[40,-0.0975],[71,-0.6013],[90,-0.6013],[121,-0.0975],[150,-0.0975],[169,-6.6433],[190,-6.6433],[196,-0.1192],[322,-0.0905],[354,-0.0955]]
       }
       foot_r_fk_ctrl.Scale.X {type=MovieSceneScriptingFloatChannel, default=1, const [1,15]}
       foot_r_fk_ctrl.Scale.Y {type=MovieSceneScriptingFloatChannel, default=1, const [1,15]}
       foot_r_fk_ctrl.Scale.Z {type=MovieSceneScriptingFloatChannel, default=1, const [1,15]}
       ball_r_fk_ctrl.Location.X {type=MovieSceneScriptingFloatChannel, default=0, const [0,15]}
       ball_r_fk_ctrl.Location.Y {type=MovieSceneScriptingFloatChannel, default=0, const [0,15]}
       ball_r_fk_ctrl.Location.Z {type=MovieSceneScriptingFloatChannel, default=0, const [0,15]}
       ball_r_fk_ctrl.Rotation.X {type=MovieSceneScriptingFloatChannel, default=0, const [0,15]}
       ball_r_fk_ctrl.Rotation.Y {type=MovieSceneScriptingFloatChannel, default=0, const [0,15]}
       ball_r_fk_ctrl.Rotation.Z {type=MovieSceneScriptingFloatChannel, default=0, const [0,15]}
       ball_r_fk_ctrl.Scale.X {type=MovieSceneScriptingFloatChannel, default=1, const [1,15]}
       ball_r_fk_ctrl.Scale.Y {type=MovieSceneScriptingFloatChannel, default=1, const [1,15]}
       ball_r_fk_ctrl.Scale.Z {type=MovieSceneScriptingFloatChannel, default=1, const [1,15]}
       foot_r_ik_ctrl.Location.X {type=MovieSceneScriptingFloatChannel, default=0, const [3.2262,4]}
       foot_r_ik_ctrl.Location.Y {type=MovieSceneScriptingFloatChannel, default=0, const [0,14]}
       foot_r_ik_ctrl.Location.Z {type=MovieSceneScriptingFloatChannel, default=0, const [0,12]}
       foot_r_ik_ctrl.Rotation.X {type=MovieSceneScriptingFloatChannel, default=0, thinned_from=14, keys [[0,0],[354,0]]}
       foot_r_ik_ctrl.Rotation.Y {type=MovieSceneScriptingFloatChannel, default=0, thinned_from=14, keys [[0,0],[354,0]]}
       foot_r_ik_ctrl.Rotation.Z {type=MovieSceneScriptingFloatChannel, default=-20, const [-20,16]}
       foot_r_ik_ctrl.Scale.X {type=MovieSceneScriptingFloatChannel, default=1, const [1,15]}
       foot_r_ik_ctrl.Scale.Y {type=MovieSceneScriptingFloatChannel, default=1, const [1,15]}
       foot_r_ik_ctrl.Scale.Z {type=MovieSceneScriptingFloatChannel, default=1, const [1,15]}
       ball_r_ik_ctrl.Location.X {type=MovieSceneScriptingFloatChannel, default=0, const [0,13]}
       ball_r_ik_ctrl.Location.Y {type=MovieSceneScriptingFloatChannel, default=0, const [0,13]}
       ball_r_ik_ctrl.Location.Z {type=MovieSceneScriptingFloatChannel, default=0, const [0,13]}
       ball_r_ik_ctrl.Rotation.X {type=MovieSceneScriptingFloatChannel, default=0, const [0,13]}
       ball_r_ik_ctrl.Rotation.Y {type=MovieSceneScriptingFloatChannel, default=0, const [0,13]}
       ball_r_ik_ctrl.Rotation.Z {type=MovieSceneScriptingFloatChannel, default=0, const [0,13]}
       ball_r_ik_ctrl.Scale.X {type=MovieSceneScriptingFloatChannel, default=1, const [1,13]}
       ball_r_ik_ctrl.Scale.Y {type=MovieSceneScriptingFloatChannel, default=1, const [1,13]}
       ball_r_ik_ctrl.Scale.Z {type=MovieSceneScriptingFloatChannel, default=1, const [1,13]}
       foot_bk1_r_ctrl.Location.X {type=MovieSceneScriptingFloatChannel, default=0, const [0,15]}
       foot_bk1_r_ctrl.Location.Y {type=MovieSceneScriptingFloatChannel, default=0, const [0,15]}
       foot_bk1_r_ctrl.Location.Z {type=MovieSceneScriptingFloatChannel, default=0, const [0,15]}
       foot_bk1_r_ctrl.Rotation.X {type=MovieSceneScriptingFloatChannel, default=0, const [0,15]}
       foot_bk1_r_ctrl.Rotation.Y {type=MovieSceneScriptingFloatChannel, default=0, const [0,15]}
       foot_bk1_r_ctrl.Rotation.Z {type=MovieSceneScriptingFloatChannel, default=0, thinned_from=16, keys [[0,0],[179,0],[182,-1.154],[187,0],[354,0]]}
       foot_bk1_r_ctrl.Scale.X {type=MovieSceneScriptingFloatChannel, default=1, const [1,15]}
       foot_bk1_r_ctrl.Scale.Y {type=MovieSceneScriptingFloatChannel, default=1, const [1,15]}
       foot_bk1_r_ctrl.Scale.Z {type=MovieSceneScriptingFloatChannel, default=1, const [1,15]}
       foot_roll_r_ctrl.Location.X {type=MovieSceneScriptingFloatChannel, default=0, const [0,15]}
       foot_roll_r_ctrl.Location.Y {type=MovieSceneScriptingFloatChannel, default=0, const [0,15]}
       foot_roll_r_ctrl.Location.Z {type=MovieSceneScriptingFloatChannel, default=0, const [0,15]}
       foot_roll_r_ctrl.Rotation.X {type=MovieSceneScriptingFloatChannel, default=0, const [0,15]}
       foot_roll_r_ctrl.Rotation.Y {type=MovieSceneScriptingFloatChannel, default=0, const [0,15]}
       foot_roll_r_ctrl.Rotation.Z {type=MovieSceneScriptingFloatChannel, default=0, const [0,15]}
       foot_roll_r_ctrl.Scale.X {type=MovieSceneScriptingFloatChannel, default=1, const [1,15]}
       foot_roll_r_ctrl.Scale.Y {type=MovieSceneScriptingFloatChannel, default=1, const [1,15]}
       foot_roll_r_ctrl.Scale.Z {type=MovieSceneScriptingFloatChannel, default=1, const [1,15]}
       roll_blend_r_ctrl {type=MovieSceneScriptingFloatChannel, default=1, const [1,15]}
       tip_r_ctrl.Location.X {type=MovieSceneScriptingFloatChannel, default=0, const [0,16]}
       tip_r_ctrl.Location.Y {type=MovieSceneScriptingFloatChannel, default=0, const [0,16]}
       tip_r_ctrl.Location.Z {type=MovieSceneScriptingFloatChannel, default=0, const [0,16]}
       tip_r_ctrl.Rotation.X {type=MovieSceneScriptingFloatChannel, default=0, const [0,16]}
       tip_r_ctrl.Rotation.Y {type=MovieSceneScriptingFloatChannel, default=0, const [0,16]}
       tip_r_ctrl.Rotation.Z {type=MovieSceneScriptingFloatChannel, default=0, thinned_from=16, keys [[0,0],[182,0],[187,-2.0168],[196,-0.0088],[322,0],[354,0]]}
       tip_r_ctrl.Scale.X {type=MovieSceneScriptingFloatChannel, default=1, const [1,16]}
       tip_r_ctrl.Scale.Y {type=MovieSceneScriptingFloatChannel, default=1, const [1,16]}
       tip_r_ctrl.Scale.Z {type=MovieSceneScriptingFloatChannel, default=1, const [1,16]}
       heel_r_ctrl.Location.X {type=MovieSceneScriptingFloatChannel, default=0, thinned_from=21, keys [[0,0],[354,0]]}
       heel_r_ctrl.Location.Y {type=MovieSceneScriptingFloatChannel, default=0, thinned_from=21, keys [[0,0],[354,0]]}
       heel_r_ctrl.Location.Z {type=MovieSceneScriptingFloatChannel, default=0, thinned_from=21, keys [[0,0],[354,0]]}
       heel_r_ctrl.Rotation.X {type=MovieSceneScriptingFloatChannel, default=0, const [0,21]}
       heel_r_ctrl.Rotation.Y {type=MovieSceneScriptingFloatChannel, default=0, const [0,21]}
       heel_r_ctrl.Rotation.Z {type=MovieSceneScriptingFloatChannel, default=0, thinned_from=23, keys [[0,0],[183,0],[187,1.4626],[190,0.8867],[198,0],[354,0]]}
       heel_r_ctrl.Scale.X {type=MovieSceneScriptingFloatChannel, default=1, const [1,21]}
       heel_r_ctrl.Scale.Y {type=MovieSceneScriptingFloatChannel, default=1, const [1,21]}
       heel_r_ctrl.Scale.Z {type=MovieSceneScriptingFloatChannel, default=1, const [1,21]}
       leg_r_softness {type=MovieSceneScriptingFloatChannel, default=0, const [0.5175,12]}
       leg_r_pv_ik_ctrl.Location.X {type=MovieSceneScriptingFloatChannel, default=0, const [0,21]}
       leg_r_pv_ik_ctrl.Location.Y {type=MovieSceneScriptingFloatChannel, default=0, const [0,21]}
       leg_r_pv_ik_ctrl.Location.Z {type=MovieSceneScriptingFloatChannel, default=0, const [0,21]}
       leg_r_pv_ik_ctrl.Rotation.X {type=MovieSceneScriptingFloatChannel, default=0, const [0,21]}
       leg_r_pv_ik_ctrl.Rotation.Y {type=MovieSceneScriptingFloatChannel, default=0, const [0,21]}
       leg_r_pv_ik_ctrl.Rotation.Z {type=MovieSceneScriptingFloatChannel, default=0, const [0,21]}
       leg_r_pv_ik_ctrl.Scale.X {type=MovieSceneScriptingFloatChannel, default=1, const [1,21]}
       leg_r_pv_ik_ctrl.Scale.Y {type=MovieSceneScriptingFloatChannel, default=1, const [1,21]}
       leg_r_pv_ik_ctrl.Scale.Z {type=MovieSceneScriptingFloatChannel, default=1, const [1,21]}
       upperarm_l_fk_ctrl.Location.X {type=MovieSceneScriptingFloatChannel, default=0, thinned_from=21, keys [[0,0],[354,0]]}
       upperarm_l_fk_ctrl.Location.Y {type=MovieSceneScriptingFloatChannel, default=0, thinned_from=21, keys [[0,0],[354,0]]}
       upperarm_l_fk_ctrl.Location.Z {type=MovieSceneScriptingFloatChannel, default=0, thinned_from=21, keys [[0,0],[354,0]]}
       upperarm_l_fk_ctrl.Rotation.X {
        type=MovieSceneScriptingFloatChannel
        default=1.1491
        thinned_from=21
        keys [[0,52.0085],[25,51.4943],[34,51.4943],[40,50.5059],[71,54.9568],[90,54.9568],[121,51.4364],[150,51.4943],[164,54.4444],[205,54.4444],[216,54.8169],[240,54.4444],[260,39.4717],[301,39.4717],[322,52.0085],[354,51.4943]]
       }
       upperarm_l_fk_ctrl.Rotation.Y {
        type=MovieSceneScriptingFloatChannel
        default=-0.7961
        thinned_from=21
        keys [[0,12.8604],[25,15.4952],[34,15.4952],[40,-19.8949],[71,17.6203],[90,17.6203],[121,15.5119],[150,15.4952],[164,13.8083],[205,13.8083],[216,16.9754],[240,13.8083],[260,9.5803],[301,9.5803],[322,12.8604],[354,15.4952]]
       }
       upperarm_l_fk_ctrl.Rotation.Z {
        type=MovieSceneScriptingFloatChannel
        default=-0.0209
        thinned_from=21
        keys [[0,-27.6968],[25,-29.8492],[34,-29.8492],[40,-2.382],[71,-22.0742],[90,-22.0742],[121,-29.7145],[150,-29.8492],[164,-26.3934],[205,-26.3934],[216,-27.9282],[240,-26.3934],[260,-39.7769],[301,-39.7769],[322,-27.6968],[354,-29.8492]]
       }
       upperarm_l_fk_ctrl.Scale.X {type=MovieSceneScriptingFloatChannel, default=1, const [1,21]}
       upperarm_l_fk_ctrl.Scale.Y {type=MovieSceneScriptingFloatChannel, default=1, const [1,21]}
       upperarm_l_fk_ctrl.Scale.Z {type=MovieSceneScriptingFloatChannel, default=1, const [1,21]}
       lowerarm_l_fk_ctrl.Location.X {type=MovieSceneScriptingFloatChannel, default=0, thinned_from=21, keys [[0,0],[354,0]]}
       lowerarm_l_fk_ctrl.Location.Y {type=MovieSceneScriptingFloatChannel, default=0, thinned_from=21, keys [[0,0],[354,0]]}
       lowerarm_l_fk_ctrl.Location.Z {type=MovieSceneScriptingFloatChannel, default=0, thinned_from=21, keys [[0,0],[354,0]]}
       lowerarm_l_fk_ctrl.Rotation.X {
        type=MovieSceneScriptingFloatChannel
        default=0.2616
        thinned_from=21
        keys [[0,1.0321],[25,1.0784],[34,1.0784],[40,-0.2884],[71,1.0764],[121,1.0791],[150,1.0784],[164,1.0321],[205,1.0321],[216,1.0807],[240,1.0321],[260,1.0825],[301,1.0825],[322,1.0321],[354,1.0784]]
       }
       lowerarm_l_fk_ctrl.Rotation.Y {
        type=MovieSceneScriptingFloatChannel
        default=1.0908
        thinned_from=21
        keys [[0,0.4394],[25,0.3087],[34,0.3087],[40,1.0841],[71,0.3158],[90,0.3158],[121,0.3064],[150,0.3087],[164,0.4394],[205,0.4394],[216,0.3008],[240,0.4394],[260,0.2942],[301,0.2942],[322,0.4394],[354,0.3087]]
       }
       lowerarm_l_fk_ctrl.Rotation.Z {
        type=MovieSceneScriptingFloatChannel
        default=-0.1228
        thinned_from=21
        keys [[0,53.3318],[25,60.4177],[34,60.4177],[40,-28.5068],[71,60.0437],[90,60.0437],[121,60.5417],[150,60.4177],[164,53.3318],[205,53.3318],[216,60.8399],[240,53.3318],[260,61.1906],[301,61.1906],[322,53.3318],[354,60.4177]]
       }
       lowerarm_l_fk_ctrl.Scale.X {type=MovieSceneScriptingFloatChannel, default=1, const [1,21]}
       lowerarm_l_fk_ctrl.Scale.Y {type=MovieSceneScriptingFloatChannel, default=1, const [1,21]}
       lowerarm_l_fk_ctrl.Scale.Z {type=MovieSceneScriptingFloatChannel, default=1, const [1,21]}
       hand_l_fk_ctrl.Location.X {type=MovieSceneScriptingFloatChannel, default=0, thinned_from=21, keys [[0,0],[354,0]]}
       hand_l_fk_ctrl.Location.Y {type=MovieSceneScriptingFloatChannel, default=-0.431, thinned_from=21, keys [[0,-0.431],[354,-0.431]]}
       hand_l_fk_ctrl.Location.Z {type=MovieSceneScriptingFloatChannel, default=0.0668, thinned_from=21, keys [[0,0.0668],[354,0.0668]]}
       hand_l_fk_ctrl.Rotation.X {
        type=MovieSceneScriptingFloatChannel
        default=-0.8938
        thinned_from=21
        keys [[0,1.2218],[25,4.6957],[34,4.6957],[40,-57.2519],[71,0.2642],[90,0.2642],[121,4.7589],[150,4.6957],[164,-0.5381],[205,-0.5381],[216,2.7744],[240,-0.5381],[260,16.3696],[301,16.3696],[322,1.2218],[354,4.6957]]
       }
       hand_l_fk_ctrl.Rotation.Y {
        type=MovieSceneScriptingFloatChannel
        default=0.1458
        thinned_from=21
        keys [[0,43.4725],[25,40.78],[34,40.78],[40,59.1883],[71,33.9878],[90,33.9878],[121,40.5639],[150,40.78],[164,42.119],[205,42.119],[216,38.8492],[240,42.119],[260,48.4972],[301,48.4972],[322,43.4725],[354,40.78]]
       }
       hand_l_fk_ctrl.Rotation.Z {
        type=MovieSceneScriptingFloatChannel
        default=-0.1579
        thinned_from=21
        keys [[0,-44.8004],[25,-42.483],[34,-42.483],[40,-89.8153],[71,-46.2615],[90,-46.2615],[121,-42.346],[150,-42.483],[164,-47.8849],[205,-47.8849],[216,-46.4627],[240,-47.8849],[260,-24.524],[301,-24.524],[322,-44.8004],[354,-42.483]]
       }
       hand_l_fk_ctrl.Scale.X {type=MovieSceneScriptingFloatChannel, default=1, const [1,21]}
       hand_l_fk_ctrl.Scale.Y {type=MovieSceneScriptingFloatChannel, default=1, const [1,21]}
       hand_l_fk_ctrl.Scale.Z {type=MovieSceneScriptingFloatChannel, default=1, const [1,21]}
       thumb_01_l_ctrl.Location.X {type=MovieSceneScriptingFloatChannel, default=0, const [0,16]}
       thumb_01_l_ctrl.Location.Y {type=MovieSceneScriptingFloatChannel, default=0, const [0,16]}
       thumb_01_l_ctrl.Location.Z {type=MovieSceneScriptingFloatChannel, default=0, const [0,16]}
       thumb_01_l_ctrl.Rotation.X {type=MovieSceneScriptingFloatChannel, default=0, const [27.2929,16]}
       thumb_01_l_ctrl.Rotation.Y {type=MovieSceneScriptingFloatChannel, default=0, const [-31.948,16]}
       thumb_01_l_ctrl.Rotation.Z {type=MovieSceneScriptingFloatChannel, default=0, const [16.5161,16]}
       thumb_01_l_ctrl.Scale.X {type=MovieSceneScriptingFloatChannel, default=1, const [1,16]}
       thumb_01_l_ctrl.Scale.Y {type=MovieSceneScriptingFloatChannel, default=1, const [1,16]}
       thumb_01_l_ctrl.Scale.Z {type=MovieSceneScriptingFloatChannel, default=1, const [1,16]}
       thumb_02_l_ctrl.Location.X {type=MovieSceneScriptingFloatChannel, default=0, const [0,16]}
       thumb_02_l_ctrl.Location.Y {type=MovieSceneScriptingFloatChannel, default=0, const [0,16]}
       thumb_02_l_ctrl.Location.Z {type=MovieSceneScriptingFloatChannel, default=0, const [0,16]}
       thumb_02_l_ctrl.Rotation.X {type=MovieSceneScriptingFloatChannel, default=0, const [0,16]}
       thumb_02_l_ctrl.Rotation.Y {type=MovieSceneScriptingFloatChannel, default=0, const [0,16]}
       thumb_02_l_ctrl.Rotation.Z {type=MovieSceneScriptingFloatChannel, default=0, const [12.5674,16]}
       thumb_02_l_ctrl.Scale.X {type=MovieSceneScriptingFloatChannel, default=1, const [1,16]}
       thumb_02_l_ctrl.Scale.Y {type=MovieSceneScriptingFloatChannel, default=1, const [1,16]}
       thumb_02_l_ctrl.Scale.Z {type=MovieSceneScriptingFloatChannel, default=1, const [1,16]}
       thumb_03_l_ctrl.Location.X {type=MovieSceneScriptingFloatChannel, default=0, const [0,16]}
       thumb_03_l_ctrl.Location.Y {type=MovieSceneScriptingFloatChannel, default=0, const [0,16]}
       thumb_03_l_ctrl.Location.Z {type=MovieSceneScriptingFloatChannel, default=0, const [0,16]}
       thumb_03_l_ctrl.Rotation.X {type=MovieSceneScriptingFloatChannel, default=0, const [0,16]}
       thumb_03_l_ctrl.Rotation.Y {type=MovieSceneScriptingFloatChannel, default=0, const [0,16]}
       thumb_03_l_ctrl.Rotation.Z {type=MovieSceneScriptingFloatChannel, default=0, const [20.8894,16]}
       thumb_03_l_ctrl.Scale.X {type=MovieSceneScriptingFloatChannel, default=1, const [1,16]}
       thumb_03_l_ctrl.Scale.Y {type=MovieSceneScriptingFloatChannel, default=1, const [1,16]}
       thumb_03_l_ctrl.Scale.Z {type=MovieSceneScriptingFloatChannel, default=1, const [1,16]}
       index_metacarpal_l_ctrl.Location.X {type=MovieSceneScriptingFloatChannel, default=0, const [0,16]}
       index_metacarpal_l_ctrl.Location.Y {type=MovieSceneScriptingFloatChannel, default=0, const [0,16]}
       index_metacarpal_l_ctrl.Location.Z {type=MovieSceneScriptingFloatChannel, default=0, const [0,16]}
       index_metacarpal_l_ctrl.Rotation.X {type=MovieSceneScriptingFloatChannel, default=0, const [0,16]}
       index_metacarpal_l_ctrl.Rotation.Y {type=MovieSceneScriptingFloatChannel, default=0, const [0,16]}
       index_metacarpal_l_ctrl.Rotation.Z {type=MovieSceneScriptingFloatChannel, default=8.1349, const [8.1349,16]}
       index_metacarpal_l_ctrl.Scale.X {type=MovieSceneScriptingFloatChannel, default=1, const [1,16]}
       index_metacarpal_l_ctrl.Scale.Y {type=MovieSceneScriptingFloatChannel, default=1, const [1,16]}
       index_metacarpal_l_ctrl.Scale.Z {type=MovieSceneScriptingFloatChannel, default=1, const [1,16]}
       index_01_l_ctrl.Location.X {type=MovieSceneScriptingFloatChannel, default=0, const [0,16]}
       index_01_l_ctrl.Location.Y {type=MovieSceneScriptingFloatChannel, default=0, const [0,16]}
       index_01_l_ctrl.Location.Z {type=MovieSceneScriptingFloatChannel, default=0, const [0,16]}
       index_01_l_ctrl.Rotation.X {type=MovieSceneScriptingFloatChannel, default=-9.7327, const [9.5688,16]}
       index_01_l_ctrl.Rotation.Y {type=MovieSceneScriptingFloatChannel, default=-8.5609, const [-6.9345,16]}
       index_01_l_ctrl.Rotation.Z {type=MovieSceneScriptingFloatChannel, default=10.8513, const [11.1285,16]}
       index_01_l_ctrl.Scale.X {type=MovieSceneScriptingFloatChannel, default=1, const [1,16]}
       index_01_l_ctrl.Scale.Y {type=MovieSceneScriptingFloatChannel, default=1, const [1,16]}
       index_01_l_ctrl.Scale.Z {type=MovieSceneScriptingFloatChannel, default=1, const [1,16]}
       index_02_l_ctrl.Location.X {type=MovieSceneScriptingFloatChannel, default=0, const [0,16]}
       index_02_l_ctrl.Location.Y {type=MovieSceneScriptingFloatChannel, default=0, const [0,16]}
       index_02_l_ctrl.Location.Z {type=MovieSceneScriptingFloatChannel, default=0, const [0,16]}
       index_02_l_ctrl.Rotation.X {type=MovieSceneScriptingFloatChannel, default=0, const [0,16]}
       index_02_l_ctrl.Rotation.Y {type=MovieSceneScriptingFloatChannel, default=0, const [0,16]}
       index_02_l_ctrl.Rotation.Z {type=MovieSceneScriptingFloatChannel, default=23.4199, const [14.0252,16]}
       index_02_l_ctrl.Scale.X {type=MovieSceneScriptingFloatChannel, default=1, const [1,16]}
       index_02_l_ctrl.Scale.Y {type=MovieSceneScriptingFloatChannel, default=1, const [1,16]}
       index_02_l_ctrl.Scale.Z {type=MovieSceneScriptingFloatChannel, default=1, const [1,16]}
       index_03_l_ctrl.Location.X {type=MovieSceneScriptingFloatChannel, default=0, const [0,16]}
       index_03_l_ctrl.Location.Y {type=MovieSceneScriptingFloatChannel, default=0, const [0,16]}
       index_03_l_ctrl.Location.Z {type=MovieSceneScriptingFloatChannel, default=0, const [0,16]}
       index_03_l_ctrl.Rotation.X {type=MovieSceneScriptingFloatChannel, default=0, const [0,16]}
       index_03_l_ctrl.Rotation.Y {type=MovieSceneScriptingFloatChannel, default=0, const [0,16]}
       index_03_l_ctrl.Rotation.Z {type=MovieSceneScriptingFloatChannel, default=27.6891, const [25.7629,16]}
       index_03_l_ctrl.Scale.X {type=MovieSceneScriptingFloatChannel, default=1, const [1,16]}
       index_03_l_ctrl.Scale.Y {type=MovieSceneScriptingFloatChannel, default=1, const [1,16]}
       index_03_l_ctrl.Scale.Z {type=MovieSceneScriptingFloatChannel, default=1, const [1,16]}
       middle_metacarpal_l_ctrl.Location.X {type=MovieSceneScriptingFloatChannel, default=0, const [0,16]}
       middle_metacarpal_l_ctrl.Location.Y {type=MovieSceneScriptingFloatChannel, default=0, const [0,16]}
       middle_metacarpal_l_ctrl.Location.Z {type=MovieSceneScriptingFloatChannel, default=0, const [0,16]}
       middle_metacarpal_l_ctrl.Rotation.X {type=MovieSceneScriptingFloatChannel, default=-0.2325, const [-0.2325,16]}
       middle_metacarpal_l_ctrl.Rotation.Y {type=MovieSceneScriptingFloatChannel, default=-3.2794, const [-3.2794,16]}
       middle_metacarpal_l_ctrl.Rotation.Z {type=MovieSceneScriptingFloatChannel, default=4.0575, const [4.0575,16]}
       middle_metacarpal_l_ctrl.Scale.X {type=MovieSceneScriptingFloatChannel, default=1, const [1,16]}
       middle_metacarpal_l_ctrl.Scale.Y {type=MovieSceneScriptingFloatChannel, default=1, const [1,16]}
       middle_metacarpal_l_ctrl.Scale.Z {type=MovieSceneScriptingFloatChannel, default=1, const [1,16]}
       middle_01_l_ctrl.Location.X {type=MovieSceneScriptingFloatChannel, default=0, const [0,16]}
       middle_01_l_ctrl.Location.Y {type=MovieSceneScriptingFloatChannel, default=0, const [0,16]}
       middle_01_l_ctrl.Location.Z {type=MovieSceneScriptingFloatChannel, default=0, const [0,16]}
       middle_01_l_ctrl.Rotation.X {type=MovieSceneScriptingFloatChannel, default=-0.843, const [11.0189,16]}
       middle_01_l_ctrl.Rotation.Y {type=MovieSceneScriptingFloatChannel, default=-3.0866, const [-2.7747,16]}
       middle_01_l_ctrl.Rotation.Z {type=MovieSceneScriptingFloatChannel, default=18.1497, const [16.5399,16]}
       middle_01_l_ctrl.Scale.X {type=MovieSceneScriptingFloatChannel, default=1, const [1,16]}
       middle_01_l_ctrl.Scale.Y {type=MovieSceneScriptingFloatChannel, default=1, const [1,16]}
       middle_01_l_ctrl.Scale.Z {type=MovieSceneScriptingFloatChannel, default=1, const [1,16]}
       middle_02_l_ctrl.Location.X {type=MovieSceneScriptingFloatChannel, default=0, const [0,16]}
       middle_02_l_ctrl.Location.Y {type=MovieSceneScriptingFloatChannel, default=0, const [0,16]}
       middle_02_l_ctrl.Location.Z {type=MovieSceneScriptingFloatChannel, default=0, const [0,16]}
       middle_02_l_ctrl.Rotation.X {type=MovieSceneScriptingFloatChannel, default=0, const [0,16]}
       middle_02_l_ctrl.Rotation.Y {type=MovieSceneScriptingFloatChannel, default=0, const [0,16]}
       middle_02_l_ctrl.Rotation.Z {type=MovieSceneScriptingFloatChannel, default=10.1049, const [-2.4917,16]}
       middle_02_l_ctrl.Scale.X {type=MovieSceneScriptingFloatChannel, default=1, const [1,16]}
       middle_02_l_ctrl.Scale.Y {type=MovieSceneScriptingFloatChannel, default=1, const [1,16]}
       middle_02_l_ctrl.Scale.Z {type=MovieSceneScriptingFloatChannel, default=1, const [1,16]}
       middle_03_l_ctrl.Location.X {type=MovieSceneScriptingFloatChannel, default=0, const [0,16]}
       middle_03_l_ctrl.Location.Y {type=MovieSceneScriptingFloatChannel, default=0, const [0,16]}
       middle_03_l_ctrl.Location.Z {type=MovieSceneScriptingFloatChannel, default=0, const [0,16]}
       middle_03_l_ctrl.Rotation.X {type=MovieSceneScriptingFloatChannel, default=0, const [0,16]}
       middle_03_l_ctrl.Rotation.Y {type=MovieSceneScriptingFloatChannel, default=0, const [0,16]}
       middle_03_l_ctrl.Rotation.Z {type=MovieSceneScriptingFloatChannel, default=26.7545, const [26.7545,16]}
       middle_03_l_ctrl.Scale.X {type=MovieSceneScriptingFloatChannel, default=1, const [1,16]}
       middle_03_l_ctrl.Scale.Y {type=MovieSceneScriptingFloatChannel, default=1, const [1,16]}
       middle_03_l_ctrl.Scale.Z {type=MovieSceneScriptingFloatChannel, default=1, const [1,16]}
       ring_metacarpal_l_ctrl.Location.X {type=MovieSceneScriptingFloatChannel, default=0, const [0,16]}
       ring_metacarpal_l_ctrl.Location.Y {type=MovieSceneScriptingFloatChannel, default=0, const [0,16]}
       ring_metacarpal_l_ctrl.Location.Z {type=MovieSceneScriptingFloatChannel, default=0, const [0,16]}
       ring_metacarpal_l_ctrl.Rotation.X {type=MovieSceneScriptingFloatChannel, default=0, const [0,16]}
       ring_metacarpal_l_ctrl.Rotation.Y {type=MovieSceneScriptingFloatChannel, default=-2.346, const [-2.346,16]}
       ring_metacarpal_l_ctrl.Rotation.Z {type=MovieSceneScriptingFloatChannel, default=0, const [0,16]}
       ring_metacarpal_l_ctrl.Scale.X {type=MovieSceneScriptingFloatChannel, default=1, const [1,16]}
       ring_metacarpal_l_ctrl.Scale.Y {type=MovieSceneScriptingFloatChannel, default=1, const [1,16]}
       ring_metacarpal_l_ctrl.Scale.Z {type=MovieSceneScriptingFloatChannel, default=1, const [1,16]}
       ring_01_l_ctrl.Location.X {type=MovieSceneScriptingFloatChannel, default=0, const [0,16]}
       ring_01_l_ctrl.Location.Y {type=MovieSceneScriptingFloatChannel, default=0, const [0,16]}
       ring_01_l_ctrl.Location.Z {type=MovieSceneScriptingFloatChannel, default=0, const [0,16]}
       ring_01_l_ctrl.Rotation.X {type=MovieSceneScriptingFloatChannel, default=6.6072, const [13.3385,16]}
       ring_01_l_ctrl.Rotation.Y {type=MovieSceneScriptingFloatChannel, default=-13.306, const [-14.436,16]}
       ring_01_l_ctrl.Rotation.Z {type=MovieSceneScriptingFloatChannel, default=16.5011, const [21.2011,16]}
       ring_01_l_ctrl.Scale.X {type=MovieSceneScriptingFloatChannel, default=1, const [1,16]}
       ring_01_l_ctrl.Scale.Y {type=MovieSceneScriptingFloatChannel, default=1, const [1,16]}
       ring_01_l_ctrl.Scale.Z {type=MovieSceneScriptingFloatChannel, default=1, const [1,16]}
       ring_02_l_ctrl.Location.X {type=MovieSceneScriptingFloatChannel, default=0, const [0,16]}
       ring_02_l_ctrl.Location.Y {type=MovieSceneScriptingFloatChannel, default=0, const [0,16]}
       ring_02_l_ctrl.Location.Z {type=MovieSceneScriptingFloatChannel, default=0, const [0,16]}
       ring_02_l_ctrl.Rotation.X {type=MovieSceneScriptingFloatChannel, default=0, const [0,16]}
       ring_02_l_ctrl.Rotation.Y {type=MovieSceneScriptingFloatChannel, default=0, const [0,16]}
       ring_02_l_ctrl.Rotation.Z {type=MovieSceneScriptingFloatChannel, default=-10.3965, const [-10.3965,16]}
       ring_02_l_ctrl.Scale.X {type=MovieSceneScriptingFloatChannel, default=1, const [1,16]}
       ring_02_l_ctrl.Scale.Y {type=MovieSceneScriptingFloatChannel, default=1, const [1,16]}
       ring_02_l_ctrl.Scale.Z {type=MovieSceneScriptingFloatChannel, default=1, const [1,16]}
       ring_03_l_ctrl.Location.X {type=MovieSceneScriptingFloatChannel, default=0, const [0,16]}
       ring_03_l_ctrl.Location.Y {type=MovieSceneScriptingFloatChannel, default=0, const [0,16]}
       ring_03_l_ctrl.Location.Z {type=MovieSceneScriptingFloatChannel, default=0, const [0,16]}
       ring_03_l_ctrl.Rotation.X {type=MovieSceneScriptingFloatChannel, default=0, const [0,16]}
       ring_03_l_ctrl.Rotation.Y {type=MovieSceneScriptingFloatChannel, default=0, const [0,16]}
       ring_03_l_ctrl.Rotation.Z {type=MovieSceneScriptingFloatChannel, default=28.0865, const [28.0755,16]}
       ring_03_l_ctrl.Scale.X {type=MovieSceneScriptingFloatChannel, default=1, const [1,16]}
       ring_03_l_ctrl.Scale.Y {type=MovieSceneScriptingFloatChannel, default=1, const [1,16]}
       ring_03_l_ctrl.Scale.Z {type=MovieSceneScriptingFloatChannel, default=1, const [1,16]}
       pinky_metacarpal_l_ctrl.Location.X {type=MovieSceneScriptingFloatChannel, default=0, const [0,16]}
       pinky_metacarpal_l_ctrl.Location.Y {type=MovieSceneScriptingFloatChannel, default=0, const [0,16]}
       pinky_metacarpal_l_ctrl.Location.Z {type=MovieSceneScriptingFloatChannel, default=0, const [0,16]}
       pinky_metacarpal_l_ctrl.Rotation.X {type=MovieSceneScriptingFloatChannel, default=0.2697, const [0.2711,16]}
       pinky_metacarpal_l_ctrl.Rotation.Y {type=MovieSceneScriptingFloatChannel, default=-5.469, const [-8.0536,16]}
       pinky_metacarpal_l_ctrl.Rotation.Z {type=MovieSceneScriptingFloatChannel, default=-3.6065, const [-3.6187,16]}
       pinky_metacarpal_l_ctrl.Scale.X {type=MovieSceneScriptingFloatChannel, default=1, const [1,16]}
       pinky_metacarpal_l_ctrl.Scale.Y {type=MovieSceneScriptingFloatChannel, default=1, const [1,16]}
       pinky_metacarpal_l_ctrl.Scale.Z {type=MovieSceneScriptingFloatChannel, default=1, const [1,16]}
       pinky_01_l_ctrl.Location.X {type=MovieSceneScriptingFloatChannel, default=0, const [0,16]}
       pinky_01_l_ctrl.Location.Y {type=MovieSceneScriptingFloatChannel, default=0, const [0,16]}
       pinky_01_l_ctrl.Location.Z {type=MovieSceneScriptingFloatChannel, default=0, const [0,16]}
       pinky_01_l_ctrl.Rotation.X {type=MovieSceneScriptingFloatChannel, default=12.0626, const [25.3683,16]}
       pinky_01_l_ctrl.Rotation.Y {type=MovieSceneScriptingFloatChannel, default=-4.4978, const [-4.4978,16]}
       pinky_01_l_ctrl.Rotation.Z {type=MovieSceneScriptingFloatChannel, default=6.5382, const [6.5382,16]}
       pinky_01_l_ctrl.Scale.X {type=MovieSceneScriptingFloatChannel, default=1, const [1,16]}
       pinky_01_l_ctrl.Scale.Y {type=MovieSceneScriptingFloatChannel, default=1, const [1,16]}
       pinky_01_l_ctrl.Scale.Z {type=MovieSceneScriptingFloatChannel, default=1, const [1,16]}
       pinky_02_l_ctrl.Location.X {type=MovieSceneScriptingFloatChannel, default=0, const [0,16]}
       pinky_02_l_ctrl.Location.Y {type=MovieSceneScriptingFloatChannel, default=0, const [0,16]}
       pinky_02_l_ctrl.Location.Z {type=MovieSceneScriptingFloatChannel, default=0, const [0,16]}
       pinky_02_l_ctrl.Rotation.X {type=MovieSceneScriptingFloatChannel, default=0, const [0,16]}
       pinky_02_l_ctrl.Rotation.Y {type=MovieSceneScriptingFloatChannel, default=0, const [0,16]}
       pinky_02_l_ctrl.Rotation.Z {type=MovieSceneScriptingFloatChannel, default=18.3226, const [10.2539,16]}
       pinky_02_l_ctrl.Scale.X {type=MovieSceneScriptingFloatChannel, default=1, const [1,16]}
       pinky_02_l_ctrl.Scale.Y {type=MovieSceneScriptingFloatChannel, default=1, const [1,16]}
       pinky_02_l_ctrl.Scale.Z {type=MovieSceneScriptingFloatChannel, default=1, const [1,16]}
       pinky_03_l_ctrl.Location.X {type=MovieSceneScriptingFloatChannel, default=0, const [0,16]}
       pinky_03_l_ctrl.Location.Y {type=MovieSceneScriptingFloatChannel, default=0, const [0,16]}
       pinky_03_l_ctrl.Location.Z {type=MovieSceneScriptingFloatChannel, default=0, const [0,16]}
       pinky_03_l_ctrl.Rotation.X {type=MovieSceneScriptingFloatChannel, default=0, const [0,16]}
       pinky_03_l_ctrl.Rotation.Y {type=MovieSceneScriptingFloatChannel, default=0, const [0,16]}
       pinky_03_l_ctrl.Rotation.Z {type=MovieSceneScriptingFloatChannel, default=10.982, const [10.982,16]}
       pinky_03_l_ctrl.Scale.X {type=MovieSceneScriptingFloatChannel, default=1, const [1,16]}
       pinky_03_l_ctrl.Scale.Y {type=MovieSceneScriptingFloatChannel, default=1, const [1,16]}
       pinky_03_l_ctrl.Scale.Z {type=MovieSceneScriptingFloatChannel, default=1, const [1,16]}
       hand_l_ik_ctrl.Location.X {type=MovieSceneScriptingFloatChannel, default=0, thinned_from=18, keys [[-1,-17.463],[-0.0013,-18.2486],[0,66.303],[352,66.303]]}
       hand_l_ik_ctrl.Location.Y {type=MovieSceneScriptingFloatChannel, default=0, thinned_from=18, keys [[-1,8.7993],[-0.0013,8.4827],[0,42.8812],[352,42.8812]]}
       hand_l_ik_ctrl.Location.Z {type=MovieSceneScriptingFloatChannel, default=0, thinned_from=18, keys [[-1,-4.5505],[-0.0013,-4.1976],[0,-44.9584],[352,-44.9584]]}
       hand_l_ik_ctrl.Rotation.X {type=MovieSceneScriptingFloatChannel, default=0, thinned_from=23, keys [[-1,53.826],[-0.0013,53.8487],[0,50.3031],[352,50.3031]]}
       hand_l_ik_ctrl.Rotation.Y {type=MovieSceneScriptingFloatChannel, default=0, thinned_from=23, keys [[-1,66.604],[-0.0013,66.6006],[0,63.2485],[352,63.2485]]}
       hand_l_ik_ctrl.Rotation.Z {type=MovieSceneScriptingFloatChannel, default=0, thinned_from=23, keys [[-1,43.1645],[-0.0013,43.1896],[0,40.5828],[352,40.5828]]}
       hand_l_ik_ctrl.Scale.X {type=MovieSceneScriptingFloatChannel, default=1, const [1,23]}
       hand_l_ik_ctrl.Scale.Y {type=MovieSceneScriptingFloatChannel, default=1, const [1,23]}
       hand_l_ik_ctrl.Scale.Z {type=MovieSceneScriptingFloatChannel, default=1, const [1,23]}
       Softness {type=MovieSceneScriptingFloatChannel, default=0, const [0,23]}
       hand_l_ikorient_ctrl.Rotation.Roll {type=MovieSceneScriptingFloatChannel, default=0, const [0,21]}
       hand_l_ikorient_ctrl.Rotation.Pitch {type=MovieSceneScriptingFloatChannel, default=0, const [0,21]}
       hand_l_ikorient_ctrl.Rotation.Yaw {type=MovieSceneScriptingFloatChannel, default=0, const [0,21]}
       arm_l_pv_ik_ctrl.Location.X {
        type=MovieSceneScriptingFloatChannel
        default=0
        keys [[0,61.2458],[29,62.3515],[59,47.8639],[78,45.8364],[114,64.7602],[147,61.2458],[238,61.2458],[281,82.7817],[303,88.418],[338,57.4321],[352,61.2458]]
       }
       arm_l_pv_ik_ctrl.Location.Y {
        type=MovieSceneScriptingFloatChannel
        default=0
        keys [[0,-42.4403],[29,-39.0932],[59,-64.2328],[78,-67.5347],[114,-36.7171],[147,-42.4403],[238,-42.4403],[281,-32.7835],[303,-30.2035],[338,-45.6158],[352,-42.4403]]
       }
       arm_l_pv_ik_ctrl.Location.Z {type=MovieSceneScriptingFloatChannel, default=0, const [10.7674,11]}
       arm_l_pv_ik_ctrl.Rotation.X {type=MovieSceneScriptingFloatChannel, default=0, const [0,11]}
       arm_l_pv_ik_ctrl.Rotation.Y {type=MovieSceneScriptingFloatChannel, default=0, const [0,11]}
       arm_l_pv_ik_ctrl.Rotation.Z {type=MovieSceneScriptingFloatChannel, default=0, const [0,11]}
       arm_l_pv_ik_ctrl.Scale.X {type=MovieSceneScriptingFloatChannel, default=1, const [1,11]}
       arm_l_pv_ik_ctrl.Scale.Y {type=MovieSceneScriptingFloatChannel, default=1, const [1,11]}
       arm_l_pv_ik_ctrl.Scale.Z {type=MovieSceneScriptingFloatChannel, default=1, const [1,11]}
       upperarm_r_fk_ctrl.Location.X {type=MovieSceneScriptingFloatChannel, default=0, thinned_from=21, keys [[0,0],[353,0]]}
       upperarm_r_fk_ctrl.Location.Y {type=MovieSceneScriptingFloatChannel, default=0, thinned_from=21, keys [[0,0],[352,0]]}
       upperarm_r_fk_ctrl.Location.Z {type=MovieSceneScriptingFloatChannel, default=0, thinned_from=21, keys [[0,0],[352,0]]}
       upperarm_r_fk_ctrl.Rotation.X {
        type=MovieSceneScriptingFloatChannel
        default=1.1491
        thinned_from=21
        keys [[0,46.2313],[25,49.3483],[34,49.3483],[40,50.377],[71,45.053],[90,45.053],[121,49.2235],[150,49.3483],[164,46.743],[205,46.743],[216,45.913],[236,46.743],[258,62.4283],[299,62.4283],[320,49.3483],[352,49.3483]]
       }
       upperarm_r_fk_ctrl.Rotation.Y {
        type=MovieSceneScriptingFloatChannel
        default=-0.7961
        thinned_from=21
        keys [[0,17.3021],[25,17.5259],[34,17.5259],[40,-37.478],[71,16.8336],[90,16.8336],[121,17.5545],[150,17.5259],[164,16.212],[205,16.212],[216,15.7013],[236,16.212],[258,22.6112],[299,22.6112],[320,17.5259],[352,17.5259]]
       }
       upperarm_r_fk_ctrl.Rotation.Z {
        type=MovieSceneScriptingFloatChannel
        default=-0.0209
        thinned_from=21
        keys [[0,-26.5061],[25,-30.0312],[34,-30.0312],[40,-2.1843],[71,-38.1243],[90,-38.1243],[121,-29.8432],[150,-30.0312],[164,-31.5584],[205,-31.5584],[216,-32.0443],[236,-31.5584],[258,-18.4664],[299,-18.4664],[320,-30.0312],[352,-30.0312]]
       }
       upperarm_r_fk_ctrl.Scale.X {type=MovieSceneScriptingFloatChannel, default=1, const [1,21]}
       upperarm_r_fk_ctrl.Scale.Y {type=MovieSceneScriptingFloatChannel, default=1, const [1,21]}
       upperarm_r_fk_ctrl.Scale.Z {type=MovieSceneScriptingFloatChannel, default=1, const [1,21]}
       lowerarm_r_fk_ctrl.Location.X {type=MovieSceneScriptingFloatChannel, default=0, thinned_from=21, keys [[0,0],[352,0]]}
       lowerarm_r_fk_ctrl.Location.Y {type=MovieSceneScriptingFloatChannel, default=0, thinned_from=21, keys [[0,0],[352,0]]}
       lowerarm_r_fk_ctrl.Location.Z {type=MovieSceneScriptingFloatChannel, default=0, thinned_from=21, keys [[0,0],[352,0]]}
       lowerarm_r_fk_ctrl.Rotation.X {
        type=MovieSceneScriptingFloatChannel
        default=0.2616
        thinned_from=21
        keys [[0,1.1078],[34,1.1071],[40,-0.2881],[71,1.1047],[90,1.1047],[121,1.1077],[216,1.1059],[236,1.1071],[258,1.102],[299,1.102],[320,1.1071],[352,1.1071]]
       }
       lowerarm_r_fk_ctrl.Rotation.Y {
        type=MovieSceneScriptingFloatChannel
        default=1.0909
        thinned_from=21
        keys [[0,0.1765],[25,0.1807],[34,0.1807],[40,1.0842],[71,0.1948],[90,0.1948],[121,0.1771],[150,0.1807],[205,0.1807],[216,0.1884],[236,0.1807],[258,0.2095],[299,0.2095],[320,0.1807],[352,0.1807]]
       }
       lowerarm_r_fk_ctrl.Rotation.Z {
        type=MovieSceneScriptingFloatChannel
        default=-0.1228
        thinned_from=21
        keys [[0,67.3411],[34,67.1213],[40,-28.4891],[71,66.3928],[90,66.3928],[121,67.3117],[205,67.1213],[216,66.7269],[236,67.1213],[258,65.628],[299,65.628],[320,67.1213],[352,67.1213]]
       }
       lowerarm_r_fk_ctrl.Scale.X {type=MovieSceneScriptingFloatChannel, default=1, const [1,21]}
       lowerarm_r_fk_ctrl.Scale.Y {type=MovieSceneScriptingFloatChannel, default=1, const [1,21]}
       lowerarm_r_fk_ctrl.Scale.Z {type=MovieSceneScriptingFloatChannel, default=1, const [1,21]}
       hand_r_fk_ctrl.Location.X {type=MovieSceneScriptingFloatChannel, default=0, thinned_from=21, keys [[0,0],[352,0]]}
       hand_r_fk_ctrl.Location.Y {type=MovieSceneScriptingFloatChannel, default=-0.431, thinned_from=21, keys [[0,-0.431],[352,-0.431]]}
       hand_r_fk_ctrl.Location.Z {type=MovieSceneScriptingFloatChannel, default=0.0668, thinned_from=21, keys [[0,0.0668],[352,0.0668]]}
       hand_r_fk_ctrl.Rotation.X {
        type=MovieSceneScriptingFloatChannel
        default=-0.8938
        thinned_from=21
        keys [[0,12.6472],[25,8.8641],[34,8.8641],[40,106.2389],[71,11.8013],[90,11.8013],[121,8.8669],[150,8.8641],[164,10.5954],[205,10.5954],[216,11.0424],[236,10.5954],[258,-0.3716],[299,-0.3716],[320,8.8641],[352,8.8641]]
       }
       hand_r_fk_ctrl.Rotation.Y {
        type=MovieSceneScriptingFloatChannel
        default=0.1458
        thinned_from=21
        keys [[0,27.7911],[25,28.7775],[34,28.7775],[40,-10.4668],[71,36.4703],[90,36.4703],[121,28.4441],[150,28.7775],[164,29.6479],[205,29.6479],[216,30.2457],[236,29.6479],[258,21.57],[299,21.57],[320,28.7775],[352,28.7775]]
       }
       hand_r_fk_ctrl.Rotation.Z {
        type=MovieSceneScriptingFloatChannel
        default=-0.1579
        thinned_from=21
        keys [[0,-20.1875],[25,-31.5346],[34,-31.5346],[40,92.167],[71,-27.4896],[90,-27.4896],[121,-31.368],[150,-31.5346],[164,-28.3453],[205,-28.3453],[216,-27.3635],[236,-28.3453],[258,-45.5456],[299,-45.5456],[320,-31.5346],[352,-31.5346]]
       }
       hand_r_fk_ctrl.Scale.X {type=MovieSceneScriptingFloatChannel, default=1, thinned_from=21, keys [[0,1],[34,1],[40,-1],[71,1],[352,1]]}
       hand_r_fk_ctrl.Scale.Y {type=MovieSceneScriptingFloatChannel, default=1, const [1,21]}
       hand_r_fk_ctrl.Scale.Z {type=MovieSceneScriptingFloatChannel, default=1, const [1,21]}
       thumb_01_r_ctrl.Location.X {type=MovieSceneScriptingFloatChannel, default=0, thinned_from=21, keys [[0,0],[352,0]]}
       thumb_01_r_ctrl.Location.Y {type=MovieSceneScriptingFloatChannel, default=0, thinned_from=21, keys [[0,0],[352,0]]}
       thumb_01_r_ctrl.Location.Z {type=MovieSceneScriptingFloatChannel, default=0, thinned_from=21, keys [[0,0],[352,0]]}
       thumb_01_r_ctrl.Rotation.X {type=MovieSceneScriptingFloatChannel, default=15.8656, const [15.8656,21]}
       thumb_01_r_ctrl.Rotation.Y {type=MovieSceneScriptingFloatChannel, default=-36.5095, const [-36.5095,21]}
       thumb_01_r_ctrl.Rotation.Z {type=MovieSceneScriptingFloatChannel, default=-0.3323, thinned_from=21, keys [[0,-0.3323],[352,-0.3323]]}
       thumb_01_r_ctrl.Scale.X {type=MovieSceneScriptingFloatChannel, default=1, const [1,21]}
       thumb_01_r_ctrl.Scale.Y {type=MovieSceneScriptingFloatChannel, default=1, const [1,21]}
       thumb_01_r_ctrl.Scale.Z {type=MovieSceneScriptingFloatChannel, default=1, const [1,21]}
       thumb_02_r_ctrl.Location.X {type=MovieSceneScriptingFloatChannel, default=0, const [0,21]}
       thumb_02_r_ctrl.Location.Y {type=MovieSceneScriptingFloatChannel, default=0, const [0,21]}
       thumb_02_r_ctrl.Location.Z {type=MovieSceneScriptingFloatChannel, default=0, const [0,21]}
       thumb_02_r_ctrl.Rotation.X {type=MovieSceneScriptingFloatChannel, default=0, thinned_from=21, keys [[0,0],[352,0]]}
       thumb_02_r_ctrl.Rotation.Y {type=MovieSceneScriptingFloatChannel, default=0, thinned_from=21, keys [[0,0],[352,0]]}
       thumb_02_r_ctrl.Rotation.Z {type=MovieSceneScriptingFloatChannel, default=11.7116, const [11.7116,21]}
       thumb_02_r_ctrl.Scale.X {type=MovieSceneScriptingFloatChannel, default=1, const [1,21]}
       thumb_02_r_ctrl.Scale.Y {type=MovieSceneScriptingFloatChannel, default=1, const [1,21]}
       thumb_02_r_ctrl.Scale.Z {type=MovieSceneScriptingFloatChannel, default=1, const [1,21]}
       thumb_03_r_ctrl.Location.X {type=MovieSceneScriptingFloatChannel, default=0, thinned_from=21, keys [[0,0],[352,0]]}
       thumb_03_r_ctrl.Location.Y {type=MovieSceneScriptingFloatChannel, default=0, thinned_from=21, keys [[0,0],[352,0]]}
       thumb_03_r_ctrl.Location.Z {type=MovieSceneScriptingFloatChannel, default=0, thinned_from=21, keys [[0,0],[352,0]]}
       thumb_03_r_ctrl.Rotation.X {type=MovieSceneScriptingFloatChannel, default=0, thinned_from=21, keys [[0,0],[352,0]]}
       thumb_03_r_ctrl.Rotation.Y {type=MovieSceneScriptingFloatChannel, default=0, thinned_from=21, keys [[0,0],[352,0]]}
       thumb_03_r_ctrl.Rotation.Z {type=MovieSceneScriptingFloatChannel, default=32.4932, const [32.4932,21]}
       thumb_03_r_ctrl.Scale.X {type=MovieSceneScriptingFloatChannel, default=1, const [1,21]}
       thumb_03_r_ctrl.Scale.Y {type=MovieSceneScriptingFloatChannel, default=1, const [1,21]}
       thumb_03_r_ctrl.Scale.Z {type=MovieSceneScriptingFloatChannel, default=1, const [1,21]}
       index_metacarpal_r_ctrl.Location.X {type=MovieSceneScriptingFloatChannel, default=0, const [0,1]}
       index_metacarpal_r_ctrl.Location.Y {type=MovieSceneScriptingFloatChannel, default=0, const [0,1]}
       index_metacarpal_r_ctrl.Location.Z {type=MovieSceneScriptingFloatChannel, default=0, const [0,1]}
       index_metacarpal_r_ctrl.Rotation.X {type=MovieSceneScriptingFloatChannel, default=0, const [0,1]}
       index_metacarpal_r_ctrl.Rotation.Y {type=MovieSceneScriptingFloatChannel, default=0, const [0,1]}
       index_metacarpal_r_ctrl.Rotation.Z {type=MovieSceneScriptingFloatChannel, default=0, const [2.4531,1]}
       index_metacarpal_r_ctrl.Scale.X {type=MovieSceneScriptingFloatChannel, default=1, const [1,1]}
       index_metacarpal_r_ctrl.Scale.Y {type=MovieSceneScriptingFloatChannel, default=1, const [1,1]}
       index_metacarpal_r_ctrl.Scale.Z {type=MovieSceneScriptingFloatChannel, default=1, const [1,1]}
       index_01_r_ctrl.Location.X {type=MovieSceneScriptingFloatChannel, default=0, thinned_from=21, keys [[0,0],[352,0]]}
       index_01_r_ctrl.Location.Y {type=MovieSceneScriptingFloatChannel, default=0, thinned_from=21, keys [[0,0],[352,0]]}
       index_01_r_ctrl.Location.Z {type=MovieSceneScriptingFloatChannel, default=0, thinned_from=21, keys [[0,0],[352,0]]}
       index_01_r_ctrl.Rotation.X {type=MovieSceneScriptingFloatChannel, default=0, thinned_from=21, keys [[0,0],[352,0]]}
       index_01_r_ctrl.Rotation.Y {type=MovieSceneScriptingFloatChannel, default=0, thinned_from=21, keys [[0,0],[352,0]]}
       index_01_r_ctrl.Rotation.Z {type=MovieSceneScriptingFloatChannel, default=16.8566, const [16.8566,21]}
       index_01_r_ctrl.Scale.X {type=MovieSceneScriptingFloatChannel, default=1, const [1,21]}
       index_01_r_ctrl.Scale.Y {type=MovieSceneScriptingFloatChannel, default=1, const [1,21]}
       index_01_r_ctrl.Scale.Z {type=MovieSceneScriptingFloatChannel, default=1, const [1,21]}
       index_02_r_ctrl.Location.X {type=MovieSceneScriptingFloatChannel, default=0, const [0,21]}
       index_02_r_ctrl.Location.Y {type=MovieSceneScriptingFloatChannel, default=0, const [0,21]}
       index_02_r_ctrl.Location.Z {type=MovieSceneScriptingFloatChannel, default=0, const [0,21]}
       index_02_r_ctrl.Rotation.X {type=MovieSceneScriptingFloatChannel, default=0, thinned_from=21, keys [[0,0],[352,0]]}
       index_02_r_ctrl.Rotation.Y {type=MovieSceneScriptingFloatChannel, default=0, thinned_from=21, keys [[0,0],[352,0]]}
       index_02_r_ctrl.Rotation.Z {type=MovieSceneScriptingFloatChannel, default=17.0889, const [17.0889,21]}
       index_02_r_ctrl.Scale.X {type=MovieSceneScriptingFloatChannel, default=1, const [1,21]}
       index_02_r_ctrl.Scale.Y {type=MovieSceneScriptingFloatChannel, default=1, const [1,21]}
       index_02_r_ctrl.Scale.Z {type=MovieSceneScriptingFloatChannel, default=1, const [1,21]}
       index_03_r_ctrl.Location.X {type=MovieSceneScriptingFloatChannel, default=0, thinned_from=21, keys [[0,0],[352,0]]}
       index_03_r_ctrl.Location.Y {type=MovieSceneScriptingFloatChannel, default=0, thinned_from=21, keys [[0,0],[352,0]]}
       index_03_r_ctrl.Location.Z {type=MovieSceneScriptingFloatChannel, default=0, thinned_from=21, keys [[0,0],[352,0]]}
       index_03_r_ctrl.Rotation.X {type=MovieSceneScriptingFloatChannel, default=0, thinned_from=21, keys [[0,0],[352,0]]}
       index_03_r_ctrl.Rotation.Y {type=MovieSceneScriptingFloatChannel, default=0, thinned_from=21, keys [[0,0],[352,0]]}
       index_03_r_ctrl.Rotation.Z {type=MovieSceneScriptingFloatChannel, default=27.1266, const [27.1266,21]}
       index_03_r_ctrl.Scale.X {type=MovieSceneScriptingFloatChannel, default=1, const [1,21]}
       index_03_r_ctrl.Scale.Y {type=MovieSceneScriptingFloatChannel, default=1, const [1,21]}
       index_03_r_ctrl.Scale.Z {type=MovieSceneScriptingFloatChannel, default=1, const [1,21]}
       middle_metacarpal_r_ctrl.Location.X {type=MovieSceneScriptingFloatChannel, default=0, const [0,1]}
       middle_metacarpal_r_ctrl.Location.Y {type=MovieSceneScriptingFloatChannel, default=0, const [0,1]}
       middle_metacarpal_r_ctrl.Location.Z {type=MovieSceneScriptingFloatChannel, default=0, const [0,1]}
       middle_metacarpal_r_ctrl.Rotation.X {type=MovieSceneScriptingFloatChannel, default=0, const [0,1]}
       middle_metacarpal_r_ctrl.Rotation.Y {type=MovieSceneScriptingFloatChannel, default=0, const [0,1]}
       middle_metacarpal_r_ctrl.Rotation.Z {type=MovieSceneScriptingFloatChannel, default=0, const [1.7966,1]}
       middle_metacarpal_r_ctrl.Scale.X {type=MovieSceneScriptingFloatChannel, default=1, const [1,1]}
       middle_metacarpal_r_ctrl.Scale.Y {type=MovieSceneScriptingFloatChannel, default=1, const [1,1]}
       middle_metacarpal_r_ctrl.Scale.Z {type=MovieSceneScriptingFloatChannel, default=1, const [1,1]}
       middle_01_r_ctrl.Location.X {type=MovieSceneScriptingFloatChannel, default=0, const [0,21]}
       middle_01_r_ctrl.Location.Y {type=MovieSceneScriptingFloatChannel, default=0, const [0,21]}
       middle_01_r_ctrl.Location.Z {type=MovieSceneScriptingFloatChannel, default=0, const [0,21]}
       middle_01_r_ctrl.Rotation.X {type=MovieSceneScriptingFloatChannel, default=0, thinned_from=21, keys [[0,0],[352,0]]}
       middle_01_r_ctrl.Rotation.Y {type=MovieSceneScriptingFloatChannel, default=0, thinned_from=21, keys [[0,0],[352,0]]}
       middle_01_r_ctrl.Rotation.Z {type=MovieSceneScriptingFloatChannel, default=13.1248, const [13.1248,21]}
       middle_01_r_ctrl.Scale.X {type=MovieSceneScriptingFloatChannel, default=1, const [1,21]}
       middle_01_r_ctrl.Scale.Y {type=MovieSceneScriptingFloatChannel, default=1, const [1,21]}
       middle_01_r_ctrl.Scale.Z {type=MovieSceneScriptingFloatChannel, default=1, const [1,21]}
       middle_02_r_ctrl.Location.X {type=MovieSceneScriptingFloatChannel, default=0, thinned_from=21, keys [[0,0],[352,0]]}
       middle_02_r_ctrl.Location.Y {type=MovieSceneScriptingFloatChannel, default=0, thinned_from=21, keys [[0,0],[352,0]]}
       middle_02_r_ctrl.Location.Z {type=MovieSceneScriptingFloatChannel, default=0, thinned_from=21, keys [[0,0],[352,0]]}
       middle_02_r_ctrl.Rotation.X {type=MovieSceneScriptingFloatChannel, default=0, thinned_from=21, keys [[0,0],[352,0]]}
       middle_02_r_ctrl.Rotation.Y {type=MovieSceneScriptingFloatChannel, default=0, thinned_from=21, keys [[0,0],[352,0]]}
       middle_02_r_ctrl.Rotation.Z {type=MovieSceneScriptingFloatChannel, default=1.5495, thinned_from=21, keys [[0,1.5495],[352,1.5495]]}
       middle_02_r_ctrl.Scale.X {type=MovieSceneScriptingFloatChannel, default=1, const [1,21]}
       middle_02_r_ctrl.Scale.Y {type=MovieSceneScriptingFloatChannel, default=1, const [1,21]}
       middle_02_r_ctrl.Scale.Z {type=MovieSceneScriptingFloatChannel, default=1, const [1,21]}
       middle_03_r_ctrl.Location.X {type=MovieSceneScriptingFloatChannel, default=0, thinned_from=21, keys [[0,0],[352,0]]}
       middle_03_r_ctrl.Location.Y {type=MovieSceneScriptingFloatChannel, default=0, thinned_from=21, keys [[0,0],[352,0]]}
       middle_03_r_ctrl.Location.Z {type=MovieSceneScriptingFloatChannel, default=0, thinned_from=21, keys [[0,0],[352,0]]}
       middle_03_r_ctrl.Rotation.X {type=MovieSceneScriptingFloatChannel, default=0, thinned_from=21, keys [[0,0],[352,0]]}
       middle_03_r_ctrl.Rotation.Y {type=MovieSceneScriptingFloatChannel, default=0, thinned_from=21, keys [[0,0],[352,0]]}
       middle_03_r_ctrl.Rotation.Z {type=MovieSceneScriptingFloatChannel, default=26.6798, const [26.6798,21]}
       middle_03_r_ctrl.Scale.X {type=MovieSceneScriptingFloatChannel, default=1, const [1,21]}
       middle_03_r_ctrl.Scale.Y {type=MovieSceneScriptingFloatChannel, default=1, const [1,21]}
       middle_03_r_ctrl.Scale.Z {type=MovieSceneScriptingFloatChannel, default=1, const [1,21]}
       ring_metacarpal_r_ctrl.Location.X {type=MovieSceneScriptingFloatChannel, default=0, const [0,1]}
       ring_metacarpal_r_ctrl.Location.Y {type=MovieSceneScriptingFloatChannel, default=0, const [0,1]}
       ring_metacarpal_r_ctrl.Location.Z {type=MovieSceneScriptingFloatChannel, default=0, const [0,1]}
       ring_metacarpal_r_ctrl.Rotation.X {type=MovieSceneScriptingFloatChannel, default=0, const [0.0002,1]}
       ring_metacarpal_r_ctrl.Rotation.Y {type=MovieSceneScriptingFloatChannel, default=0, const [-0.0055,1]}
       ring_metacarpal_r_ctrl.Rotation.Z {type=MovieSceneScriptingFloatChannel, default=0, const [2.3559,1]}
       ring_metacarpal_r_ctrl.Scale.X {type=MovieSceneScriptingFloatChannel, default=1, const [1,1]}
       ring_metacarpal_r_ctrl.Scale.Y {type=MovieSceneScriptingFloatChannel, default=1, const [1,1]}
       ring_metacarpal_r_ctrl.Scale.Z {type=MovieSceneScriptingFloatChannel, default=1, const [1,1]}
       ring_01_r_ctrl.Location.X {type=MovieSceneScriptingFloatChannel, default=0, thinned_from=21, keys [[0,0],[352,0]]}
       ring_01_r_ctrl.Location.Y {type=MovieSceneScriptingFloatChannel, default=0, thinned_from=21, keys [[0,0],[352,0]]}
       ring_01_r_ctrl.Location.Z {type=MovieSceneScriptingFloatChannel, default=0, thinned_from=21, keys [[0,0],[352,0]]}
       ring_01_r_ctrl.Rotation.X {type=MovieSceneScriptingFloatChannel, default=2.6785, thinned_from=22, keys [[0,2.6785],[121,2.6785],[132,2.6627],[150,2.6785],[352,2.6785]]}
       ring_01_r_ctrl.Rotation.Y {type=MovieSceneScriptingFloatChannel, default=-7.5309, thinned_from=22, keys [[0,-4.3434],[352,-4.3434]]}
       ring_01_r_ctrl.Rotation.Z {type=MovieSceneScriptingFloatChannel, default=-2.4569, thinned_from=22, keys [[0,-2.3049],[352,-2.3049]]}
       ring_01_r_ctrl.Scale.X {type=MovieSceneScriptingFloatChannel, default=1, const [1,21]}
       ring_01_r_ctrl.Scale.Y {type=MovieSceneScriptingFloatChannel, default=1, const [1,21]}
       ring_01_r_ctrl.Scale.Z {type=MovieSceneScriptingFloatChannel, default=1, const [1,21]}
       ring_02_r_ctrl.Location.X {type=MovieSceneScriptingFloatChannel, default=0, thinned_from=21, keys [[0,0],[352,0]]}
       ring_02_r_ctrl.Location.Y {type=MovieSceneScriptingFloatChannel, default=0, thinned_from=21, keys [[0,0],[352,0]]}
       ring_02_r_ctrl.Location.Z {type=MovieSceneScriptingFloatChannel, default=0, const [0,21]}
       ring_02_r_ctrl.Rotation.X {type=MovieSceneScriptingFloatChannel, default=0, thinned_from=21, keys [[0,0],[352,0]]}
       ring_02_r_ctrl.Rotation.Y {type=MovieSceneScriptingFloatChannel, default=0, thinned_from=21, keys [[0,0],[352,0]]}
       ring_02_r_ctrl.Rotation.Z {type=MovieSceneScriptingFloatChannel, default=7.506, const [7.506,21]}
       ring_02_r_ctrl.Scale.X {type=MovieSceneScriptingFloatChannel, default=1, const [1,21]}
       ring_02_r_ctrl.Scale.Y {type=MovieSceneScriptingFloatChannel, default=1, const [1,21]}
       ring_02_r_ctrl.Scale.Z {type=MovieSceneScriptingFloatChannel, default=1, const [1,21]}
       ring_03_r_ctrl.Location.X {type=MovieSceneScriptingFloatChannel, default=0, thinned_from=21, keys [[0,0],[352,0]]}
       ring_03_r_ctrl.Location.Y {type=MovieSceneScriptingFloatChannel, default=0, const [0,21]}
       ring_03_r_ctrl.Location.Z {type=MovieSceneScriptingFloatChannel, default=0, const [0,21]}
       ring_03_r_ctrl.Rotation.X {type=MovieSceneScriptingFloatChannel, default=0, thinned_from=21, keys [[0,0],[352,0]]}
       ring_03_r_ctrl.Rotation.Y {type=MovieSceneScriptingFloatChannel, default=0, thinned_from=21, keys [[0,0],[352,0]]}
       ring_03_r_ctrl.Rotation.Z {type=MovieSceneScriptingFloatChannel, default=26.0112, const [26.0112,21]}
       ring_03_r_ctrl.Scale.X {type=MovieSceneScriptingFloatChannel, default=1, const [1,21]}
       ring_03_r_ctrl.Scale.Y {type=MovieSceneScriptingFloatChannel, default=1, const [1,21]}
       ring_03_r_ctrl.Scale.Z {type=MovieSceneScriptingFloatChannel, default=1, const [1,21]}
       pinky_metacarpal_r_ctrl.Location.X {type=MovieSceneScriptingFloatChannel, default=0, const [0,1]}
       pinky_metacarpal_r_ctrl.Location.Y {type=MovieSceneScriptingFloatChannel, default=0, const [0,1]}
       pinky_metacarpal_r_ctrl.Location.Z {type=MovieSceneScriptingFloatChannel, default=0, const [0,1]}
       pinky_metacarpal_r_ctrl.Rotation.X {type=MovieSceneScriptingFloatChannel, default=0, const [0,1]}
       pinky_metacarpal_r_ctrl.Rotation.Y {type=MovieSceneScriptingFloatChannel, default=0, const [0,1]}
       pinky_metacarpal_r_ctrl.Rotation.Z {type=MovieSceneScriptingFloatChannel, default=0, const [2.3332,1]}
       pinky_metacarpal_r_ctrl.Scale.X {type=MovieSceneScriptingFloatChannel, default=1, const [1,1]}
       pinky_metacarpal_r_ctrl.Scale.Y {type=MovieSceneScriptingFloatChannel, default=1, const [1,1]}
       pinky_metacarpal_r_ctrl.Scale.Z {type=MovieSceneScriptingFloatChannel, default=1, const [1,1]}
       pinky_01_r_ctrl.Location.X {type=MovieSceneScriptingFloatChannel, default=0, thinned_from=21, keys [[0,0],[352,0]]}
       pinky_01_r_ctrl.Location.Y {type=MovieSceneScriptingFloatChannel, default=0, thinned_from=21, keys [[0,0],[352,0]]}
       pinky_01_r_ctrl.Location.Z {type=MovieSceneScriptingFloatChannel, default=0, thinned_from=21, keys [[0,0],[352,0]]}
       pinky_01_r_ctrl.Rotation.X {type=MovieSceneScriptingFloatChannel, default=5.4063, thinned_from=21, keys [[0,5.4063],[352,5.4063]]}
       pinky_01_r_ctrl.Rotation.Y {type=MovieSceneScriptingFloatChannel, default=-9.5379, const [-9.5379,21]}
       pinky_01_r_ctrl.Rotation.Z {type=MovieSceneScriptingFloatChannel, default=-11.5927, const [-11.5927,21]}
       pinky_01_r_ctrl.Scale.X {type=MovieSceneScriptingFloatChannel, default=1, const [1,21]}
       pinky_01_r_ctrl.Scale.Y {type=MovieSceneScriptingFloatChannel, default=1, const [1,21]}
       pinky_01_r_ctrl.Scale.Z {type=MovieSceneScriptingFloatChannel, default=1, const [1,21]}
       pinky_02_r_ctrl.Location.X {type=MovieSceneScriptingFloatChannel, default=0, thinned_from=21, keys [[0,0],[352,0]]}
       pinky_02_r_ctrl.Location.Y {type=MovieSceneScriptingFloatChannel, default=0, thinned_from=21, keys [[0,0],[352,0]]}
       pinky_02_r_ctrl.Location.Z {type=MovieSceneScriptingFloatChannel, default=0, thinned_from=21, keys [[0,0],[352,0]]}
       pinky_02_r_ctrl.Rotation.X {type=MovieSceneScriptingFloatChannel, default=0, thinned_from=21, keys [[0,0],[352,0]]}
       pinky_02_r_ctrl.Rotation.Y {type=MovieSceneScriptingFloatChannel, default=0, thinned_from=21, keys [[0,0],[352,0]]}
       pinky_02_r_ctrl.Rotation.Z {type=MovieSceneScriptingFloatChannel, default=14.3888, const [14.3888,21]}
       pinky_02_r_ctrl.Scale.X {type=MovieSceneScriptingFloatChannel, default=1, const [1,21]}
       pinky_02_r_ctrl.Scale.Y {type=MovieSceneScriptingFloatChannel, default=1, const [1,21]}
       pinky_02_r_ctrl.Scale.Z {type=MovieSceneScriptingFloatChannel, default=1, const [1,21]}
       pinky_03_r_ctrl.Location.X {type=MovieSceneScriptingFloatChannel, default=0, thinned_from=21, keys [[0,0],[352,0]]}
       pinky_03_r_ctrl.Location.Y {type=MovieSceneScriptingFloatChannel, default=0, thinned_from=21, keys [[0,0],[352,0]]}
       pinky_03_r_ctrl.Location.Z {type=MovieSceneScriptingFloatChannel, default=0, const [0,21]}
       pinky_03_r_ctrl.Rotation.X {type=MovieSceneScriptingFloatChannel, default=0, thinned_from=21, keys [[0,0],[352,0]]}
       pinky_03_r_ctrl.Rotation.Y {type=MovieSceneScriptingFloatChannel, default=0, thinned_from=21, keys [[0,0],[352,0]]}
       pinky_03_r_ctrl.Rotation.Z {type=MovieSceneScriptingFloatChannel, default=20.3909, const [20.3909,21]}
       pinky_03_r_ctrl.Scale.X {type=MovieSceneScriptingFloatChannel, default=1, const [1,21]}
       pinky_03_r_ctrl.Scale.Y {type=MovieSceneScriptingFloatChannel, default=1, const [1,21]}
       pinky_03_r_ctrl.Scale.Z {type=MovieSceneScriptingFloatChannel, default=1, const [1,21]}
       hand_r_ik_ctrl.Location.X {type=MovieSceneScriptingFloatChannel, default=0, const [52.6482,17]}
       hand_r_ik_ctrl.Location.Y {type=MovieSceneScriptingFloatChannel, default=0, const [81.8265,17]}
       hand_r_ik_ctrl.Location.Z {type=MovieSceneScriptingFloatChannel, default=0, thinned_from=20, keys [[0,-35.777],[121,-35.7175],[352,-35.777]]}
       hand_r_ik_ctrl.Rotation.X {type=MovieSceneScriptingFloatChannel, default=0, const [71.7779,20]}
       hand_r_ik_ctrl.Rotation.Y {type=MovieSceneScriptingFloatChannel, default=0, const [-43.7142,20]}
       hand_r_ik_ctrl.Rotation.Z {type=MovieSceneScriptingFloatChannel, default=0, const [-202.9392,25]}
       hand_r_ik_ctrl.Scale.X {type=MovieSceneScriptingFloatChannel, default=1, const [-1,20]}
       hand_r_ik_ctrl.Scale.Y {type=MovieSceneScriptingFloatChannel, default=1, const [1,20]}
       hand_r_ik_ctrl.Scale.Z {type=MovieSceneScriptingFloatChannel, default=1, const [1,20]}
       Softness_2 {type=MovieSceneScriptingFloatChannel, default=0, const [0,20]}
       hand_r_ikorient_ctrl.Rotation.Roll {type=MovieSceneScriptingFloatChannel, default=0, const [0,21]}
       hand_r_ikorient_ctrl.Rotation.Pitch {type=MovieSceneScriptingFloatChannel, default=0, const [0,21]}
       hand_r_ikorient_ctrl.Rotation.Yaw {type=MovieSceneScriptingFloatChannel, default=0, const [0,21]}
       arm_r_pv_ik_ctrl.Location.X {
        type=MovieSceneScriptingFloatChannel
        default=0
        keys [[0,61.25],[29,59.4706],[59,72.8269],[78,76.5075],[114,57.7702],[147,61.25],[238,61.25],[281,49.2188],[303,46.4864],[338,64.7474],[352,61.25]]
       }
       arm_r_pv_ik_ctrl.Location.Y {
        type=MovieSceneScriptingFloatChannel
        default=0
        keys [[0,-42.44],[29,-43.4581],[59,-23.3815],[78,-17.3224],[114,-48.1686],[147,-42.44],[238,-42.44],[281,-50.405],[303,-51.9642],[338,-40.4442],[352,-42.44]]
       }
       arm_r_pv_ik_ctrl.Location.Z {type=MovieSceneScriptingFloatChannel, default=0, const [10.77,11]}
       arm_r_pv_ik_ctrl.Rotation.X {type=MovieSceneScriptingFloatChannel, default=0, const [0,11]}
       arm_r_pv_ik_ctrl.Rotation.Y {type=MovieSceneScriptingFloatChannel, default=0, const [0,11]}
       arm_r_pv_ik_ctrl.Rotation.Z {type=MovieSceneScriptingFloatChannel, default=0, const [0,11]}
       arm_r_pv_ik_ctrl.Scale.X {type=MovieSceneScriptingFloatChannel, default=1, const [1,11]}
       arm_r_pv_ik_ctrl.Scale.Y {type=MovieSceneScriptingFloatChannel, default=1, const [1,11]}
       arm_r_pv_ik_ctrl.Scale.Z {type=MovieSceneScriptingFloatChannel, default=1, const [1,11]}
       body_ctrl.Location.X {
        type=MovieSceneScriptingFloatChannel
        default=0
        keys [[0,0],[33,0],[49,-0.8056],[80,-0.9684],[98,-0.1179],[131,0],[164,-8.4617],[188,3.0304],[216,-6.4585],[258,-0.041],[303,0.2637],[328,0.0248],[352,0]]
       }
       body_ctrl.Location.Y {
        type=MovieSceneScriptingFloatChannel
        default=0.8143
        thinned_from=18
        keys [[0,0.8143],[33,0.8143],[50,0.678],[80,0.6507],[98,0.7718],[143,0.8143],[352,0.8143]]
       }
       body_ctrl.Location.Z {
        type=MovieSceneScriptingFloatChannel
        default=-0.2094
        thinned_from=21
        keys [[0,0.5855],[143,0.5855],[153,0.5976],[164,0.2728],[175,0.5955],[188,0.2774],[200,0.5932],[214,0.2847],[236,0.608],[258,0.5855],[352,0.5855]]
       }
       body_ctrl.Rotation.X {
        type=MovieSceneScriptingFloatChannel
        default=-3.7217
        thinned_from=19
        keys [[0,-3.7217],[34,-3.7205],[40,-3.7115],[90,-3.7133],[143,-3.7217],[164,-3.8971],[188,-3.1793],[216,-3.7194],[234,-3.3853],[236,-3.3939],[258,-3.2141],[288,-3.1944],[299,-3.2684],[320,-3.7217],[352,-3.7217]]
       }
       body_ctrl.Rotation.Y {
        type=MovieSceneScriptingFloatChannel
        default=0
        keys [[0,0],[25,0],[34,-0.0326],[48,-0.2137],[80,-0.2755],[103,-0.0491],[121,0],[132,0],[138,0.0707],[166,3.6363],[192,-2.997],[216,2.7574],[239,-0.1412],[250,-0.2761],[265,-1.0717],[303,-1.3967],[328,-0.3499],[352,0]]
       }
       body_ctrl.Rotation.Z {
        type=MovieSceneScriptingFloatChannel
        default=0
        keys [[0,0],[33,0.9951],[65,-7.6938],[80,-6.9722],[101,-1.0998],[138,0],[164,-4.0176],[188,0.8418],[217,-0.9608],[240,-0.8814],[260,8.1464],[291,12.2074],[328,-1.4066],[352,0]]
       }
       body_ctrl.Scale.X {type=MovieSceneScriptingFloatChannel, default=1, const [1,21]}
       body_ctrl.Scale.Y {type=MovieSceneScriptingFloatChannel, default=1, const [1,21]}
       body_ctrl.Scale.Z {type=MovieSceneScriptingFloatChannel, default=1, const [1,21]}
       hips_ctrl.Location.X {type=MovieSceneScriptingFloatChannel, default=0, const [0,21]}
       hips_ctrl.Location.Y {type=MovieSceneScriptingFloatChannel, default=0, const [0,21]}
       hips_ctrl.Location.Z {type=MovieSceneScriptingFloatChannel, default=0, const [0,21]}
       hips_ctrl.Rotation.X {type=MovieSceneScriptingFloatChannel, default=0, thinned_from=21, keys [[0,0],[150,0],[164,0.0005],[174,0],[352,0]]}
       hips_ctrl.Rotation.Y {
        type=MovieSceneScriptingFloatChannel
        default=0
        thinned_from=17
        keys [[0,0],[131,0],[164,-3.3706],[188,3.3608],[216,-3.1265],[258,0],[352,0]]
       }
       hips_ctrl.Rotation.Z {type=MovieSceneScriptingFloatChannel, default=0, const [0,21]}
       hips_ctrl.Scale.X {type=MovieSceneScriptingFloatChannel, default=1, const [1,21]}
       hips_ctrl.Scale.Y {type=MovieSceneScriptingFloatChannel, default=1, const [1,21]}
       hips_ctrl.Scale.Z {type=MovieSceneScriptingFloatChannel, default=1, const [1,21]}
       hips_tan_ctrl.Location.X {type=MovieSceneScriptingFloatChannel, default=0, thinned_from=21, keys [[0,0],[352,0]]}
       hips_tan_ctrl.Location.Y {type=MovieSceneScriptingFloatChannel, default=0.008, thinned_from=21, keys [[0,0.008],[352,0.008]]}
       hips_tan_ctrl.Location.Z {type=MovieSceneScriptingFloatChannel, default=-0.0073, thinned_from=21, keys [[0,-0.0073],[352,-0.0073]]}
       hips_tan_ctrl.Rotation.X {type=MovieSceneScriptingFloatChannel, default=0, const [0,21]}
       hips_tan_ctrl.Rotation.Y {type=MovieSceneScriptingFloatChannel, default=0, const [0,21]}
       hips_tan_ctrl.Rotation.Z {type=MovieSceneScriptingFloatChannel, default=0, const [0,21]}
       hips_tan_ctrl.Scale.X {type=MovieSceneScriptingFloatChannel, default=1, const [1,21]}
       hips_tan_ctrl.Scale.Y {type=MovieSceneScriptingFloatChannel, default=1, const [1,21]}
       hips_tan_ctrl.Scale.Z {type=MovieSceneScriptingFloatChannel, default=1, const [1,21]}
       spine_01_ctrl.Location.X {type=MovieSceneScriptingFloatChannel, default=0, const [0,21]}
       spine_01_ctrl.Location.Y {type=MovieSceneScriptingFloatChannel, default=0, const [0,21]}
       spine_01_ctrl.Location.Z {type=MovieSceneScriptingFloatChannel, default=0, const [0,21]}
       spine_01_ctrl.Rotation.X {type=MovieSceneScriptingFloatChannel, default=0, const [0,21]}
       spine_01_ctrl.Rotation.Y {type=MovieSceneScriptingFloatChannel, default=0, const [0,21]}
       spine_01_ctrl.Rotation.Z {type=MovieSceneScriptingFloatChannel, default=0, const [0,21]}
       spine_01_ctrl.Scale.X {type=MovieSceneScriptingFloatChannel, default=1, const [1,21]}
       spine_01_ctrl.Scale.Y {type=MovieSceneScriptingFloatChannel, default=1, const [1,21]}
       spine_01_ctrl.Scale.Z {type=MovieSceneScriptingFloatChannel, default=1, const [1,21]}
       spine_02_ctrl.Location.X {type=MovieSceneScriptingFloatChannel, default=0, const [0,21]}
       spine_02_ctrl.Location.Y {type=MovieSceneScriptingFloatChannel, default=0, const [0,21]}
       spine_02_ctrl.Location.Z {type=MovieSceneScriptingFloatChannel, default=0, const [0,21]}
       spine_02_ctrl.Rotation.X {type=MovieSceneScriptingFloatChannel, default=0, const [0,21]}
       spine_02_ctrl.Rotation.Y {type=MovieSceneScriptingFloatChannel, default=0, const [0,21]}
       spine_02_ctrl.Rotation.Z {type=MovieSceneScriptingFloatChannel, default=0, const [0,21]}
       spine_02_ctrl.Scale.X {type=MovieSceneScriptingFloatChannel, default=1, const [1,21]}
       spine_02_ctrl.Scale.Y {type=MovieSceneScriptingFloatChannel, default=1, const [1,21]}
       spine_02_ctrl.Scale.Z {type=MovieSceneScriptingFloatChannel, default=1, const [1,21]}
       spine_03_ctrl.Location.X {type=MovieSceneScriptingFloatChannel, default=0, const [0,21]}
       spine_03_ctrl.Location.Y {type=MovieSceneScriptingFloatChannel, default=0, const [0,21]}
       spine_03_ctrl.Location.Z {type=MovieSceneScriptingFloatChannel, default=0, const [0,21]}
       spine_03_ctrl.Rotation.X {type=MovieSceneScriptingFloatChannel, default=0, const [0,21]}
       spine_03_ctrl.Rotation.Y {
        type=MovieSceneScriptingFloatChannel
        default=0
        thinned_from=19
        keys [[0,0],[132,0],[154,0.7325],[164,2.8106],[202,-1.9598],[220,0.1521],[239,-0.2805],[251,-0.3073],[268,-1.101],[288,-1.2169],[320,0],[352,0]]
       }
       spine_03_ctrl.Rotation.Z {type=MovieSceneScriptingFloatChannel, default=0, const [0,21]}
       spine_03_ctrl.Scale.X {type=MovieSceneScriptingFloatChannel, default=1, const [1,21]}
       spine_03_ctrl.Scale.Y {type=MovieSceneScriptingFloatChannel, default=1, const [1,21]}
       spine_03_ctrl.Scale.Z {type=MovieSceneScriptingFloatChannel, default=1, const [1,21]}
       spine_04_ctrl.Location.X {type=MovieSceneScriptingFloatChannel, default=0, const [0,21]}
       spine_04_ctrl.Location.Y {type=MovieSceneScriptingFloatChannel, default=0, const [0,21]}
       spine_04_ctrl.Location.Z {type=MovieSceneScriptingFloatChannel, default=0, const [0,21]}
       spine_04_ctrl.Rotation.X {
        type=MovieSceneScriptingFloatChannel
        default=6.0468
        thinned_from=21
        keys [[0,6.0468],[25,6.0468],[34,6.0212],[40,5.941],[74,5.941],[90,4.7753],[121,4.6054],[143,6.0468],[164,6.0565],[174,6.0519],[216,6.0509],[236,6.5453],[258,6.8562],[288,6.8562],[299,6.6503],[320,6.0468],[352,6.0468]]
       }
       spine_04_ctrl.Rotation.Y {
        type=MovieSceneScriptingFloatChannel
        default=0
        keys [[0,0],[29,0],[44,0.7825],[74,1.128],[94,0.1435],[130,0],[167,0.2345],[203,-0.5426],[222,-0.206],[241,-1.1603],[253,-1.288],[269,-2.0177],[307,-2.2152],[337,-0.16],[352,0]]
       }
       spine_04_ctrl.Rotation.Z {
        type=MovieSceneScriptingFloatChannel
        default=0
        keys [[0,0],[29,0],[44,-10.7119],[75,-13.5551],[94,-1.3625],[143,0],[167,-0.3433],[194,0.1429],[216,-0.2303],[241,-0.3089],[257,5.7814],[299,8.0087],[324,0.8207],[352,0]]
       }
       spine_04_ctrl.Scale.X {type=MovieSceneScriptingFloatChannel, default=1, const [1,21]}
       spine_04_ctrl.Scale.Y {type=MovieSceneScriptingFloatChannel, default=1, const [1,21]}
       spine_04_ctrl.Scale.Z {type=MovieSceneScriptingFloatChannel, default=1, const [1,21]}
       head_ctrl.Location.X {type=MovieSceneScriptingFloatChannel, default=0, thinned_from=21, keys [[0,0],[352,0]]}
       head_ctrl.Location.Y {type=MovieSceneScriptingFloatChannel, default=-1.6873, const [0,21]}
       head_ctrl.Location.Z {type=MovieSceneScriptingFloatChannel, default=152.0342, const [0,21]}
       head_ctrl.Rotation.X {
        type=MovieSceneScriptingFloatChannel
        default=0
        keys [[0,9.4536],[25,9.4536],[40,2.5957],[71,2.2178],[90,8.6924],[123,9.45],[145,-3.1698],[176,-8.098],[204,-8.8388],[218,1.4826],[232,-2.7171],[256,5.084],[320,9.4536],[352,9.9925]]
       }
       head_ctrl.Rotation.Y {
        type=MovieSceneScriptingFloatChannel
        default=0
        keys [[0,0],[25,0],[40,9.119],[71,9.8991],[90,-6.1135],[101,-6.5608],[113,-3.9785],[121,-3.462],[145,10.6888],[177,-9.4768],[206,7.2987],[236,9.7993],[258,-6.8187],[295,-8.0228],[309,-2.6905],[352,0]]
       }
       head_ctrl.Rotation.Z {
        type=MovieSceneScriptingFloatChannel
        default=0
        keys [[0,-1.6],[25,-1.6],[40,-29.5595],[71,-34.8587],[90,-5.1731],[126,2.9062],[148,10.195],[177,-4.382],[206,1.9108],[236,4.4693],[255,34.6951],[295,39.7574],[309,2.0313],[320,-1.5962],[352,-1.6]]
       }
       head_ctrl.Scale.X {type=MovieSceneScriptingFloatChannel, default=1, const [1,21]}
       head_ctrl.Scale.Y {type=MovieSceneScriptingFloatChannel, default=1, const [1,21]}
       head_ctrl.Scale.Z {type=MovieSceneScriptingFloatChannel, default=1, const [1,21]}
       clavicle_l_ctrl.Location.X {type=MovieSceneScriptingFloatChannel, default=0, const [0,21]}
       clavicle_l_ctrl.Location.Y {type=MovieSceneScriptingFloatChannel, default=0, const [0,21]}
       clavicle_l_ctrl.Location.Z {type=MovieSceneScriptingFloatChannel, default=0, const [0,21]}
       clavicle_l_ctrl.Rotation.X {type=MovieSceneScriptingFloatChannel, default=0, thinned_from=21, keys [[0,0],[352,0]]}
       clavicle_l_ctrl.Rotation.Y {
        type=MovieSceneScriptingFloatChannel
        default=5
        thinned_from=18
        keys [[0,5],[143,5],[164,2.9428],[192,6.9241],[216,2.9517],[236,5],[352,5]]
       }
       clavicle_l_ctrl.Rotation.Z {type=MovieSceneScriptingFloatChannel, default=0, thinned_from=21, keys [[0,0],[352,0]]}
       clavicle_l_ctrl.Scale.X {type=MovieSceneScriptingFloatChannel, default=1, const [1,21]}
       clavicle_l_ctrl.Scale.Y {type=MovieSceneScriptingFloatChannel, default=1, const [1,21]}
       clavicle_l_ctrl.Scale.Z {type=MovieSceneScriptingFloatChannel, default=1, const [1,21]}
       clavicle_r_ctrl.Location.X {type=MovieSceneScriptingFloatChannel, default=0, thinned_from=21, keys [[0,0],[352,0]]}
       clavicle_r_ctrl.Location.Y {type=MovieSceneScriptingFloatChannel, default=0, thinned_from=21, keys [[0,0],[352,0]]}
       clavicle_r_ctrl.Location.Z {type=MovieSceneScriptingFloatChannel, default=0, const [0,21]}
       clavicle_r_ctrl.Rotation.X {type=MovieSceneScriptingFloatChannel, default=0, thinned_from=21, keys [[0,0],[352,0]]}
       clavicle_r_ctrl.Rotation.Y {
        type=MovieSceneScriptingFloatChannel
        default=5
        thinned_from=18
        keys [[0,5],[143,5],[164,6.9276],[192,2.9265],[216,7.0351],[236,5],[352,5]]
       }
       clavicle_r_ctrl.Rotation.Z {type=MovieSceneScriptingFloatChannel, default=0, thinned_from=21, keys [[0,0],[352,0]]}
       clavicle_r_ctrl.Scale.X {type=MovieSceneScriptingFloatChannel, default=1, const [1,21]}
       clavicle_r_ctrl.Scale.Y {type=MovieSceneScriptingFloatChannel, default=1, const [1,21]}
       clavicle_r_ctrl.Scale.Z {type=MovieSceneScriptingFloatChannel, default=1, const [1,21]}
       neck_01_ctrl.Location.X {type=MovieSceneScriptingFloatChannel, default=0, thinned_from=24, keys [[0,0],[352,0]]}
       neck_01_ctrl.Location.Y {type=MovieSceneScriptingFloatChannel, default=-3.5695, const [0,24]}
       neck_01_ctrl.Location.Z {type=MovieSceneScriptingFloatChannel, default=131.159, const [0,24]}
       neck_01_ctrl.Rotation.X {
        type=MovieSceneScriptingFloatChannel
        default=0
        thinned_from=26
        keys [[0,0],[123,0],[157,-0.998],[162,-0.7652],[173,0],[198,0],[206,-0.8184],[216,0],[352,0]]
       }
       neck_01_ctrl.Rotation.Y {
        type=MovieSceneScriptingFloatChannel
        default=0
        keys [[0,0],[25,0],[114,0.0129],[120,0.0129],[144,6.3691],[177,-7.0181],[206,7.0108],[236,8.9726],[258,4.8064],[288,3.6023],[299,1.7538],[320,0],[352,0]]
       }
       neck_01_ctrl.Rotation.Z {type=MovieSceneScriptingFloatChannel, default=0, thinned_from=26, keys [[0,0],[157,0],[162,0.0938],[173,0],[352,0]]}
       neck_01_ctrl.Scale.X {type=MovieSceneScriptingFloatChannel, default=1, const [1,24]}
       neck_01_ctrl.Scale.Y {type=MovieSceneScriptingFloatChannel, default=1, const [1,24]}
       neck_01_ctrl.Scale.Z {type=MovieSceneScriptingFloatChannel, default=1, const [1,24]}
       neck_02_ctrl.Location.X {type=MovieSceneScriptingFloatChannel, default=0, thinned_from=24, keys [[0,0],[352,0]]}
       neck_02_ctrl.Location.Y {type=MovieSceneScriptingFloatChannel, default=-4.3237, const [0,24]}
       neck_02_ctrl.Location.Z {type=MovieSceneScriptingFloatChannel, default=142.4487, const [0,24]}
       neck_02_ctrl.Rotation.X {type=MovieSceneScriptingFloatChannel, default=0, const [-0.6691,2]}
       neck_02_ctrl.Rotation.Y {type=MovieSceneScriptingFloatChannel, default=0, const [0,18]}
       neck_02_ctrl.Rotation.Z {type=MovieSceneScriptingFloatChannel, default=0, thinned_from=18, keys [[0,0],[352,0]]}
       neck_02_ctrl.Scale.X {type=MovieSceneScriptingFloatChannel, default=1, const [1,24]}
       neck_02_ctrl.Scale.Y {type=MovieSceneScriptingFloatChannel, default=1, const [1,24]}
       neck_02_ctrl.Scale.Z {type=MovieSceneScriptingFloatChannel, default=1, const [1,24]}
       head_ik_ctrl.Location.X {
        type=MovieSceneScriptingFloatChannel
        default=0
        thinned_from=22
        keys [[0,0],[40,0],[71,-0.0969],[90,-0.0969],[121,0.0008],[150,0],[162,0.4369],[164,0],[205,0],[216,-0.4418],[236,0],[258,-0.5496],[299,-0.5496],[320,0],[352,0]]
       }
       head_ik_ctrl.Location.Y {
        type=MovieSceneScriptingFloatChannel
        default=0
        thinned_from=22
        keys [[0,0],[25,-0.5061],[40,-0.5061],[71,-0.497],[90,-0.497],[121,0],[150,-0.5061],[162,-0.1867],[164,-0.5061],[205,-0.5061],[216,-0.0589],[236,-0.5061],[258,-0.5645],[299,-0.5645],[320,-0.5061],[352,-0.5061]]
       }
       head_ik_ctrl.Location.Z {
        type=MovieSceneScriptingFloatChannel
        default=0
        thinned_from=22
        keys [[0,0],[25,0.1449],[40,0.1449],[74,0.1439],[90,0.1439],[121,0],[150,0.1449],[162,0.102],[164,0.1449],[205,0.1449],[216,0.0126],[236,0.1449],[258,0.1515],[299,0.1515],[320,0.1449],[352,0.1449]]
       }
       head_ik_ctrl.Rotation.X {
        type=MovieSceneScriptingFloatChannel
        default=0
        thinned_from=22
        keys [[0,0],[25,9.4536],[34,9.4536],[40,0],[71,9.4536],[74,9.4536],[90,9.5483],[121,9.45],[150,10.9969],[162,-6.9994],[164,-8.1367],[205,-8.1367],[216,-4.0333],[236,-8.1367],[258,5.7507],[288,5.6942],[299,5.6942],[320,0],[352,9.4536]]
       }
       head_ik_ctrl.Rotation.Y {
        type=MovieSceneScriptingFloatChannel
        default=0
        thinned_from=22
        keys [[0,0],[74,0],[90,0.0249],[121,0.406],[150,0.406],[162,-1.5581],[164,0],[205,0],[216,3.8317],[236,0],[258,-8.0228],[288,-0.0862],[299,-0.0862],[320,0],[352,0]]
       }
       head_ik_ctrl.Rotation.Z {
        type=MovieSceneScriptingFloatChannel
        default=0
        thinned_from=22
        keys [[0,0],[25,-1.6],[34,-4.4397],[40,-9.8075],[71,-42.0664],[74,-42.0664],[90,-39.3099],[121,2.8245],[150,2.8245],[162,-1.7563],[164,1.3993],[205,1.3993],[216,0.6876],[236,1.3993],[258,39.7574],[288,40.554],[299,40.554],[320,-9.8075],[352,-1.5962]]
       }
       head_ik_ctrl.Scale.X {type=MovieSceneScriptingFloatChannel, default=1, const [1,21]}
       head_ik_ctrl.Scale.Y {type=MovieSceneScriptingFloatChannel, default=1, const [1,21]}
       head_ik_ctrl.Scale.Z {type=MovieSceneScriptingFloatChannel, default=1, const [1,21]}
       chest_ctrl.Location.X {type=MovieSceneScriptingFloatChannel, default=0, thinned_from=21, keys [[0,0],[352,0]]}
       chest_ctrl.Location.Y {type=MovieSceneScriptingFloatChannel, default=0, thinned_from=21, keys [[0,0],[352,0]]}
       chest_ctrl.Location.Z {type=MovieSceneScriptingFloatChannel, default=0, thinned_from=21, keys [[0,0],[352,0]]}
       chest_ctrl.Rotation.X {type=MovieSceneScriptingFloatChannel, default=0, thinned_from=21, keys [[0,0],[352,0]]}
       chest_ctrl.Rotation.Y {type=MovieSceneScriptingFloatChannel, default=0, thinned_from=21, keys [[0,0],[205,0],[216,-0.076],[236,0],[352,0]]}
       chest_ctrl.Rotation.Z {type=MovieSceneScriptingFloatChannel, default=0, thinned_from=21, keys [[0,0],[352,0]]}
       chest_ctrl.Scale.X {type=MovieSceneScriptingFloatChannel, default=1, const [1,21]}
       chest_ctrl.Scale.Y {type=MovieSceneScriptingFloatChannel, default=1, const [1,21]}
       chest_ctrl.Scale.Z {type=MovieSceneScriptingFloatChannel, default=1, const [1,21]}
       chest_tan_ctrl.Location.X {type=MovieSceneScriptingFloatChannel, default=0, thinned_from=21, keys [[0,0],[205,0],[216,0.0131],[236,0],[352,0]]}
       chest_tan_ctrl.Location.Y {type=MovieSceneScriptingFloatChannel, default=-0.0303, thinned_from=21, keys [[0,-0.0303],[352,-0.0303]]}
       chest_tan_ctrl.Location.Z {type=MovieSceneScriptingFloatChannel, default=0.265, thinned_from=21, keys [[0,0.265],[352,0.265]]}
       chest_tan_ctrl.Rotation.X {type=MovieSceneScriptingFloatChannel, default=0, const [0,21]}
       chest_tan_ctrl.Rotation.Y {type=MovieSceneScriptingFloatChannel, default=0, thinned_from=21, keys [[0,0],[205,0],[216,0.076],[236,0],[352,0]]}
       chest_tan_ctrl.Rotation.Z {type=MovieSceneScriptingFloatChannel, default=0, thinned_from=21, keys [[0,0],[352,0]]}
       chest_tan_ctrl.Scale.X {type=MovieSceneScriptingFloatChannel, default=1, const [1,21]}
       chest_tan_ctrl.Scale.Y {type=MovieSceneScriptingFloatChannel, default=1, const [1,21]}
       chest_tan_ctrl.Scale.Z {type=MovieSceneScriptingFloatChannel, default=1, const [1,21]}
       Weight {type=MovieSceneScriptingFloatChannel, default=1}
       hand_l_ik_ctrl.Space {
        type=MovieSceneScriptingSpaceChannel
        keys [
         [-17, "<Struct 'MovieSceneControlRigSpaceBaseKey' (0x49d479d3a80) {space_type: Parent, control_rig_element: {type: None, name: \"\"}}>"]
         [
          0
          "<Struct 'MovieSceneControlRigSpaceBaseKey' (0x49d479d3d60) {space_type: ControlRig, control_rig_element: {type: Control, name: \"hips_ctrl\"}}>"
         ]
        ]
       }
       hand_r_ik_ctrl.Space {
        type=MovieSceneScriptingSpaceChannel
        const [
         "<Struct 'MovieSceneControlRigSpaceBaseKey' (0x49d479d3760) {space_type: ControlRig, control_rig_element: {type: Control, name: \"hips_ctrl\"}}>"
         1
        ]
       }
       head_ctrl.Space {
        type=MovieSceneScriptingSpaceChannel
        const [
         "<Struct 'MovieSceneControlRigSpaceBaseKey' (0x49d479d3440) {space_type: Parent, control_rig_element: {type: Null, name: \"head_fk_space\"}}>"
         1
        ]
       }
       neck_01_ctrl.Space {
        type=MovieSceneScriptingSpaceChannel
        const [
         "<Struct 'MovieSceneControlRigSpaceBaseKey' (0x49d479d32c0) {space_type: Parent, control_rig_element: {type: Control, name: \"spine_04_ctrl\"}}>"
         1
        ]
       }
       neck_02_ctrl.Space {
        type=MovieSceneScriptingSpaceChannel
        const [
         "<Struct 'MovieSceneControlRigSpaceBaseKey' (0x49d479d3c00) {space_type: Parent, control_rig_element: {type: Control, name: \"neck_01_ctrl\"}}>"
         1
        ]
       }
      ]
     }
     {
      type=MovieSceneControlRigParameterSection
      channels [
       thumb_01_l_ctrl.Location.X {type=MovieSceneScriptingFloatChannel, default=0, const [0,15]}
       thumb_01_l_ctrl.Location.Y {type=MovieSceneScriptingFloatChannel, default=0, const [0,15]}
       thumb_01_l_ctrl.Location.Z {type=MovieSceneScriptingFloatChannel, default=0, const [0,15]}
       thumb_01_l_ctrl.Rotation.X {
        type=MovieSceneScriptingFloatChannel
        default=0
        keys [[0,0],[32,0],[55,-0.334],[78,-0.3644],[102,0],[132,0],[140,0.0257],[166,0.6513],[190,-0.3209],[221,-0.0007],[242,-0.0005],[272,-0.2488],[302,-0.3186],[330,-0.0005],[352,0]]
       }
       thumb_01_l_ctrl.Rotation.Y {
        type=MovieSceneScriptingFloatChannel
        default=0
        keys [[0,0],[32,0],[55,5.4046],[78,5.896],[102,0],[132,0],[140,-0.0461],[166,-1.169],[190,0.5761],[221,0.0012],[242,0.0008],[272,0.4466],[302,0.5718],[330,0.0008],[352,0]]
       }
       thumb_01_l_ctrl.Rotation.Z {
        type=MovieSceneScriptingFloatChannel
        default=0
        keys [[0,0],[32,0],[55,7.925],[78,8.6455],[102,0],[132,0],[140,-0.146],[166,-3.6984],[190,1.8225],[221,0.0038],[242,0.0026],[272,1.4128],[302,1.809],[330,0.0026],[352,0]]
       }
       thumb_01_l_ctrl.Scale.X {type=MovieSceneScriptingFloatChannel, default=0, const [0,15]}
       thumb_01_l_ctrl.Scale.Y {type=MovieSceneScriptingFloatChannel, default=0, const [0,15]}
       thumb_01_l_ctrl.Scale.Z {type=MovieSceneScriptingFloatChannel, default=0, const [0,15]}
       thumb_02_l_ctrl.Location.X {type=MovieSceneScriptingFloatChannel, default=0, const [0,14]}
       thumb_02_l_ctrl.Location.Y {type=MovieSceneScriptingFloatChannel, default=0, const [0,14]}
       thumb_02_l_ctrl.Location.Z {type=MovieSceneScriptingFloatChannel, default=0, const [0,14]}
       thumb_02_l_ctrl.Rotation.X {type=MovieSceneScriptingFloatChannel, default=0, const [0,14]}
       thumb_02_l_ctrl.Rotation.Y {type=MovieSceneScriptingFloatChannel, default=0, const [0,14]}
       thumb_02_l_ctrl.Rotation.Z {type=MovieSceneScriptingFloatChannel, default=0, const [0,14]}
       thumb_02_l_ctrl.Scale.X {type=MovieSceneScriptingFloatChannel, default=0, const [0,14]}
       thumb_02_l_ctrl.Scale.Y {type=MovieSceneScriptingFloatChannel, default=0, const [0,14]}
       thumb_02_l_ctrl.Scale.Z {type=MovieSceneScriptingFloatChannel, default=0, const [0,14]}
       thumb_03_l_ctrl.Location.X {type=MovieSceneScriptingFloatChannel, default=0, const [0,14]}
       thumb_03_l_ctrl.Location.Y {type=MovieSceneScriptingFloatChannel, default=0, const [0,14]}
       thumb_03_l_ctrl.Location.Z {type=MovieSceneScriptingFloatChannel, default=0, const [0,14]}
       thumb_03_l_ctrl.Rotation.X {type=MovieSceneScriptingFloatChannel, default=0, const [0,14]}
       thumb_03_l_ctrl.Rotation.Y {type=MovieSceneScriptingFloatChannel, default=0, const [0,14]}
       thumb_03_l_ctrl.Rotation.Z {type=MovieSceneScriptingFloatChannel, default=0, const [0,14]}
       thumb_03_l_ctrl.Scale.X {type=MovieSceneScriptingFloatChannel, default=0, const [0,14]}
       thumb_03_l_ctrl.Scale.Y {type=MovieSceneScriptingFloatChannel, default=0, const [0,14]}
       thumb_03_l_ctrl.Scale.Z {type=MovieSceneScriptingFloatChannel, default=0, const [0,14]}
       index_metacarpal_l_ctrl.Location.X {type=MovieSceneScriptingFloatChannel, default=0, const [0,14]}
       index_metacarpal_l_ctrl.Location.Y {type=MovieSceneScriptingFloatChannel, default=0, const [0,14]}
       index_metacarpal_l_ctrl.Location.Z {type=MovieSceneScriptingFloatChannel, default=0, const [0,14]}
       index_metacarpal_l_ctrl.Rotation.X {
        type=MovieSceneScriptingFloatChannel
        default=0
        thinned_from=15
        keys [[0,0],[32,0],[55,-8.9287],[78,-9.6026],[102,0],[242,0],[272,-1.3292],[302,-1.7026],[330,0],[352,0]]
       }
       index_metacarpal_l_ctrl.Rotation.Y {
        type=MovieSceneScriptingFloatChannel
        default=0
        thinned_from=15
        keys [[0,0],[32,0],[55,-2.5427],[78,-2.7346],[102,0],[242,0],[272,2.8089],[302,3.5979],[330,0],[352,0]]
       }
       index_metacarpal_l_ctrl.Rotation.Z {
        type=MovieSceneScriptingFloatChannel
        default=0
        thinned_from=14
        keys [[0,0],[32,0],[55,3.182],[78,3.4221],[102,0],[242,0],[272,-1.4839],[302,-1.9007],[330,0],[352,0]]
       }
       index_metacarpal_l_ctrl.Scale.X {type=MovieSceneScriptingFloatChannel, default=0, const [0,14]}
       index_metacarpal_l_ctrl.Scale.Y {type=MovieSceneScriptingFloatChannel, default=0, const [0,14]}
       index_metacarpal_l_ctrl.Scale.Z {type=MovieSceneScriptingFloatChannel, default=0, const [0,14]}
       index_01_l_ctrl.Location.X {type=MovieSceneScriptingFloatChannel, default=0, const [0,14]}
       index_01_l_ctrl.Location.Y {type=MovieSceneScriptingFloatChannel, default=0, const [0,14]}
       index_01_l_ctrl.Location.Z {type=MovieSceneScriptingFloatChannel, default=0, const [0,14]}
       index_01_l_ctrl.Rotation.X {
        type=MovieSceneScriptingFloatChannel
        default=0
        keys [[0,0],[32,0],[55,4.5667],[78,4.9114],[102,0],[132,0],[140,0.0308],[166,0.7815],[190,-0.3851],[221,-0.0008],[242,-0.0011],[272,6.2462],[302,8.001],[330,-0.0011],[352,0]]
       }
       index_01_l_ctrl.Rotation.Y {
        type=MovieSceneScriptingFloatChannel
        default=0
        keys [[0,0],[32,0],[55,-7.9781],[78,-8.5802],[102,0],[132,0],[140,0.1206],[166,3.0562],[190,-1.506],[221,-0.0032],[242,-0.0043],[272,-8.2148],[302,-10.5211],[330,-0.0043],[352,0]]
       }
       index_01_l_ctrl.Rotation.Z {
        type=MovieSceneScriptingFloatChannel
        default=0
        keys [[0,0],[32,0],[55,1.5204],[78,1.6351],[102,0],[132,0],[140,0.2555],[166,6.4722],[190,-3.1893],[221,-0.0067],[242,-0.0092],[272,-0.0768],[302,-0.0957],[330,-0.0092],[352,0]]
       }
       index_01_l_ctrl.Scale.X {type=MovieSceneScriptingFloatChannel, default=0, const [0,14]}
       index_01_l_ctrl.Scale.Y {type=MovieSceneScriptingFloatChannel, default=0, const [0,14]}
       index_01_l_ctrl.Scale.Z {type=MovieSceneScriptingFloatChannel, default=0, const [0,14]}
       index_02_l_ctrl.Location.X {type=MovieSceneScriptingFloatChannel, default=0, const [0,14]}
       index_02_l_ctrl.Location.Y {type=MovieSceneScriptingFloatChannel, default=0, const [0,14]}
       index_02_l_ctrl.Location.Z {type=MovieSceneScriptingFloatChannel, default=0, const [0,14]}
       index_02_l_ctrl.Rotation.X {type=MovieSceneScriptingFloatChannel, default=0, thinned_from=14, keys [[0,0],[352,0]]}
       index_02_l_ctrl.Rotation.Y {type=MovieSceneScriptingFloatChannel, default=0, thinned_from=14, keys [[0,0],[352,0]]}
       index_02_l_ctrl.Rotation.Z {
        type=MovieSceneScriptingFloatChannel
        default=0
        thinned_from=14
        keys [[0,0],[132,0],[140,-0.2661],[166,-6.7403],[190,3.3215],[221,0.007],[242,0.0095],[272,-3.0955],[302,-3.9678],[330,0.0095],[352,0]]
       }
       index_02_l_ctrl.Scale.X {type=MovieSceneScriptingFloatChannel, default=0, const [0,14]}
       index_02_l_ctrl.Scale.Y {type=MovieSceneScriptingFloatChannel, default=0, const [0,14]}
       index_02_l_ctrl.Scale.Z {type=MovieSceneScriptingFloatChannel, default=0, const [0,14]}
       index_03_l_ctrl.Location.X {type=MovieSceneScriptingFloatChannel, default=0, const [0,14]}
       index_03_l_ctrl.Location.Y {type=MovieSceneScriptingFloatChannel, default=0, const [0,14]}
       index_03_l_ctrl.Location.Z {type=MovieSceneScriptingFloatChannel, default=0, const [0,14]}
       index_03_l_ctrl.Rotation.X {type=MovieSceneScriptingFloatChannel, default=0, const [0,14]}
       index_03_l_ctrl.Rotation.Y {type=MovieSceneScriptingFloatChannel, default=0, const [0,14]}
       index_03_l_ctrl.Rotation.Z {
        type=MovieSceneScriptingFloatChannel
        default=0
        thinned_from=14
        keys [[0,0],[132,0],[140,0.0573],[166,1.452],[190,-0.7155],[221,-0.0015],[242,-0.0021],[272,-0.5497],[302,-0.7035],[330,-0.0083],[352,0]]
       }
       index_03_l_ctrl.Scale.X {type=MovieSceneScriptingFloatChannel, default=0, const [0,14]}
       index_03_l_ctrl.Scale.Y {type=MovieSceneScriptingFloatChannel, default=0, const [0,14]}
       index_03_l_ctrl.Scale.Z {type=MovieSceneScriptingFloatChannel, default=0, const [0,14]}
       middle_metacarpal_l_ctrl.Location.X {type=MovieSceneScriptingFloatChannel, default=0, const [0,14]}
       middle_metacarpal_l_ctrl.Location.Y {type=MovieSceneScriptingFloatChannel, default=0, const [0,14]}
       middle_metacarpal_l_ctrl.Location.Z {type=MovieSceneScriptingFloatChannel, default=0, const [0,14]}
       middle_metacarpal_l_ctrl.Rotation.X {
        type=MovieSceneScriptingFloatChannel
        default=0
        thinned_from=15
        keys [[0,0],[32,0],[55,-7.4321],[78,-7.9931],[102,0],[242,0],[272,-0.4858],[302,-0.6223],[330,0],[352,0]]
       }
       middle_metacarpal_l_ctrl.Rotation.Y {
        type=MovieSceneScriptingFloatChannel
        default=0
        thinned_from=15
        keys [[0,0],[32,0],[55,-5.608],[78,-6.0313],[102,0],[242,0],[272,0.914],[302,1.1707],[330,0],[352,0]]
       }
       middle_metacarpal_l_ctrl.Rotation.Z {
        type=MovieSceneScriptingFloatChannel
        default=0
        thinned_from=15
        keys [[0,0],[32,0],[55,2.4257],[78,2.6088],[102,0],[242,0],[272,-0.4664],[302,-0.5975],[330,0],[352,0]]
       }
       middle_metacarpal_l_ctrl.Scale.X {type=MovieSceneScriptingFloatChannel, default=0, const [0,14]}
       middle_metacarpal_l_ctrl.Scale.Y {type=MovieSceneScriptingFloatChannel, default=0, const [0,14]}
       middle_metacarpal_l_ctrl.Scale.Z {type=MovieSceneScriptingFloatChannel, default=0, const [0,14]}
       middle_01_l_ctrl.Location.X {type=MovieSceneScriptingFloatChannel, default=0, const [0,14]}
       middle_01_l_ctrl.Location.Y {type=MovieSceneScriptingFloatChannel, default=0, const [0,14]}
       middle_01_l_ctrl.Location.Z {type=MovieSceneScriptingFloatChannel, default=0, const [0,14]}
       middle_01_l_ctrl.Rotation.X {
        type=MovieSceneScriptingFloatChannel
        default=0
        keys [[0,0],[32,0],[55,0.2654],[78,0.2854],[102,0],[132,0],[140,-0.1832],[166,-4.6412],[190,2.2871],[221,0.0048],[242,0.0066],[272,7.4622],[302,9.5565],[330,0.0066],[352,0]]
       }
       middle_01_l_ctrl.Rotation.Y {
        type=MovieSceneScriptingFloatChannel
        default=0
        keys [[0,0],[32,0],[55,-4.406],[78,-4.7386],[102,0],[132,0],[140,0.2091],[166,5.2966],[190,-2.6101],[221,-0.0055],[242,-0.0075],[272,-6.5957],[302,-8.4463],[330,-0.0075],[352,0]]
       }
       middle_01_l_ctrl.Rotation.Z {
        type=MovieSceneScriptingFloatChannel
        default=0
        keys [[0,0],[32,0],[55,3.3782],[78,3.6332],[102,0],[132,0],[140,0.2212],[166,5.6034],[190,-2.7612],[221,-0.0058],[242,-0.0079],[272,1.2088],[302,1.5505],[330,-0.0079],[352,0]]
       }
       middle_01_l_ctrl.Scale.X {type=MovieSceneScriptingFloatChannel, default=0, const [0,14]}
       middle_01_l_ctrl.Scale.Y {type=MovieSceneScriptingFloatChannel, default=0, const [0,14]}
       middle_01_l_ctrl.Scale.Z {type=MovieSceneScriptingFloatChannel, default=0, const [0,14]}
       middle_02_l_ctrl.Location.X {type=MovieSceneScriptingFloatChannel, default=0, const [0,14]}
       middle_02_l_ctrl.Location.Y {type=MovieSceneScriptingFloatChannel, default=0, const [0,14]}
       middle_02_l_ctrl.Location.Z {type=MovieSceneScriptingFloatChannel, default=0, const [0,14]}
       middle_02_l_ctrl.Rotation.X {type=MovieSceneScriptingFloatChannel, default=0, thinned_from=14, keys [[0,0],[352,0]]}
       middle_02_l_ctrl.Rotation.Y {type=MovieSceneScriptingFloatChannel, default=0, thinned_from=14, keys [[0,0],[352,0]]}
       middle_02_l_ctrl.Rotation.Z {
        type=MovieSceneScriptingFloatChannel
        default=0
        thinned_from=14
        keys [[0,0],[132,0],[140,-0.0946],[166,-2.3953],[190,1.1803],[221,0.0025],[242,0.0108],[272,-4.9647],[302,-6.3624],[330,0.0108],[352,0]]
       }
       middle_02_l_ctrl.Scale.X {type=MovieSceneScriptingFloatChannel, default=0, const [0,14]}
       middle_02_l_ctrl.Scale.Y {type=MovieSceneScriptingFloatChannel, default=0, const [0,14]}
       middle_02_l_ctrl.Scale.Z {type=MovieSceneScriptingFloatChannel, default=0, const [0,14]}
       middle_03_l_ctrl.Location.X {type=MovieSceneScriptingFloatChannel, default=0, const [0,14]}
       middle_03_l_ctrl.Location.Y {type=MovieSceneScriptingFloatChannel, default=0, const [0,14]}
       middle_03_l_ctrl.Location.Z {type=MovieSceneScriptingFloatChannel, default=0, const [0,14]}
       middle_03_l_ctrl.Rotation.X {type=MovieSceneScriptingFloatChannel, default=0, const [0,14]}
       middle_03_l_ctrl.Rotation.Y {type=MovieSceneScriptingFloatChannel, default=0, const [0,14]}
       middle_03_l_ctrl.Rotation.Z {
        type=MovieSceneScriptingFloatChannel
        default=0
        thinned_from=14
        keys [[0,0],[132,0],[140,-0.3442],[166,-8.7185],[190,4.2963],[221,0.0091],[242,0.0123],[272,3.3342],[302,4.2673],[330,0.0123],[352,0]]
       }
       middle_03_l_ctrl.Scale.X {type=MovieSceneScriptingFloatChannel, default=0, const [0,14]}
       middle_03_l_ctrl.Scale.Y {type=MovieSceneScriptingFloatChannel, default=0, const [0,14]}
       middle_03_l_ctrl.Scale.Z {type=MovieSceneScriptingFloatChannel, default=0, const [0,14]}
       ring_metacarpal_l_ctrl.Location.X {type=MovieSceneScriptingFloatChannel, default=0, const [0,14]}
       ring_metacarpal_l_ctrl.Location.Y {type=MovieSceneScriptingFloatChannel, default=0, const [0,14]}
       ring_metacarpal_l_ctrl.Location.Z {type=MovieSceneScriptingFloatChannel, default=0, const [0,14]}
       ring_metacarpal_l_ctrl.Rotation.X {
        type=MovieSceneScriptingFloatChannel
        default=0
        thinned_from=15
        keys [[0,0],[32,0],[55,-7.0962],[78,-7.6318],[102,0],[242,0],[272,0.9142],[302,1.171],[330,0],[352,0]]
       }
       ring_metacarpal_l_ctrl.Rotation.Y {
        type=MovieSceneScriptingFloatChannel
        default=0
        thinned_from=15
        keys [[0,0],[32,0],[55,-6.192],[78,-6.6593],[102,0],[242,0],[272,0.2409],[302,0.3086],[330,0],[352,0]]
       }
       ring_metacarpal_l_ctrl.Rotation.Z {
        type=MovieSceneScriptingFloatChannel
        default=0
        thinned_from=15
        keys [[0,0],[32,0],[55,2.2175],[78,2.3849],[102,0],[242,0],[272,0.9425],[302,1.2072],[330,0],[352,0]]
       }
       ring_metacarpal_l_ctrl.Scale.X {type=MovieSceneScriptingFloatChannel, default=0, const [0,14]}
       ring_metacarpal_l_ctrl.Scale.Y {type=MovieSceneScriptingFloatChannel, default=0, const [0,14]}
       ring_metacarpal_l_ctrl.Scale.Z {type=MovieSceneScriptingFloatChannel, default=0, const [0,14]}
       ring_01_l_ctrl.Location.X {type=MovieSceneScriptingFloatChannel, default=0, const [0,14]}
       ring_01_l_ctrl.Location.Y {type=MovieSceneScriptingFloatChannel, default=0, const [0,14]}
       ring_01_l_ctrl.Location.Z {type=MovieSceneScriptingFloatChannel, default=0, const [0,14]}
       ring_01_l_ctrl.Rotation.X {
        type=MovieSceneScriptingFloatChannel
        default=0
        keys [[0,0],[32,0],[55,-0.2885],[78,-0.3103],[102,0],[132,0],[140,0.0708],[166,1.7935],[190,-0.8838],[221,-0.0019],[242,-0.0081],[272,9.0103],[302,11.5435],[330,-0.0081],[352,0]]
       }
       ring_01_l_ctrl.Rotation.Y {
        type=MovieSceneScriptingFloatChannel
        default=0
        keys [[0,0],[32,0],[55,-4.4049],[78,-4.7373],[102,0],[132,0],[140,0.0799],[166,2.0243],[190,-0.9975],[221,-0.0021],[242,-0.0091],[272,-3.9749],[302,-5.0889],[330,-0.0091],[352,0]]
       }
       ring_01_l_ctrl.Rotation.Z {
        type=MovieSceneScriptingFloatChannel
        default=0
        keys [[0,0],[32,0],[55,0.8445],[78,0.9082],[102,0],[132,0],[140,-0.0198],[166,-0.5018],[190,0.2473],[221,0.0005],[242,0.0023],[272,-0.2465],[302,-0.3164],[330,0.0023],[352,0]]
       }
       ring_01_l_ctrl.Scale.X {type=MovieSceneScriptingFloatChannel, default=0, const [0,14]}
       ring_01_l_ctrl.Scale.Y {type=MovieSceneScriptingFloatChannel, default=0, const [0,14]}
       ring_01_l_ctrl.Scale.Z {type=MovieSceneScriptingFloatChannel, default=0, const [0,14]}
       ring_02_l_ctrl.Location.X {type=MovieSceneScriptingFloatChannel, default=0, const [0,14]}
       ring_02_l_ctrl.Location.Y {type=MovieSceneScriptingFloatChannel, default=0, const [0,14]}
       ring_02_l_ctrl.Location.Z {type=MovieSceneScriptingFloatChannel, default=0, const [0,14]}
       ring_02_l_ctrl.Rotation.X {type=MovieSceneScriptingFloatChannel, default=0, const [0,14]}
       ring_02_l_ctrl.Rotation.Y {type=MovieSceneScriptingFloatChannel, default=0, const [0,14]}
       ring_02_l_ctrl.Rotation.Z {
        type=MovieSceneScriptingFloatChannel
        default=0
        thinned_from=14
        keys [[0,0],[132,0],[140,-0.081],[166,-2.0531],[190,1.0117],[221,0.0021],[242,0.0093],[272,0.7811],[302,0.9979],[330,0.0063],[352,0]]
       }
       ring_02_l_ctrl.Scale.X {type=MovieSceneScriptingFloatChannel, default=0, const [0,14]}
       ring_02_l_ctrl.Scale.Y {type=MovieSceneScriptingFloatChannel, default=0, const [0,14]}
       ring_02_l_ctrl.Scale.Z {type=MovieSceneScriptingFloatChannel, default=0, const [0,14]}
       ring_03_l_ctrl.Location.X {type=MovieSceneScriptingFloatChannel, default=0, const [0,14]}
       ring_03_l_ctrl.Location.Y {type=MovieSceneScriptingFloatChannel, default=0, const [0,14]}
       ring_03_l_ctrl.Location.Z {type=MovieSceneScriptingFloatChannel, default=0, const [0,14]}
       ring_03_l_ctrl.Rotation.X {type=MovieSceneScriptingFloatChannel, default=0, const [0,14]}
       ring_03_l_ctrl.Rotation.Y {type=MovieSceneScriptingFloatChannel, default=0, const [0,14]}
       ring_03_l_ctrl.Rotation.Z {
        type=MovieSceneScriptingFloatChannel
        default=0
        thinned_from=14
        keys [[0,0],[132,0],[140,-0.1099],[166,-2.7833],[190,1.3716],[221,0.0029],[242,0.0125],[272,1.0657],[302,1.3615],[330,0.0125],[352,0]]
       }
       ring_03_l_ctrl.Scale.X {type=MovieSceneScriptingFloatChannel, default=0, const [0,14]}
       ring_03_l_ctrl.Scale.Y {type=MovieSceneScriptingFloatChannel, default=0, const [0,14]}
       ring_03_l_ctrl.Scale.Z {type=MovieSceneScriptingFloatChannel, default=0, const [0,14]}
       pinky_metacarpal_l_ctrl.Location.X {type=MovieSceneScriptingFloatChannel, default=0, const [0,14]}
       pinky_metacarpal_l_ctrl.Location.Y {type=MovieSceneScriptingFloatChannel, default=0, const [0,14]}
       pinky_metacarpal_l_ctrl.Location.Z {type=MovieSceneScriptingFloatChannel, default=0, const [0,14]}
       pinky_metacarpal_l_ctrl.Rotation.X {
        type=MovieSceneScriptingFloatChannel
        default=0
        thinned_from=15
        keys [[0,0],[32,0],[55,-7.385],[78,-7.9423],[102,0],[242,0],[272,1.0682],[302,1.3682],[330,0],[352,0]]
       }
       pinky_metacarpal_l_ctrl.Rotation.Y {
        type=MovieSceneScriptingFloatChannel
        default=0
        thinned_from=15
        keys [[0,0],[32,0],[55,-5.291],[78,-5.6903],[102,0],[242,0],[272,-1.399],[302,-1.792],[330,0],[352,0]]
       }
       pinky_metacarpal_l_ctrl.Rotation.Z {
        type=MovieSceneScriptingFloatChannel
        default=0
        thinned_from=15
        keys [[0,0],[32,0],[55,-0.8685],[78,-0.934],[102,0],[242,0],[272,1.3651],[302,1.7486],[330,0],[352,0]]
       }
       pinky_metacarpal_l_ctrl.Scale.X {type=MovieSceneScriptingFloatChannel, default=0, const [0,14]}
       pinky_metacarpal_l_ctrl.Scale.Y {type=MovieSceneScriptingFloatChannel, default=0, const [0,14]}
       pinky_metacarpal_l_ctrl.Scale.Z {type=MovieSceneScriptingFloatChannel, default=0, const [0,14]}
       pinky_01_l_ctrl.Location.X {type=MovieSceneScriptingFloatChannel, default=0, const [0,14]}
       pinky_01_l_ctrl.Location.Y {type=MovieSceneScriptingFloatChannel, default=0, const [0,14]}
       pinky_01_l_ctrl.Location.Z {type=MovieSceneScriptingFloatChannel, default=0, const [0,14]}
       pinky_01_l_ctrl.Rotation.X {
        type=MovieSceneScriptingFloatChannel
        default=0
        keys [[0,0],[32,0],[55,-7.1948],[78,-7.7378],[102,0],[132,0],[140,0.0033],[166,0.083],[190,-0.0409],[221,-0.0001],[242,-0.0001],[272,-0.2796],[302,-0.3581],[330,-0.0032],[352,0]]
       }
       pinky_01_l_ctrl.Rotation.Y {
        type=MovieSceneScriptingFloatChannel
        default=0
        keys [[0,0],[32,0],[55,-9.3178],[78,-10.021],[102,0],[132,0],[140,0.1066],[166,2.7009],[190,-1.3309],[221,-0.0028],[242,-0.0016],[272,-0.7507],[302,-0.9611],[330,-0.0102],[352,0]]
       }
       pinky_01_l_ctrl.Rotation.Z {
        type=MovieSceneScriptingFloatChannel
        default=0
        keys [[0,0],[32,0],[55,8.4359],[79,9.0011],[102,0],[132,0],[140,-0.0955],[166,-2.4199],[190,1.1925],[221,0.0025],[242,0.0015],[272,-0.3191],[302,-0.4091],[330,-0.0022],[352,0]]
       }
       pinky_01_l_ctrl.Scale.X {type=MovieSceneScriptingFloatChannel, default=0, const [0,14]}
       pinky_01_l_ctrl.Scale.Y {type=MovieSceneScriptingFloatChannel, default=0, const [0,14]}
       pinky_01_l_ctrl.Scale.Z {type=MovieSceneScriptingFloatChannel, default=0, const [0,14]}
       pinky_02_l_ctrl.Location.X {type=MovieSceneScriptingFloatChannel, default=0, const [0,14]}
       pinky_02_l_ctrl.Location.Y {type=MovieSceneScriptingFloatChannel, default=0, const [0,14]}
       pinky_02_l_ctrl.Location.Z {type=MovieSceneScriptingFloatChannel, default=0, const [0,14]}
       pinky_02_l_ctrl.Rotation.X {type=MovieSceneScriptingFloatChannel, default=0, const [0,14]}
       pinky_02_l_ctrl.Rotation.Y {type=MovieSceneScriptingFloatChannel, default=0, const [0,14]}
       pinky_02_l_ctrl.Rotation.Z {
        type=MovieSceneScriptingFloatChannel
        default=0
        thinned_from=14
        keys [[0,0],[132,0],[140,-0.2498],[166,-6.3272],[190,3.1179],[221,0.0066],[242,0.009],[272,-0.8679],[302,-1.1141],[330,-0.001],[352,0]]
       }
       pinky_02_l_ctrl.Scale.X {type=MovieSceneScriptingFloatChannel, default=0, const [0,14]}
       pinky_02_l_ctrl.Scale.Y {type=MovieSceneScriptingFloatChannel, default=0, const [0,14]}
       pinky_02_l_ctrl.Scale.Z {type=MovieSceneScriptingFloatChannel, default=0, const [0,14]}
       pinky_03_l_ctrl.Location.X {type=MovieSceneScriptingFloatChannel, default=0, const [0,14]}
       pinky_03_l_ctrl.Location.Y {type=MovieSceneScriptingFloatChannel, default=0, const [0,14]}
       pinky_03_l_ctrl.Location.Z {type=MovieSceneScriptingFloatChannel, default=0, const [0,14]}
       pinky_03_l_ctrl.Rotation.X {type=MovieSceneScriptingFloatChannel, default=0, const [0,14]}
       pinky_03_l_ctrl.Rotation.Y {type=MovieSceneScriptingFloatChannel, default=0, const [0,14]}
       pinky_03_l_ctrl.Rotation.Z {
        type=MovieSceneScriptingFloatChannel
        default=0
        thinned_from=14
        keys [[0,0],[132,0],[140,-0.1632],[166,-4.1345],[190,2.0374],[221,0.0043],[242,0.0059],[272,1.6297],[302,2.0858],[330,0.0059],[352,0]]
       }
       pinky_03_l_ctrl.Scale.X {type=MovieSceneScriptingFloatChannel, default=0, const [0,14]}
       pinky_03_l_ctrl.Scale.Y {type=MovieSceneScriptingFloatChannel, default=0, const [0,14]}
       pinky_03_l_ctrl.Scale.Z {type=MovieSceneScriptingFloatChannel, default=0, const [0,14]}
       hand_l_ik_ctrl.Location.X {
        type=MovieSceneScriptingFloatChannel
        default=0
        keys [[-0.0013,-0.0021],[0,0],[32,0],[55,0.7637],[78,0.862],[102,0],[132,0],[140,-0.0639],[166,-1.6195],[190,0.798],[221,0.0017],[242,0.0013],[272,0.6233],[302,0.798],[330,0],[352,0]]
       }
       hand_l_ik_ctrl.Location.Y {
        type=MovieSceneScriptingFloatChannel
        default=0
        keys [[-0.0013,0.0014],[0,0],[32,0],[55,0.315],[78,0.3556],[102,0],[132,0],[140,-0.0276],[166,-0.6988],[190,0.3444],[221,0.0007],[242,0.0006],[272,0.269],[302,0.3444],[330,0],[352,0]]
       }
       hand_l_ik_ctrl.Location.Z {
        type=MovieSceneScriptingFloatChannel
        default=0
        keys [[-0.0013,-0.0032],[0,0],[32,0],[55,-0.4198],[78,-0.4738],[102,0],[132,0],[140,0.0368],[166,0.9324],[190,-0.4594],[221,-0.001],[242,-0.0008],[272,-0.3589],[302,-0.4594],[330,0],[352,0]]
       }
       hand_l_ik_ctrl.Rotation.X {
        type=MovieSceneScriptingFloatChannel
        default=0
        keys [[-0.0013,-0.0227],[0,0],[32,0],[55,1.5148],[78,1.8319],[102,0],[132,0],[140,0.3431],[166,8.6926],[190,-4.2835],[221,-0.009],[242,-0.0073],[272,-4.9647],[302,-6.3573],[330,-0.0002],[352,0]]
       }
       hand_l_ik_ctrl.Rotation.Y {
        type=MovieSceneScriptingFloatChannel
        default=0
        keys [[-0.0013,0.0034],[0,0],[32,0],[55,-1.8822],[78,-2.1245],[102,0],[132,0],[140,-0.1981],[166,-5.0176],[190,2.4725],[221,0.0052],[242,0.0042],[272,2.9169],[302,3.7351],[330,0.0001],[352,0]]
       }
       hand_l_ik_ctrl.Rotation.Z {
        type=MovieSceneScriptingFloatChannel
        default=0
        keys [[-0.0013,-0.0251],[0,0],[32,0],[55,-2.8189],[78,-3.1817],[102,0],[132,0],[140,0.088],[166,2.2304],[190,-1.0991],[221,-0.0023],[242,-0.0019],[272,-0.8828],[302,-1.1303],[330,0],[352,0]]
       }
       hand_l_ik_ctrl.Scale.X {type=MovieSceneScriptingFloatChannel, default=0, const [0,15]}
       hand_l_ik_ctrl.Scale.Y {type=MovieSceneScriptingFloatChannel, default=0, const [0,15]}
       hand_l_ik_ctrl.Scale.Z {type=MovieSceneScriptingFloatChannel, default=0, const [0,15]}
       Softness {type=MovieSceneScriptingFloatChannel, default=0, const [0,14]}
       thumb_01_r_ctrl.Location.X {type=MovieSceneScriptingFloatChannel, default=0, thinned_from=15, keys [[0,0],[352,0]]}
       thumb_01_r_ctrl.Location.Y {type=MovieSceneScriptingFloatChannel, default=0, thinned_from=15, keys [[0,0],[352,0]]}
       thumb_01_r_ctrl.Location.Z {type=MovieSceneScriptingFloatChannel, default=0, thinned_from=15, keys [[0,0],[352,0]]}
       thumb_01_r_ctrl.Rotation.X {
        type=MovieSceneScriptingFloatChannel
        default=0
        keys [[0,0],[32,0],[55,0.4134],[78,0.4553],[102,0],[132,0],[140,0.0812],[166,2.6444],[190,-1.1134],[221,-0.0049],[242,-0.0028],[272,-0.8699],[302,-1.1134],[330,-0.0127],[352,0]]
       }
       thumb_01_r_ctrl.Rotation.Y {
        type=MovieSceneScriptingFloatChannel
        default=0
        keys [[0,0],[32,0],[55,-1.0521],[78,-1.1588],[102,0],[132,0],[140,0.0021],[166,0.068],[190,-0.0286],[221,-0.0001],[242,-0.0001],[272,-0.0224],[302,-0.0287],[330,-0.0003],[352,0]]
       }
       thumb_01_r_ctrl.Rotation.Z {
        type=MovieSceneScriptingFloatChannel
        default=0
        keys [[0,0],[32,0],[55,-1.8356],[78,-2.0219],[102,0],[132,0],[140,0.0021],[166,0.0691],[190,-0.0291],[221,-0.0001],[242,-0.0001],[272,-0.0227],[302,-0.0291],[330,-0.0003],[352,0]]
       }
       thumb_01_r_ctrl.Scale.X {type=MovieSceneScriptingFloatChannel, default=0, const [0,15]}
       thumb_01_r_ctrl.Scale.Y {type=MovieSceneScriptingFloatChannel, default=0, const [0,15]}
       thumb_01_r_ctrl.Scale.Z {type=MovieSceneScriptingFloatChannel, default=0, const [0,15]}
       thumb_02_r_ctrl.Location.X {type=MovieSceneScriptingFloatChannel, default=0, const [0,15]}
       thumb_02_r_ctrl.Location.Y {type=MovieSceneScriptingFloatChannel, default=0, const [0,15]}
       thumb_02_r_ctrl.Location.Z {type=MovieSceneScriptingFloatChannel, default=0, const [0,15]}
       thumb_02_r_ctrl.Rotation.X {type=MovieSceneScriptingFloatChannel, default=0, thinned_from=15, keys [[0,0],[352,0]]}
       thumb_02_r_ctrl.Rotation.Y {type=MovieSceneScriptingFloatChannel, default=0, thinned_from=15, keys [[0,0],[352,0]]}
       thumb_02_r_ctrl.Rotation.Z {
        type=MovieSceneScriptingFloatChannel
        default=0
        keys [[0,0],[32,0],[55,-1.3664],[78,-1.505],[102,0],[132,0],[140,-0.0759],[166,-2.4709],[190,1.0404],[221,0.0046],[242,0.0139],[272,0.8153],[302,1.0404],[330,0.0117],[352,0]]
       }
       thumb_02_r_ctrl.Scale.X {type=MovieSceneScriptingFloatChannel, default=0, const [0,15]}
       thumb_02_r_ctrl.Scale.Y {type=MovieSceneScriptingFloatChannel, default=0, const [0,15]}
       thumb_02_r_ctrl.Scale.Z {type=MovieSceneScriptingFloatChannel, default=0, const [0,15]}
       thumb_03_r_ctrl.Location.X {type=MovieSceneScriptingFloatChannel, default=0, thinned_from=15, keys [[0,0],[352,0]]}
       thumb_03_r_ctrl.Location.Y {type=MovieSceneScriptingFloatChannel, default=0, thinned_from=15, keys [[0,0],[352,0]]}
       thumb_03_r_ctrl.Location.Z {type=MovieSceneScriptingFloatChannel, default=0, thinned_from=15, keys [[0,0],[352,0]]}
       thumb_03_r_ctrl.Rotation.X {type=MovieSceneScriptingFloatChannel, default=0, thinned_from=15, keys [[0,0],[352,0]]}
       thumb_03_r_ctrl.Rotation.Y {type=MovieSceneScriptingFloatChannel, default=0, thinned_from=15, keys [[0,0],[352,0]]}
       thumb_03_r_ctrl.Rotation.Z {
        type=MovieSceneScriptingFloatChannel
        default=0
        keys [[0,0],[32,-0.0082],[55,-3.125],[78,-3.4411],[102,0],[132,0],[140,-0.2319],[166,-7.5535],[190,3.1804],[221,0.0139],[242,0.0171],[272,2.4867],[302,3.1804],[330,0.005],[352,0]]
       }
       thumb_03_r_ctrl.Scale.X {type=MovieSceneScriptingFloatChannel, default=0, const [0,15]}
       thumb_03_r_ctrl.Scale.Y {type=MovieSceneScriptingFloatChannel, default=0, const [0,15]}
       thumb_03_r_ctrl.Scale.Z {type=MovieSceneScriptingFloatChannel, default=0, const [0,15]}
       index_metacarpal_r_ctrl.Location.X {type=MovieSceneScriptingFloatChannel, default=0, thinned_from=15, keys [[0,0],[352,0]]}
       index_metacarpal_r_ctrl.Location.Y {type=MovieSceneScriptingFloatChannel, default=0, thinned_from=15, keys [[0,0],[352,0]]}
       index_metacarpal_r_ctrl.Location.Z {type=MovieSceneScriptingFloatChannel, default=0, thinned_from=15, keys [[0,0],[352,0]]}
       index_metacarpal_r_ctrl.Rotation.X {type=MovieSceneScriptingFloatChannel, default=0, thinned_from=15, keys [[0,0],[32,0],[55,-4.6961],[78,-5.1725],[102,0],[352,0]]}
       index_metacarpal_r_ctrl.Rotation.Y {type=MovieSceneScriptingFloatChannel, default=0, thinned_from=15, keys [[0,0],[32,0],[55,-2.8564],[78,-3.1462],[102,0],[352,0]]}
       index_metacarpal_r_ctrl.Rotation.Z {type=MovieSceneScriptingFloatChannel, default=0, thinned_from=15, keys [[0,0],[32,0],[55,-0.6378],[78,-0.7025],[102,0],[352,0]]}
       index_metacarpal_r_ctrl.Scale.X {type=MovieSceneScriptingFloatChannel, default=0, const [0,15]}
       index_metacarpal_r_ctrl.Scale.Y {type=MovieSceneScriptingFloatChannel, default=0, const [0,15]}
       index_metacarpal_r_ctrl.Scale.Z {type=MovieSceneScriptingFloatChannel, default=0, const [0,15]}
       index_01_r_ctrl.Location.X {type=MovieSceneScriptingFloatChannel, default=0, thinned_from=15, keys [[0,0],[352,0]]}
       index_01_r_ctrl.Location.Y {type=MovieSceneScriptingFloatChannel, default=0, thinned_from=15, keys [[0,0],[352,0]]}
       index_01_r_ctrl.Location.Z {type=MovieSceneScriptingFloatChannel, default=0, thinned_from=15, keys [[0,0],[352,0]]}
       index_01_r_ctrl.Rotation.X {
        type=MovieSceneScriptingFloatChannel
        default=0
        keys [[0,0],[32,0.0056],[55,6.4262],[78,7.0776],[102,0],[132,0],[140,-0.0032],[166,-0.1042],[190,0.036],[221,0.0002],[242,0.0002],[272,0.0281],[302,0.036],[330,0.0002],[352,0]]
       }
       index_01_r_ctrl.Rotation.Y {
        type=MovieSceneScriptingFloatChannel
        default=0
        keys [[0,0],[32,0.0028],[55,-1.9931],[78,-2.1955],[102,0],[132,0],[140,-0.0013],[166,-0.041],[190,0.0119],[221,0.0001],[242,0.0001],[272,0.0093],[302,0.0119],[330,0.0001],[352,0]]
       }
       index_01_r_ctrl.Rotation.Z {
        type=MovieSceneScriptingFloatChannel
        default=0
        thinned_from=15
        keys [[0,0],[32,0.0001],[55,1.1801],[78,1.2998],[102,0],[140,-0.0001],[166,-0.0023],[190,3.1248],[221,0],[242,0.003],[272,2.4402],[302,3.1248],[330,0.003],[352,0]]
       }
       index_01_r_ctrl.Scale.X {type=MovieSceneScriptingFloatChannel, default=0, const [0,15]}
       index_01_r_ctrl.Scale.Y {type=MovieSceneScriptingFloatChannel, default=0, const [0,15]}
       index_01_r_ctrl.Scale.Z {type=MovieSceneScriptingFloatChannel, default=0, const [0,15]}
       index_02_r_ctrl.Location.X {type=MovieSceneScriptingFloatChannel, default=0, const [0,15]}
       index_02_r_ctrl.Location.Y {type=MovieSceneScriptingFloatChannel, default=0, const [0,15]}
       index_02_r_ctrl.Location.Z {type=MovieSceneScriptingFloatChannel, default=0, const [0,15]}
       index_02_r_ctrl.Rotation.X {type=MovieSceneScriptingFloatChannel, default=0, thinned_from=15, keys [[0,0],[352,0]]}
       index_02_r_ctrl.Rotation.Y {type=MovieSceneScriptingFloatChannel, default=0, thinned_from=15, keys [[0,0],[352,0]]}
       index_02_r_ctrl.Rotation.Z {
        type=MovieSceneScriptingFloatChannel
        default=0
        thinned_from=15
        keys [[0,0],[32,0],[55,-6.429],[78,-7.0812],[102,0],[221,0],[242,-0.009],[272,-0.002],[302,0],[352,0]]
       }
       index_02_r_ctrl.Scale.X {type=MovieSceneScriptingFloatChannel, default=0, const [0,15]}
       index_02_r_ctrl.Scale.Y {type=MovieSceneScriptingFloatChannel, default=0, const [0,15]}
       index_02_r_ctrl.Scale.Z {type=MovieSceneScriptingFloatChannel, default=0, const [0,15]}
       index_03_r_ctrl.Location.X {type=MovieSceneScriptingFloatChannel, default=0, thinned_from=15, keys [[0,0],[352,0]]}
       index_03_r_ctrl.Location.Y {type=MovieSceneScriptingFloatChannel, default=0, thinned_from=15, keys [[0,0],[352,0]]}
       index_03_r_ctrl.Location.Z {type=MovieSceneScriptingFloatChannel, default=0, thinned_from=15, keys [[0,0],[352,0]]}
       index_03_r_ctrl.Rotation.X {type=MovieSceneScriptingFloatChannel, default=0, thinned_from=15, keys [[0,0],[352,0]]}
       index_03_r_ctrl.Rotation.Y {type=MovieSceneScriptingFloatChannel, default=0, thinned_from=15, keys [[0,0],[352,0]]}
       index_03_r_ctrl.Rotation.Z {
        type=MovieSceneScriptingFloatChannel
        default=0
        thinned_from=15
        keys [[0,0],[32,0],[55,-7.274],[78,-8.012],[102,0],[221,0],[242,-0.0102],[272,-0.0022],[302,0],[352,0]]
       }
       index_03_r_ctrl.Scale.X {type=MovieSceneScriptingFloatChannel, default=0, const [0,15]}
       index_03_r_ctrl.Scale.Y {type=MovieSceneScriptingFloatChannel, default=0, const [0,15]}
       index_03_r_ctrl.Scale.Z {type=MovieSceneScriptingFloatChannel, default=0, const [0,15]}
       middle_metacarpal_r_ctrl.Location.X {type=MovieSceneScriptingFloatChannel, default=0, thinned_from=15, keys [[0,0],[352,0]]}
       middle_metacarpal_r_ctrl.Location.Y {type=MovieSceneScriptingFloatChannel, default=0, const [0,15]}
       middle_metacarpal_r_ctrl.Location.Z {type=MovieSceneScriptingFloatChannel, default=0, const [0,15]}
       middle_metacarpal_r_ctrl.Rotation.X {type=MovieSceneScriptingFloatChannel, default=0, thinned_from=15, keys [[0,0],[32,0],[55,0.041],[78,0.0452],[102,0],[352,0]]}
       middle_metacarpal_r_ctrl.Rotation.Y {type=MovieSceneScriptingFloatChannel, default=0, thinned_from=15, keys [[0,0],[32,0],[55,-1.2444],[78,-1.3707],[102,0],[352,0]]}
       middle_metacarpal_r_ctrl.Rotation.Z {type=MovieSceneScriptingFloatChannel, default=0, thinned_from=15, keys [[0,0],[32,0],[55,-1.0626],[78,-1.1704],[102,0],[352,0]]}
       middle_metacarpal_r_ctrl.Scale.X {type=MovieSceneScriptingFloatChannel, default=0, const [0,15]}
       middle_metacarpal_r_ctrl.Scale.Y {type=MovieSceneScriptingFloatChannel, default=0, const [0,15]}
       middle_metacarpal_r_ctrl.Scale.Z {type=MovieSceneScriptingFloatChannel, default=0, const [0,15]}
       middle_01_r_ctrl.Location.X {type=MovieSceneScriptingFloatChannel, default=0, const [0,15]}
       middle_01_r_ctrl.Location.Y {type=MovieSceneScriptingFloatChannel, default=0, const [0,15]}
       middle_01_r_ctrl.Location.Z {type=MovieSceneScriptingFloatChannel, default=0, const [0,15]}
       middle_01_r_ctrl.Rotation.X {
        type=MovieSceneScriptingFloatChannel
        default=0
        keys [[0,0],[32,0.0088],[55,13.328],[78,14.6792],[102,0],[132,0],[140,-0.0092],[166,-0.2996],[190,0.1277],[221,0.0006],[242,0.0007],[272,0.0998],[302,0.1277],[330,0.0007],[352,0]]
       }
       middle_01_r_ctrl.Rotation.Y {
        type=MovieSceneScriptingFloatChannel
        default=0
        keys [[0,0],[32,-0.0064],[55,-11.4107],[78,-12.5677],[102,0],[132,0],[140,-0.0032],[166,-0.1052],[190,0.0363],[221,0.0002],[242,0.0002],[272,0.0284],[302,0.0363],[330,0.0002],[352,0]]
       }
       middle_01_r_ctrl.Rotation.Z {
        type=MovieSceneScriptingFloatChannel
        default=0
        keys [[0,0],[32,0.0013],[55,3.5245],[78,3.882],[102,0],[132,0],[140,0.0987],[166,3.2134],[190,2.1428],[221,-0.0059],[242,-0.004],[272,1.672],[302,2.1428],[330,-0.004],[352,0]]
       }
       middle_01_r_ctrl.Scale.X {type=MovieSceneScriptingFloatChannel, default=0, const [0,15]}
       middle_01_r_ctrl.Scale.Y {type=MovieSceneScriptingFloatChannel, default=0, const [0,15]}
       middle_01_r_ctrl.Scale.Z {type=MovieSceneScriptingFloatChannel, default=0, const [0,15]}
       middle_02_r_ctrl.Location.X {type=MovieSceneScriptingFloatChannel, default=0, thinned_from=15, keys [[0,0],[352,0]]}
       middle_02_r_ctrl.Location.Y {type=MovieSceneScriptingFloatChannel, default=0, thinned_from=15, keys [[0,0],[352,0]]}
       middle_02_r_ctrl.Location.Z {type=MovieSceneScriptingFloatChannel, default=0, thinned_from=15, keys [[0,0],[352,0]]}
       middle_02_r_ctrl.Rotation.X {type=MovieSceneScriptingFloatChannel, default=0, thinned_from=15, keys [[0,0],[352,0]]}
       middle_02_r_ctrl.Rotation.Y {type=MovieSceneScriptingFloatChannel, default=0, thinned_from=15, keys [[0,0],[352,0]]}
       middle_02_r_ctrl.Rotation.Z {type=MovieSceneScriptingFloatChannel, default=0, thinned_from=15, keys [[0,0],[32,0],[55,-2.0695],[78,-2.2795],[102,0],[352,0]]}
       middle_02_r_ctrl.Scale.X {type=MovieSceneScriptingFloatChannel, default=0, const [0,15]}
       middle_02_r_ctrl.Scale.Y {type=MovieSceneScriptingFloatChannel, default=0, const [0,15]}
       middle_02_r_ctrl.Scale.Z {type=MovieSceneScriptingFloatChannel, default=0, const [0,15]}
       middle_03_r_ctrl.Location.X {type=MovieSceneScriptingFloatChannel, default=0, thinned_from=15, keys [[0,0],[352,0]]}
       middle_03_r_ctrl.Location.Y {type=MovieSceneScriptingFloatChannel, default=0, thinned_from=15, keys [[0,0],[352,0]]}
       middle_03_r_ctrl.Location.Z {type=MovieSceneScriptingFloatChannel, default=0, thinned_from=15, keys [[0,0],[352,0]]}
       middle_03_r_ctrl.Rotation.X {type=MovieSceneScriptingFloatChannel, default=0, thinned_from=15, keys [[0,0],[352,0]]}
       middle_03_r_ctrl.Rotation.Y {type=MovieSceneScriptingFloatChannel, default=0, thinned_from=15, keys [[0,0],[352,0]]}
       middle_03_r_ctrl.Rotation.Z {type=MovieSceneScriptingFloatChannel, default=0, const [0,15]}
       middle_03_r_ctrl.Scale.X {type=MovieSceneScriptingFloatChannel, default=0, const [0,15]}
       middle_03_r_ctrl.Scale.Y {type=MovieSceneScriptingFloatChannel, default=0, const [0,15]}
       middle_03_r_ctrl.Scale.Z {type=MovieSceneScriptingFloatChannel, default=0, const [0,15]}
       ring_metacarpal_r_ctrl.Location.X {type=MovieSceneScriptingFloatChannel, default=0, const [0,15]}
       ring_metacarpal_r_ctrl.Location.Y {type=MovieSceneScriptingFloatChannel, default=0, const [0,15]}
       ring_metacarpal_r_ctrl.Location.Z {type=MovieSceneScriptingFloatChannel, default=0, const [0,15]}
       ring_metacarpal_r_ctrl.Rotation.X {type=MovieSceneScriptingFloatChannel, default=0, thinned_from=15, keys [[0,0],[32,0],[55,0.2452],[78,0.2701],[102,0],[352,0]]}
       ring_metacarpal_r_ctrl.Rotation.Y {type=MovieSceneScriptingFloatChannel, default=0, thinned_from=15, keys [[0,0],[32,0],[55,-2.0013],[78,-2.2044],[102,0],[352,0]]}
       ring_metacarpal_r_ctrl.Rotation.Z {type=MovieSceneScriptingFloatChannel, default=0, thinned_from=15, keys [[0,0],[32,0],[55,-2.4291],[78,-2.6756],[102,0],[352,0]]}
       ring_metacarpal_r_ctrl.Scale.X {type=MovieSceneScriptingFloatChannel, default=0, const [0,15]}
       ring_metacarpal_r_ctrl.Scale.Y {type=MovieSceneScriptingFloatChannel, default=0, const [0,15]}
       ring_metacarpal_r_ctrl.Scale.Z {type=MovieSceneScriptingFloatChannel, default=0, const [0,15]}
       ring_01_r_ctrl.Location.X {type=MovieSceneScriptingFloatChannel, default=0, thinned_from=15, keys [[0,0],[352,0]]}
       ring_01_r_ctrl.Location.Y {type=MovieSceneScriptingFloatChannel, default=0, thinned_from=15, keys [[0,0],[352,0]]}
       ring_01_r_ctrl.Location.Z {type=MovieSceneScriptingFloatChannel, default=0, thinned_from=15, keys [[0,0],[352,0]]}
       ring_01_r_ctrl.Rotation.X {
        type=MovieSceneScriptingFloatChannel
        default=0
        keys [[0,0],[32,0],[55,10.576],[78,11.6489],[102,0],[132,0],[140,-0.0148],[166,-0.4827],[190,0.177],[221,0.0009],[242,0.0011],[272,0.1384],[302,0.177],[330,0.0011],[352,0]]
       }
       ring_01_r_ctrl.Rotation.Y {
        type=MovieSceneScriptingFloatChannel
        default=0
        keys [[0,0],[32,0],[55,-6.6033],[78,-7.2733],[102,0],[132,0],[140,-0.031],[166,-1.0105],[190,1.9766],[221,0.0019],[242,0.0038],[272,1.544],[302,1.9766],[330,0.0038],[352,0]]
       }
       ring_01_r_ctrl.Rotation.Z {
        type=MovieSceneScriptingFloatChannel
        default=0
        keys [[0,0],[32,0],[55,13.2446],[78,14.5883],[102,0],[132,0],[140,0.1119],[166,3.6462],[190,3.0923],[221,-0.0067],[242,-0.0039],[272,2.4133],[302,3.0923],[330,-0.0039],[352,0]]
       }
       ring_01_r_ctrl.Scale.X {type=MovieSceneScriptingFloatChannel, default=0, const [0,15]}
       ring_01_r_ctrl.Scale.Y {type=MovieSceneScriptingFloatChannel, default=0, const [0,15]}
       ring_01_r_ctrl.Scale.Z {type=MovieSceneScriptingFloatChannel, default=0, const [0,15]}
       ring_02_r_ctrl.Location.X {type=MovieSceneScriptingFloatChannel, default=0, thinned_from=15, keys [[0,0],[352,0]]}
       ring_02_r_ctrl.Location.Y {type=MovieSceneScriptingFloatChannel, default=0, thinned_from=15, keys [[0,0],[352,0]]}
       ring_02_r_ctrl.Location.Z {type=MovieSceneScriptingFloatChannel, default=0, const [0,15]}
       ring_02_r_ctrl.Rotation.X {type=MovieSceneScriptingFloatChannel, default=0, thinned_from=15, keys [[0,0],[352,0]]}
       ring_02_r_ctrl.Rotation.Y {type=MovieSceneScriptingFloatChannel, default=0, thinned_from=15, keys [[0,0],[352,0]]}
       ring_02_r_ctrl.Rotation.Z {
        type=MovieSceneScriptingFloatChannel
        default=0
        thinned_from=15
        keys [[0,0],[132,0],[140,0.1299],[166,4.2317],[190,-1.7818],[221,-0.0078],[242,-0.0096],[272,-1.3931],[302,-1.7818],[330,-0.0096],[352,0]]
       }
       ring_02_r_ctrl.Scale.X {type=MovieSceneScriptingFloatChannel, default=0, const [0,15]}
       ring_02_r_ctrl.Scale.Y {type=MovieSceneScriptingFloatChannel, default=0, const [0,15]}
       ring_02_r_ctrl.Scale.Z {type=MovieSceneScriptingFloatChannel, default=0, const [0,15]}
       ring_03_r_ctrl.Location.X {type=MovieSceneScriptingFloatChannel, default=0, thinned_from=15, keys [[0,0],[352,0]]}
       ring_03_r_ctrl.Location.Y {type=MovieSceneScriptingFloatChannel, default=0, const [0,15]}
       ring_03_r_ctrl.Location.Z {type=MovieSceneScriptingFloatChannel, default=0, const [0,15]}
       ring_03_r_ctrl.Rotation.X {type=MovieSceneScriptingFloatChannel, default=0, thinned_from=15, keys [[0,0],[352,0]]}
       ring_03_r_ctrl.Rotation.Y {type=MovieSceneScriptingFloatChannel, default=0, thinned_from=15, keys [[0,0],[352,0]]}
       ring_03_r_ctrl.Rotation.Z {type=MovieSceneScriptingFloatChannel, default=0, const [0,15]}
       ring_03_r_ctrl.Scale.X {type=MovieSceneScriptingFloatChannel, default=0, const [0,15]}
       ring_03_r_ctrl.Scale.Y {type=MovieSceneScriptingFloatChannel, default=0, const [0,15]}
       ring_03_r_ctrl.Scale.Z {type=MovieSceneScriptingFloatChannel, default=0, const [0,15]}
       pinky_metacarpal_r_ctrl.Location.X {type=MovieSceneScriptingFloatChannel, default=0, thinned_from=15, keys [[0,0],[352,0]]}
       pinky_metacarpal_r_ctrl.Location.Y {type=MovieSceneScriptingFloatChannel, default=0, thinned_from=15, keys [[0,0],[352,0]]}
       pinky_metacarpal_r_ctrl.Location.Z {type=MovieSceneScriptingFloatChannel, default=0, const [0,15]}
       pinky_metacarpal_r_ctrl.Rotation.X {type=MovieSceneScriptingFloatChannel, default=0, thinned_from=15, keys [[0,0],[32,0],[55,6.9626],[78,7.6689],[102,0],[352,0]]}
       pinky_metacarpal_r_ctrl.Rotation.Y {type=MovieSceneScriptingFloatChannel, default=0, thinned_from=15, keys [[0,0],[32,0],[55,-4.7081],[78,-5.1857],[102,0],[352,0]]}
       pinky_metacarpal_r_ctrl.Rotation.Z {type=MovieSceneScriptingFloatChannel, default=0, thinned_from=15, keys [[0,0],[32,0],[55,3.2251],[78,3.5523],[102,0],[352,0]]}
       pinky_metacarpal_r_ctrl.Scale.X {type=MovieSceneScriptingFloatChannel, default=0, const [0,15]}
       pinky_metacarpal_r_ctrl.Scale.Y {type=MovieSceneScriptingFloatChannel, default=0, const [0,15]}
       pinky_metacarpal_r_ctrl.Scale.Z {type=MovieSceneScriptingFloatChannel, default=0, const [0,15]}
       pinky_01_r_ctrl.Location.X {type=MovieSceneScriptingFloatChannel, default=0, thinned_from=15, keys [[0,0],[352,0]]}
       pinky_01_r_ctrl.Location.Y {type=MovieSceneScriptingFloatChannel, default=0, thinned_from=15, keys [[0,0],[352,0]]}
       pinky_01_r_ctrl.Location.Z {type=MovieSceneScriptingFloatChannel, default=0, thinned_from=15, keys [[0,0],[352,0]]}
       pinky_01_r_ctrl.Rotation.X {
        type=MovieSceneScriptingFloatChannel
        default=0
        keys [[0,0],[32,0],[55,1.0431],[78,1.1489],[102,0],[132,0],[140,-0.0668],[166,-2.177],[190,0.5315],[221,0.004],[242,0.0024],[272,0.4154],[302,0.5315],[330,0.0004],[352,0]]
       }
       pinky_01_r_ctrl.Rotation.Y {
        type=MovieSceneScriptingFloatChannel
        default=0
        keys [[0,0],[32,0],[55,-3.7345],[78,-4.1133],[102,0],[132,0],[140,-0.0839],[166,-2.7314],[190,0.8741],[221,0.005],[242,0.0031],[272,0.6831],[302,0.8741],[330,0.0005],[352,0]]
       }
       pinky_01_r_ctrl.Rotation.Z {
        type=MovieSceneScriptingFloatChannel
        default=0
        keys [[0,0],[32,0],[55,-4.406],[78,-4.853],[102,0],[132,0],[140,0.3835],[166,12.4922],[190,-2.6623],[221,-0.0231],[242,-0.014],[272,-2.0816],[302,-2.6623],[330,-0.0024],[352,0]]
       }
       pinky_01_r_ctrl.Scale.X {type=MovieSceneScriptingFloatChannel, default=0, const [0,15]}
       pinky_01_r_ctrl.Scale.Y {type=MovieSceneScriptingFloatChannel, default=0, const [0,15]}
       pinky_01_r_ctrl.Scale.Z {type=MovieSceneScriptingFloatChannel, default=0, const [0,15]}
       pinky_02_r_ctrl.Location.X {type=MovieSceneScriptingFloatChannel, default=0, thinned_from=15, keys [[0,0],[352,0]]}
       pinky_02_r_ctrl.Location.Y {type=MovieSceneScriptingFloatChannel, default=0, thinned_from=15, keys [[0,0],[352,0]]}
       pinky_02_r_ctrl.Location.Z {type=MovieSceneScriptingFloatChannel, default=0, thinned_from=15, keys [[0,0],[352,0]]}
       pinky_02_r_ctrl.Rotation.X {type=MovieSceneScriptingFloatChannel, default=0, const [0,15]}
       pinky_02_r_ctrl.Rotation.Y {type=MovieSceneScriptingFloatChannel, default=0, thinned_from=15, keys [[0,0],[352,0]]}
       pinky_02_r_ctrl.Rotation.Z {
        type=MovieSceneScriptingFloatChannel
        default=0
        thinned_from=15
        keys [[0,0],[132,0],[140,0.0796],[166,2.5925],[190,-1.0916],[221,-0.0048],[242,-0.003],[272,-0.8529],[302,-1.0916],[330,-0.0127],[352,0]]
       }
       pinky_02_r_ctrl.Scale.X {type=MovieSceneScriptingFloatChannel, default=0, const [0,15]}
       pinky_02_r_ctrl.Scale.Y {type=MovieSceneScriptingFloatChannel, default=0, const [0,15]}
       pinky_02_r_ctrl.Scale.Z {type=MovieSceneScriptingFloatChannel, default=0, const [0,15]}
       pinky_03_r_ctrl.Location.X {type=MovieSceneScriptingFloatChannel, default=0, thinned_from=15, keys [[0,0],[352,0]]}
       pinky_03_r_ctrl.Location.Y {type=MovieSceneScriptingFloatChannel, default=0, thinned_from=15, keys [[0,0],[352,0]]}
       pinky_03_r_ctrl.Location.Z {type=MovieSceneScriptingFloatChannel, default=0, const [0,15]}
       pinky_03_r_ctrl.Rotation.X {type=MovieSceneScriptingFloatChannel, default=0, thinned_from=15, keys [[0,0],[352,0]]}
       pinky_03_r_ctrl.Rotation.Y {type=MovieSceneScriptingFloatChannel, default=0, thinned_from=15, keys [[0,0],[352,0]]}
       pinky_03_r_ctrl.Rotation.Z {
        type=MovieSceneScriptingFloatChannel
        default=0
        thinned_from=15
        keys [[0,0],[132,0],[140,0.0331],[166,1.0774],[190,-0.4536],[221,-0.002],[242,-0.0019],[272,-0.3546],[302,-0.4536],[330,-0.0019],[352,0]]
       }
       pinky_03_r_ctrl.Scale.X {type=MovieSceneScriptingFloatChannel, default=0, const [0,15]}
       pinky_03_r_ctrl.Scale.Y {type=MovieSceneScriptingFloatChannel, default=0, const [0,15]}
       pinky_03_r_ctrl.Scale.Z {type=MovieSceneScriptingFloatChannel, default=0, const [0,15]}
       hand_r_ik_ctrl.Location.X {
        type=MovieSceneScriptingFloatChannel
        default=0
        keys [[0,0],[32,0],[55,-0.0849],[78,-0.0958],[102,0],[132,0],[140,0.0472],[166,1.5372],[190,-0.6473],[221,-0.0028],[242,-0.0022],[272,-0.5058],[302,-0.6473],[330,0],[352,0]]
       }
       hand_r_ik_ctrl.Location.Y {
        type=MovieSceneScriptingFloatChannel
        default=0
        keys [[0,0],[32,0],[55,-0.0388],[78,-0.0438],[102,0],[132,0],[140,0.0597],[166,1.943],[190,-0.8181],[221,-0.0036],[242,-0.0028],[272,-0.6393],[302,-0.8181],[330,-0.0001],[352,0]]
       }
       hand_r_ik_ctrl.Location.Z {
        type=MovieSceneScriptingFloatChannel
        default=0
        keys [[0,0],[32,0],[55,0.101],[78,0.114],[102,0],[132,0],[140,-0.0281],[166,-0.9138],[190,0.3847],[221,0.0017],[242,0.0013],[272,0.3007],[302,0.3847],[330,0],[352,0]]
       }
       hand_r_ik_ctrl.Rotation.X {
        type=MovieSceneScriptingFloatChannel
        default=0
        keys [[0,0],[32,0],[55,-6.9816],[78,-7.8803],[102,0],[132,0],[140,-0.3095],[166,-10.0808],[190,4.2445],[221,0.0186],[242,0.015],[272,3.317],[302,4.2446],[330,0.0004],[352,0]]
       }
       hand_r_ik_ctrl.Rotation.Y {
        type=MovieSceneScriptingFloatChannel
        default=0
        keys [[0,0],[32,0],[55,0.6198],[78,0.6996],[102,0],[132,0],[140,-0.0788],[166,-2.5674],[190,1.081],[221,0.0047],[242,0.0038],[272,0.8448],[302,1.081],[330,0.0001],[352,0]]
       }
       hand_r_ik_ctrl.Rotation.Z {
        type=MovieSceneScriptingFloatChannel
        default=0
        keys [[0,0],[32,0],[55,4.63],[78,5.2259],[102,0],[132,0],[140,-0.146],[166,-4.7543],[190,2.0018],[221,0.0088],[242,0.0071],[272,1.5644],[302,2.0018],[330,0.0002],[352,0]]
       }
       hand_r_ik_ctrl.Scale.X {type=MovieSceneScriptingFloatChannel, default=0, const [0,15]}
       hand_r_ik_ctrl.Scale.Y {type=MovieSceneScriptingFloatChannel, default=0, const [0,15]}
       hand_r_ik_ctrl.Scale.Z {type=MovieSceneScriptingFloatChannel, default=0, const [0,15]}
       Softness_2 {type=MovieSceneScriptingFloatChannel, default=0, const [0,15]}
       Weight {type=MovieSceneScriptingFloatChannel, default=1}
      ]
     }
     {
      type=MovieSceneControlRigParameterSection
      channels [
       spine_02_ctrl.Location.X {type=MovieSceneScriptingFloatChannel, default=0, const [0,171]}
       spine_02_ctrl.Location.Y {
        type=MovieSceneScriptingFloatChannel
        default=0
        thinned_from=171
        keys [[0,0],[4,0],[5,-0.0021],[7,0.0064],[8,0.0121],[9,0.0201],[10,0.029],[11,0.0402],[12,0.053],[13,0.0672],[14,0.0832],[15,0.1007],[16,0.1198],[17,0.1409],[18,0.1624],[19,0.1865],[21,0.2328],[22,0.2577],[23,0.2783],[24,0.3047],[25,0.3236],[26,0.3443],[27,0.3672],[28,0.3855],[29,0.4059],[30,0.4214],[31,0.4385],[32,0.4503],[34,0.478],[35,0.4883],[37,0.5027],[38,0.5112],[39,0.5186],[41,0.5279],[43,0.5342],[46,0.5364],[48,0.535],[52,0.5202],[55,0.498],[56,0.4884],[57,0.477],[59,0.4495],[60,0.4323],[61,0.412],[62,0.393],[63,0.3731],[64,0.3512],[70,0.2237],[72,0.1826],[73,0.1614],[74,0.1425],[75,0.1278],[76,0.1118],[77,0.0995],[78,0.0882],[79,0.0772],[80,0.0685],[81,0.059],[82,0.0526],[83,0.0452],[84,0.0406],[85,0.0346],[86,0.0299],[87,0.0261],[109,0],[218,0],[219,-0.0021],[221,0.0064],[222,0.0121],[223,0.0201],[224,0.029],[225,0.0402],[226,0.053],[227,0.0672],[228,0.0832],[229,0.1007],[230,0.1198],[231,0.1409],[232,0.1624],[233,0.1865],[235,0.2328],[236,0.2577],[237,0.2783],[238,0.3047],[239,0.3236],[240,0.3443],[241,0.3672],[242,0.3855],[243,0.4059],[244,0.4214],[245,0.4385],[246,0.4503],[248,0.478],[249,0.4883],[251,0.5027],[252,0.5112],[253,0.5186],[255,0.5279],[257,0.5342],[260,0.5364],[262,0.535],[266,0.5202],[269,0.498],[270,0.4884],[271,0.477],[273,0.4495],[274,0.4323],[275,0.412],[276,0.393],[277,0.3731],[278,0.3512],[284,0.2237],[286,0.1826],[287,0.1614],[288,0.1425],[289,0.1278],[290,0.1118],[291,0.0995],[292,0.0882],[293,0.0772],[294,0.0685],[295,0.059],[296,0.0526],[297,0.0452],[298,0.0406],[299,0.0346],[300,0.0299],[323,-0.0003],[351,-0.0003]]
       }
       spine_02_ctrl.Location.Z {
        type=MovieSceneScriptingFloatChannel
        default=0
        thinned_from=171
        keys [[0,0],[4,0],[5,0.0149],[8,0.0605],[9,0.0724],[10,0.0838],[14,0.1461],[15,0.1584],[16,0.1702],[18,0.2018],[19,0.2101],[21,0.2417],[22,0.251],[23,0.2652],[24,0.2755],[25,0.2895],[26,0.2947],[27,0.3089],[28,0.3139],[29,0.3279],[30,0.3329],[31,0.3468],[32,0.3517],[33,0.3655],[34,0.3705],[35,0.3707],[36,0.3754],[37,0.3844],[38,0.3845],[39,0.3894],[40,0.3905],[41,0.3906],[42,0.393],[44,0.393],[45,0.3907],[46,0.3907],[47,0.3864],[49,0.3818],[50,0.3751],[51,0.3727],[52,0.3661],[53,0.3636],[54,0.3546],[55,0.3479],[56,0.3388],[57,0.3274],[58,0.3203],[59,0.3035],[60,0.2909],[61,0.2817],[64,0.2304],[65,0.2176],[71,0.1138],[72,0.1042],[73,0.0867],[74,0.0737],[75,0.0565],[76,0.048],[77,0.0333],[78,0.0224],[79,0.0145],[80,0.0036],[81,-0.0006],[82,-0.015],[83,-0.019],[84,-0.0297],[85,-0.03],[86,-0.0407],[87,-0.0446],[109,0],[218,0],[219,0.0149],[222,0.0605],[223,0.0724],[224,0.0838],[228,0.1461],[229,0.1584],[230,0.1702],[232,0.2018],[233,0.2101],[235,0.2417],[236,0.251],[237,0.2652],[238,0.2755],[239,0.2895],[240,0.2947],[241,0.3089],[242,0.3139],[243,0.3279],[244,0.3329],[245,0.3468],[246,0.3517],[247,0.3655],[248,0.3705],[249,0.3707],[250,0.3754],[251,0.3844],[252,0.3845],[253,0.3894],[254,0.3905],[255,0.3906],[256,0.393],[258,0.393],[259,0.3907],[260,0.3907],[261,0.3864],[263,0.3818],[264,0.3751],[265,0.3727],[266,0.3661],[267,0.3636],[268,0.3546],[269,0.3479],[270,0.3388],[271,0.3274],[272,0.3203],[273,0.3035],[274,0.2909],[275,0.2817],[278,0.2304],[279,0.2176],[285,0.1138],[286,0.1042],[287,0.0867],[288,0.0737],[289,0.0565],[290,0.048],[291,0.0333],[292,0.0224],[293,0.0145],[294,0.0036],[295,-0.0006],[296,-0.015],[297,-0.019],[298,-0.0297],[299,-0.03],[300,-0.0407],[323,-0.0044],[351,-0.0044]]
       }
       spine_02_ctrl.Rotation.X {
        type=MovieSceneScriptingFloatChannel
        default=0
        thinned_from=171
        keys [[0,0],[4,0],[5,-0.002],[6,0.0155],[7,0.0277],[8,0.0436],[9,0.0631],[10,0.0844],[11,0.1103],[12,0.1383],[13,0.1705],[14,0.2051],[15,0.2438],[16,0.2847],[17,0.3299],[18,0.3762],[19,0.4257],[21,0.5225],[22,0.5725],[24,0.6674],[27,0.8025],[29,0.8854],[30,0.9242],[32,0.9959],[33,1.0278],[35,1.0844],[38,1.1582],[39,1.1803],[40,1.1975],[42,1.2229],[43,1.2328],[45,1.2448],[47,1.2508],[49,1.2489],[50,1.2452],[52,1.233],[55,1.1948],[56,1.1766],[57,1.155],[58,1.1278],[59,1.0936],[60,1.0547],[61,1.0107],[62,0.9644],[63,0.9159],[65,0.8153],[66,0.7629],[67,0.7125],[70,0.5544],[71,0.5009],[73,0.3967],[74,0.3456],[75,0.3012],[76,0.2616],[77,0.226],[78,0.1957],[79,0.168],[80,0.143],[81,0.1209],[82,0.1018],[83,0.0847],[84,0.0685],[85,0.0547],[86,0.0411],[87,0.0315],[109,0],[218,0],[219,-0.002],[220,0.0155],[221,0.0277],[222,0.0436],[223,0.0631],[224,0.0844],[225,0.1103],[226,0.1383],[227,0.1705],[228,0.2051],[229,0.2438],[230,0.2847],[231,0.3299],[232,0.3762],[233,0.4257],[235,0.5225],[236,0.5725],[238,0.6674],[241,0.8025],[243,0.8854],[244,0.9242],[246,0.9959],[247,1.0278],[249,1.0844],[252,1.1582],[253,1.1803],[254,1.1975],[256,1.2229],[257,1.2328],[259,1.2448],[261,1.2508],[263,1.2489],[264,1.2452],[266,1.233],[269,1.1948],[270,1.1766],[271,1.155],[272,1.1278],[273,1.0936],[274,1.0547],[275,1.0107],[276,0.9644],[277,0.9159],[279,0.8153],[280,0.7629],[281,0.7125],[284,0.5544],[285,0.5009],[287,0.3967],[288,0.3456],[289,0.3012],[290,0.2616],[291,0.226],[292,0.1957],[293,0.168],[294,0.143],[295,0.1209],[296,0.1018],[297,0.0847],[298,0.0685],[299,0.0547],[300,0.0411],[323,-0.0014],[351,-0.0014]]
       }
       spine_02_ctrl.Rotation.Y {type=MovieSceneScriptingFloatChannel, default=0, const [0,171]}
       spine_02_ctrl.Rotation.Z {type=MovieSceneScriptingFloatChannel, default=0, const [0,171]}
       spine_02_ctrl.Scale.X {type=MovieSceneScriptingFloatChannel, default=0, const [0,171]}
       spine_02_ctrl.Scale.Y {type=MovieSceneScriptingFloatChannel, default=0, const [0,171]}
       spine_02_ctrl.Scale.Z {type=MovieSceneScriptingFloatChannel, default=0, const [0,171]}
       spine_03_ctrl.Location.X {type=MovieSceneScriptingFloatChannel, default=0, const [0,171]}
       spine_03_ctrl.Location.Y {
        type=MovieSceneScriptingFloatChannel
        default=0
        thinned_from=171
        keys [[0,0],[4,0],[5,-0.0028],[6,-0.0075],[7,-0.0138],[8,-0.0257],[9,-0.0357],[10,-0.0502],[11,-0.0637],[12,-0.0828],[13,-0.1037],[14,-0.1264],[15,-0.1512],[16,-0.178],[17,-0.2104],[18,-0.2394],[19,-0.2744],[20,-0.3063],[21,-0.3421],[22,-0.3735],[23,-0.4066],[25,-0.4709],[26,-0.5018],[27,-0.5271],[28,-0.5564],[30,-0.6017],[31,-0.6229],[33,-0.6565],[35,-0.6804],[37,-0.7006],[38,-0.704],[39,-0.7121],[41,-0.7197],[44,-0.7212],[46,-0.7195],[49,-0.7052],[50,-0.699],[53,-0.6708],[54,-0.657],[55,-0.6449],[56,-0.6275],[57,-0.6116],[58,-0.5936],[60,-0.5481],[61,-0.524],[63,-0.4697],[70,-0.2742],[71,-0.2493],[72,-0.2251],[73,-0.199],[74,-0.1789],[75,-0.1596],[76,-0.1431],[77,-0.1263],[78,-0.113],[79,-0.0993],[80,-0.0876],[81,-0.0779],[82,-0.0679],[83,-0.0591],[84,-0.0507],[85,-0.0444],[86,-0.038],[87,-0.0319],[109,0],[218,0],[219,-0.0028],[220,-0.0075],[221,-0.0138],[222,-0.0257],[223,-0.0357],[224,-0.0502],[225,-0.0637],[226,-0.0828],[227,-0.1037],[228,-0.1264],[229,-0.1512],[230,-0.178],[231,-0.2104],[232,-0.2394],[233,-0.2744],[234,-0.3063],[235,-0.3421],[236,-0.3735],[237,-0.4066],[239,-0.4709],[240,-0.5018],[241,-0.5271],[242,-0.5564],[244,-0.6017],[245,-0.6229],[247,-0.6565],[249,-0.6804],[251,-0.7006],[252,-0.704],[253,-0.7121],[255,-0.7197],[258,-0.7212],[260,-0.7195],[263,-0.7052],[264,-0.699],[267,-0.6708],[268,-0.657],[269,-0.6449],[270,-0.6275],[271,-0.6116],[272,-0.5936],[274,-0.5481],[275,-0.524],[277,-0.4697],[284,-0.2742],[285,-0.2493],[286,-0.2251],[287,-0.199],[288,-0.1789],[289,-0.1596],[290,-0.1431],[291,-0.1263],[292,-0.113],[293,-0.0993],[294,-0.0876],[295,-0.0779],[296,-0.0679],[297,-0.0591],[298,-0.0507],[299,-0.0444],[300,-0.038],[323,0.0002],[351,0.0002]]
       }
       spine_03_ctrl.Location.Z {
        type=MovieSceneScriptingFloatChannel
        default=0
        thinned_from=171
        keys [[0,0],[4,0],[7,0.0595],[8,0.0787],[9,0.0982],[10,0.1075],[11,0.1264],[12,0.1447],[13,0.1626],[14,0.1801],[16,0.2142],[17,0.2302],[18,0.2367],[20,0.2674],[21,0.2822],[22,0.2922],[23,0.3061],[24,0.3141],[25,0.328],[26,0.3361],[27,0.3506],[31,0.3858],[32,0.3954],[33,0.4019],[34,0.4092],[37,0.4228],[38,0.4281],[39,0.4303],[40,0.434],[41,0.4336],[42,0.4366],[44,0.4365],[45,0.4335],[46,0.4336],[47,0.4284],[49,0.4233],[50,0.4152],[51,0.4131],[52,0.4053],[53,0.4034],[57,0.3622],[58,0.3508],[59,0.3343],[60,0.3229],[61,0.3065],[62,0.2875],[63,0.2714],[64,0.2524],[65,0.2384],[70,0.1426],[71,0.123],[72,0.1064],[73,0.0868],[74,0.0718],[75,0.0546],[76,0.042],[77,0.0277],[78,0.0131],[79,0.0028],[80,-0.0078],[81,-0.0161],[82,-0.031],[84,-0.0478],[85,-0.0497],[86,-0.0583],[87,-0.0644],[109,-0.0001],[218,0],[221,0.0595],[222,0.0787],[223,0.0982],[224,0.1075],[225,0.1264],[226,0.1447],[227,0.1626],[228,0.1801],[230,0.2142],[231,0.2302],[232,0.2367],[234,0.2674],[235,0.2822],[236,0.2922],[237,0.3061],[238,0.3141],[239,0.328],[240,0.3361],[241,0.3506],[245,0.3858],[246,0.3954],[247,0.4019],[248,0.4092],[251,0.4228],[252,0.4281],[253,0.4303],[254,0.434],[255,0.4336],[256,0.4366],[258,0.4365],[259,0.4335],[260,0.4336],[261,0.4284],[263,0.4233],[264,0.4152],[265,0.4131],[266,0.4053],[267,0.4034],[271,0.3622],[272,0.3508],[273,0.3343],[274,0.3229],[275,0.3065],[276,0.2875],[277,0.2714],[278,0.2524],[279,0.2384],[284,0.1426],[285,0.123],[286,0.1064],[287,0.0868],[288,0.0718],[289,0.0546],[290,0.042],[291,0.0277],[292,0.0131],[293,0.0028],[294,-0.0078],[295,-0.0161],[296,-0.031],[298,-0.0478],[299,-0.0497],[300,-0.0583],[323,-0.0064],[351,-0.0064]]
       }
       spine_03_ctrl.Rotation.X {
        type=MovieSceneScriptingFloatChannel
        default=0
        thinned_from=171
        keys [[0,0],[4,0],[5,-0.0094],[6,-0.0284],[7,-0.0573],[8,-0.0994],[9,-0.1564],[10,-0.2299],[11,-0.3207],[12,-0.4321],[13,-0.565],[14,-0.7219],[15,-0.906],[16,-1.1206],[17,-1.3625],[18,-1.6275],[19,-1.9106],[22,-2.7961],[23,-3.0851],[24,-3.3654],[25,-3.6341],[26,-3.8883],[27,-4.1239],[28,-4.3417],[29,-4.5431],[30,-4.7271],[31,-4.8946],[32,-5.0449],[33,-5.1805],[34,-5.3009],[35,-5.4075],[36,-5.5004],[37,-5.5819],[38,-5.6516],[39,-5.7083],[40,-5.7491],[42,-5.7942],[44,-5.7984],[46,-5.7672],[48,-5.7027],[50,-5.6048],[52,-5.4715],[54,-5.2973],[55,-5.1918],[56,-5.0722],[57,-4.9347],[58,-4.7748],[59,-4.5921],[60,-4.3928],[61,-4.1805],[63,-3.7284],[67,-2.784],[68,-2.5534],[69,-2.3293],[70,-2.1145],[71,-1.9103],[72,-1.7193],[73,-1.5441],[74,-1.3867],[75,-1.2472],[76,-1.1219],[77,-1.0073],[78,-0.9037],[79,-0.8083],[80,-0.723],[81,-0.6433],[82,-0.571],[83,-0.5055],[84,-0.4461],[85,-0.3909],[86,-0.3411],[87,-0.2961],[109,0],[218,0],[219,-0.0094],[220,-0.0284],[221,-0.0573],[222,-0.0994],[223,-0.1564],[224,-0.2299],[225,-0.3207],[226,-0.4321],[227,-0.565],[228,-0.7219],[229,-0.906],[230,-1.1206],[231,-1.3625],[232,-1.6275],[233,-1.9106],[236,-2.7961],[237,-3.0851],[238,-3.3654],[239,-3.6341],[240,-3.8883],[241,-4.1239],[242,-4.3417],[243,-4.5431],[244,-4.7271],[245,-4.8946],[246,-5.0449],[247,-5.1805],[248,-5.3009],[249,-5.4075],[250,-5.5004],[251,-5.5819],[252,-5.6516],[253,-5.7083],[254,-5.7491],[256,-5.7942],[258,-5.7984],[260,-5.7672],[262,-5.7027],[264,-5.6048],[266,-5.4715],[268,-5.2973],[269,-5.1918],[270,-5.0722],[271,-4.9347],[272,-4.7748],[273,-4.5921],[274,-4.3928],[275,-4.1805],[277,-3.7284],[281,-2.784],[282,-2.5534],[283,-2.3293],[284,-2.1145],[285,-1.9103],[286,-1.7193],[287,-1.5441],[288,-1.3867],[289,-1.2472],[290,-1.1219],[291,-1.0073],[292,-0.9037],[293,-0.8083],[294,-0.723],[295,-0.6433],[296,-0.571],[297,-0.5055],[298,-0.4461],[299,-0.3909],[300,-0.3411],[323,0.002],[351,0.002]]
       }
       spine_03_ctrl.Rotation.Y {type=MovieSceneScriptingFloatChannel, default=0, const [0,171]}
       spine_03_ctrl.Rotation.Z {type=MovieSceneScriptingFloatChannel, default=0, const [0,171]}
       spine_03_ctrl.Scale.X {type=MovieSceneScriptingFloatChannel, default=0, const [0,171]}
       spine_03_ctrl.Scale.Y {type=MovieSceneScriptingFloatChannel, default=0, const [0,171]}
       spine_03_ctrl.Scale.Z {type=MovieSceneScriptingFloatChannel, default=0, const [0,171]}
       head_ctrl.Location.X {type=MovieSceneScriptingFloatChannel, default=0, thinned_from=171, keys [[0,0],[351,0]]}
       head_ctrl.Location.Y {type=MovieSceneScriptingFloatChannel, default=0, const [0,171]}
       head_ctrl.Location.Z {type=MovieSceneScriptingFloatChannel, default=0, const [0,171]}
       head_ctrl.Rotation.X {
        type=MovieSceneScriptingFloatChannel
        default=0
        thinned_from=171
        keys [[0,0],[5,0],[6,0.0209],[7,0.0284],[8,0.0383],[9,0.0457],[12,0.0606],[15,0.0681],[16,0.0681],[18,0.0631],[20,0.0532],[21,0.0433],[22,0.0295],[23,0],[24,-0.0172],[25,-0.0494],[26,-0.0848],[27,-0.1251],[28,-0.1685],[29,-0.2152],[30,-0.2651],[31,-0.3166],[32,-0.3713],[33,-0.4277],[35,-0.5436],[38,-0.7223],[40,-0.8384],[42,-0.9505],[43,-1.0042],[44,-1.0555],[46,-1.1521],[48,-1.2368],[50,-1.3083],[51,-1.3394],[52,-1.3668],[53,-1.3895],[54,-1.4086],[55,-1.4229],[56,-1.4336],[57,-1.4408],[58,-1.442],[60,-1.4364],[62,-1.4252],[64,-1.4028],[66,-1.3693],[67,-1.3497],[68,-1.3273],[70,-1.2714],[71,-1.2406],[72,-1.2071],[73,-1.1707],[74,-1.1316],[76,-1.0473],[78,-0.9585],[80,-0.8656],[81,-0.8171],[85,-0.6314],[87,-0.5426],[109,-0.0024],[219,0],[220,0.0209],[221,0.0284],[222,0.0383],[223,0.0457],[226,0.0606],[229,0.0681],[230,0.0681],[232,0.0631],[234,0.0532],[235,0.0433],[236,0.0295],[237,0],[238,-0.0172],[239,-0.0494],[240,-0.0848],[241,-0.1251],[242,-0.1685],[243,-0.2152],[244,-0.2651],[245,-0.3166],[246,-0.3713],[247,-0.4277],[249,-0.5436],[252,-0.7223],[254,-0.8384],[256,-0.9505],[257,-1.0042],[258,-1.0555],[260,-1.1521],[262,-1.2368],[264,-1.3083],[265,-1.3394],[266,-1.3668],[267,-1.3895],[268,-1.4086],[269,-1.4229],[270,-1.4336],[271,-1.4408],[272,-1.442],[274,-1.4364],[276,-1.4252],[278,-1.4028],[280,-1.3693],[281,-1.3497],[282,-1.3273],[284,-1.2714],[285,-1.2406],[286,-1.2071],[287,-1.1707],[288,-1.1316],[290,-1.0473],[292,-0.9585],[294,-0.8656],[295,-0.8171],[299,-0.6314],[300,-0.587],[323,-0.5425],[351,-0.5425]]
       }
       head_ctrl.Rotation.Y {type=MovieSceneScriptingFloatChannel, default=0, const [0,171]}
       head_ctrl.Rotation.Z {type=MovieSceneScriptingFloatChannel, default=0, const [0,171]}
       head_ctrl.Scale.X {type=MovieSceneScriptingFloatChannel, default=0, const [0,171]}
       head_ctrl.Scale.Y {type=MovieSceneScriptingFloatChannel, default=0, const [0,171]}
       head_ctrl.Scale.Z {type=MovieSceneScriptingFloatChannel, default=0, const [0,171]}
       clavicle_l_ctrl.Location.X {type=MovieSceneScriptingFloatChannel, default=0, const [0,171]}
       clavicle_l_ctrl.Location.Y {type=MovieSceneScriptingFloatChannel, default=0, const [0,171]}
       clavicle_l_ctrl.Location.Z {type=MovieSceneScriptingFloatChannel, default=0, const [0,171]}
       clavicle_l_ctrl.Rotation.X {
        type=MovieSceneScriptingFloatChannel
        default=0
        thinned_from=171
        keys [[0,0],[5,0],[6,-0.003],[7,-0.0066],[8,-0.0111],[9,-0.0167],[10,-0.023],[11,-0.0297],[12,-0.037],[16,-0.0672],[17,-0.0743],[18,-0.081],[19,-0.0873],[20,-0.093],[21,-0.0982],[22,-0.1025],[23,-0.1063],[24,-0.1093],[25,-0.1113],[26,-0.112],[27,-0.112],[28,-0.1113],[33,-0.1059],[35,-0.1042],[37,-0.1021],[40,-0.0985],[42,-0.0958],[44,-0.0929],[45,-0.091],[46,-0.0895],[49,-0.0837],[51,-0.079],[52,-0.0764],[53,-0.0734],[54,-0.07],[55,-0.0662],[56,-0.0619],[57,-0.0569],[58,-0.0513],[59,-0.0455],[60,-0.0394],[61,-0.0338],[62,-0.0286],[63,-0.0245],[64,-0.0216],[65,-0.0196],[66,-0.0184],[67,-0.0174],[68,-0.0168],[69,-0.0165],[70,-0.0165],[72,-0.0174],[73,-0.0176],[75,-0.0177],[76,-0.0174],[78,-0.016],[79,-0.0148],[81,-0.0129],[82,-0.0116],[87,-0.0075],[109,0],[219,0],[220,-0.003],[221,-0.0066],[222,-0.0111],[223,-0.0167],[224,-0.023],[225,-0.0297],[226,-0.037],[230,-0.0672],[231,-0.0743],[232,-0.081],[233,-0.0873],[234,-0.093],[235,-0.0982],[236,-0.1025],[237,-0.1063],[238,-0.1093],[239,-0.1113],[240,-0.112],[241,-0.112],[242,-0.1113],[247,-0.1059],[249,-0.1042],[251,-0.1021],[254,-0.0985],[256,-0.0958],[258,-0.0929],[259,-0.091],[260,-0.0895],[263,-0.0837],[265,-0.079],[266,-0.0764],[267,-0.0734],[268,-0.07],[269,-0.0662],[270,-0.0619],[271,-0.0569],[272,-0.0513],[273,-0.0455],[274,-0.0394],[275,-0.0338],[276,-0.0286],[277,-0.0245],[278,-0.0216],[279,-0.0196],[280,-0.0184],[281,-0.0174],[282,-0.0168],[283,-0.0165],[284,-0.0165],[286,-0.0174],[287,-0.0176],[289,-0.0177],[290,-0.0174],[292,-0.016],[293,-0.0148],[295,-0.0129],[296,-0.0116],[300,-0.0083],[323,0],[351,0]]
       }
       clavicle_l_ctrl.Rotation.Y {
        type=MovieSceneScriptingFloatChannel
        default=0
        thinned_from=171
        keys [[0,0],[5,0],[6,0.015],[7,0.0401],[8,0.076],[9,0.1228],[10,0.1791],[11,0.2452],[12,0.3208],[13,0.4077],[14,0.5014],[15,0.6052],[16,0.7159],[17,0.8366],[18,0.9633],[19,1.0979],[20,1.2401],[21,1.3873],[23,1.6915],[24,1.8371],[25,1.9668],[26,2.0917],[27,2.2075],[28,2.3158],[29,2.4149],[30,2.5037],[31,2.5808],[32,2.6447],[33,2.6982],[34,2.7371],[37,2.8223],[40,2.885],[44,2.94],[48,2.9703],[58,2.9842],[59,2.9802],[60,2.9697],[61,2.9525],[62,2.9216],[63,2.8753],[64,2.8117],[65,2.7255],[66,2.6218],[67,2.5026],[68,2.3745],[69,2.2359],[71,1.9404],[74,1.4771],[75,1.3258],[76,1.1781],[77,1.0341],[78,0.9067],[79,0.7915],[80,0.6975],[81,0.612],[82,0.5352],[83,0.4711],[84,0.4114],[85,0.3559],[86,0.3091],[87,0.2665],[109,0],[219,0],[220,0.015],[221,0.0401],[222,0.076],[223,0.1228],[224,0.1791],[225,0.2452],[226,0.3208],[227,0.4077],[228,0.5014],[229,0.6052],[230,0.7159],[231,0.8366],[232,0.9633],[233,1.0979],[234,1.2401],[235,1.3873],[237,1.6915],[238,1.8371],[239,1.9668],[240,2.0917],[241,2.2075],[242,2.3158],[243,2.4149],[244,2.5037],[245,2.5808],[246,2.6447],[247,2.6982],[248,2.7371],[251,2.8223],[254,2.885],[258,2.94],[262,2.9703],[272,2.9842],[273,2.9802],[274,2.9697],[275,2.9525],[276,2.9216],[277,2.8753],[278,2.8117],[279,2.7255],[280,2.6218],[281,2.5026],[282,2.3745],[283,2.2359],[285,1.9404],[288,1.4771],[289,1.3258],[290,1.1781],[291,1.0341],[292,0.9067],[293,0.7915],[294,0.6975],[295,0.612],[296,0.5352],[297,0.4711],[298,0.4114],[299,0.3559],[300,0.3091],[323,0],[351,0]]
       }
       clavicle_l_ctrl.Rotation.Z {
        type=MovieSceneScriptingFloatChannel
        default=0
        thinned_from=171
        keys [[0,0],[5,0],[6,-0.0271],[7,-0.0611],[8,-0.1074],[9,-0.1649],[10,-0.2312],[11,-0.3051],[12,-0.3867],[13,-0.476],[14,-0.5705],[15,-0.6693],[16,-0.7732],[17,-0.8803],[18,-0.9856],[19,-1.0858],[20,-1.181],[21,-1.2707],[22,-1.3547],[23,-1.433],[24,-1.505],[25,-1.5705],[26,-1.6299],[27,-1.6831],[28,-1.7293],[29,-1.7694],[30,-1.8032],[31,-1.8305],[32,-1.8513],[33,-1.8658],[36,-1.8835],[39,-1.8904],[42,-1.8861],[44,-1.8732],[46,-1.8509],[48,-1.8193],[50,-1.7753],[52,-1.7191],[54,-1.651],[56,-1.5707],[58,-1.4788],[59,-1.4237],[60,-1.3626],[61,-1.2967],[62,-1.2265],[63,-1.1518],[64,-1.0746],[66,-0.9128],[68,-0.7473],[69,-0.6661],[70,-0.5874],[71,-0.5118],[72,-0.4401],[73,-0.3727],[74,-0.3116],[75,-0.2584],[76,-0.2126],[77,-0.1753],[78,-0.147],[79,-0.1291],[80,-0.1165],[81,-0.105],[82,-0.0939],[83,-0.0841],[84,-0.0748],[85,-0.0662],[86,-0.0583],[87,-0.0508],[109,0],[219,0],[220,-0.0271],[221,-0.0611],[222,-0.1074],[223,-0.1649],[224,-0.2312],[225,-0.3051],[226,-0.3867],[227,-0.476],[228,-0.5705],[229,-0.6693],[230,-0.7732],[231,-0.8803],[232,-0.9856],[233,-1.0858],[234,-1.181],[235,-1.2707],[236,-1.3547],[237,-1.433],[238,-1.505],[239,-1.5705],[240,-1.6299],[241,-1.6831],[242,-1.7293],[243,-1.7694],[244,-1.8032],[245,-1.8305],[246,-1.8513],[247,-1.8658],[250,-1.8835],[253,-1.8904],[256,-1.8861],[258,-1.8732],[260,-1.8509],[262,-1.8193],[264,-1.7753],[266,-1.7191],[268,-1.651],[270,-1.5707],[272,-1.4788],[273,-1.4237],[274,-1.3626],[275,-1.2967],[276,-1.2265],[277,-1.1518],[278,-1.0746],[280,-0.9128],[282,-0.7473],[283,-0.6661],[284,-0.5874],[285,-0.5118],[286,-0.4401],[287,-0.3727],[288,-0.3116],[289,-0.2584],[290,-0.2126],[291,-0.1753],[292,-0.147],[293,-0.1291],[294,-0.1165],[295,-0.105],[296,-0.0939],[297,-0.0841],[298,-0.0748],[299,-0.0662],[300,-0.0583],[323,0],[351,0]]
       }
       clavicle_l_ctrl.Scale.X {type=MovieSceneScriptingFloatChannel, default=0, const [0,171]}
       clavicle_l_ctrl.Scale.Y {type=MovieSceneScriptingFloatChannel, default=0, const [0,171]}
       clavicle_l_ctrl.Scale.Z {type=MovieSceneScriptingFloatChannel, default=0, const [0,171]}
       clavicle_r_ctrl.Location.X {type=MovieSceneScriptingFloatChannel, default=0, const [0,171]}
       clavicle_r_ctrl.Location.Y {type=MovieSceneScriptingFloatChannel, default=0, const [0,171]}
       clavicle_r_ctrl.Location.Z {type=MovieSceneScriptingFloatChannel, default=0, const [0,171]}
       clavicle_r_ctrl.Rotation.X {
        type=MovieSceneScriptingFloatChannel
        default=0
        thinned_from=171
        keys [[0,0],[5,0],[6,-0.0033],[7,-0.0074],[8,-0.0127],[9,-0.0192],[10,-0.0263],[11,-0.0344],[13,-0.0516],[14,-0.0607],[16,-0.0781],[17,-0.0863],[18,-0.0936],[19,-0.1003],[20,-0.106],[21,-0.1105],[22,-0.1138],[23,-0.1155],[24,-0.1159],[25,-0.1151],[26,-0.1138],[27,-0.1118],[29,-0.1058],[33,-0.091],[34,-0.0876],[35,-0.0846],[37,-0.0795],[38,-0.0773],[40,-0.0741],[42,-0.0719],[43,-0.0709],[46,-0.0692],[49,-0.0683],[50,-0.0684],[53,-0.0673],[55,-0.0655],[56,-0.064],[57,-0.0621],[58,-0.0588],[59,-0.054],[60,-0.0476],[61,-0.0399],[62,-0.0316],[63,-0.0235],[64,-0.0165],[65,-0.0114],[66,-0.0077],[67,-0.0053],[68,-0.0036],[69,-0.0025],[71,-0.0016],[73,-0.0016],[76,-0.0026],[77,-0.0028],[79,-0.0036],[82,-0.0039],[85,-0.0033],[87,-0.0028],[109,0],[219,0],[220,-0.0033],[221,-0.0074],[222,-0.0127],[223,-0.0192],[224,-0.0263],[225,-0.0344],[227,-0.0516],[228,-0.0607],[230,-0.0781],[231,-0.0863],[232,-0.0936],[233,-0.1003],[234,-0.106],[235,-0.1105],[236,-0.1138],[237,-0.1155],[238,-0.1159],[239,-0.1151],[240,-0.1138],[241,-0.1118],[243,-0.1058],[247,-0.091],[248,-0.0876],[249,-0.0846],[251,-0.0795],[252,-0.0773],[254,-0.0741],[256,-0.0719],[257,-0.0709],[260,-0.0692],[263,-0.0683],[264,-0.0684],[267,-0.0673],[269,-0.0655],[270,-0.064],[271,-0.0621],[272,-0.0588],[273,-0.054],[274,-0.0476],[275,-0.0399],[276,-0.0316],[277,-0.0235],[278,-0.0165],[279,-0.0114],[280,-0.0077],[281,-0.0053],[282,-0.0036],[283,-0.0025],[285,-0.0016],[287,-0.0016],[290,-0.0026],[291,-0.0028],[293,-0.0036],[296,-0.0039],[299,-0.0033],[300,-0.003],[323,0],[351,0]]
       }
       clavicle_r_ctrl.Rotation.Y {
        type=MovieSceneScriptingFloatChannel
        default=0
        thinned_from=171
        keys [[0,0],[5,0],[6,0.0429],[7,0.0905],[8,0.1505],[9,0.2222],[10,0.3041],[11,0.3937],[12,0.491],[13,0.5954],[14,0.7059],[15,0.8219],[16,0.9425],[17,1.0678],[18,1.197],[19,1.3303],[20,1.4666],[24,2.0293],[25,2.1589],[26,2.2787],[27,2.3862],[28,2.484],[29,2.5694],[30,2.6453],[31,2.7091],[32,2.7608],[33,2.7982],[34,2.8263],[37,2.882],[43,2.9499],[58,2.9838],[59,2.9766],[60,2.9625],[61,2.9397],[62,2.903],[63,2.8522],[64,2.7837],[65,2.6941],[66,2.5868],[67,2.4687],[68,2.3418],[69,2.2077],[71,1.9214],[72,1.7744],[74,1.4729],[76,1.1789],[77,1.0381],[78,0.9053],[79,0.7844],[80,0.678],[81,0.585],[82,0.5035],[83,0.4313],[84,0.3685],[85,0.3134],[86,0.2648],[87,0.2233],[109,0],[219,0],[220,0.0429],[221,0.0905],[222,0.1505],[223,0.2222],[224,0.3041],[225,0.3937],[226,0.491],[227,0.5954],[228,0.7059],[229,0.8219],[230,0.9425],[231,1.0678],[232,1.197],[233,1.3303],[234,1.4666],[238,2.0293],[239,2.1589],[240,2.2787],[241,2.3862],[242,2.484],[243,2.5694],[244,2.6453],[245,2.7091],[246,2.7608],[247,2.7982],[248,2.8263],[251,2.882],[257,2.9499],[272,2.9838],[273,2.9766],[274,2.9625],[275,2.9397],[276,2.903],[277,2.8522],[278,2.7837],[279,2.6941],[280,2.5868],[281,2.4687],[282,2.3418],[283,2.2077],[285,1.9214],[286,1.7744],[288,1.4729],[290,1.1789],[291,1.0381],[292,0.9053],[293,0.7844],[294,0.678],[295,0.585],[296,0.5035],[297,0.4313],[298,0.3685],[299,0.3134],[300,0.2648],[323,0],[351,0]]
       }
       clavicle_r_ctrl.Rotation.Z {
        type=MovieSceneScriptingFloatChannel
        default=0
        thinned_from=171
        keys [[0,0],[5,0],[6,-0.0226],[7,-0.0498],[8,-0.0859],[9,-0.1309],[10,-0.1836],[11,-0.2435],[12,-0.31],[13,-0.3822],[14,-0.4599],[15,-0.5423],[16,-0.6291],[17,-0.7196],[18,-0.8134],[19,-0.9095],[21,-1.1079],[26,-1.6141],[29,-1.8982],[30,-1.986],[31,-2.0682],[32,-2.145],[33,-2.2138],[34,-2.2727],[35,-2.3242],[36,-2.3644],[37,-2.3997],[39,-2.4473],[41,-2.4681],[42,-2.469],[44,-2.4522],[46,-2.4145],[47,-2.3865],[48,-2.3532],[50,-2.2738],[51,-2.2249],[53,-2.1143],[55,-1.9829],[57,-1.831],[62,-1.4199],[65,-1.1826],[66,-1.1061],[68,-0.9584],[70,-0.818],[71,-0.7515],[72,-0.6872],[73,-0.6257],[74,-0.5664],[75,-0.5108],[76,-0.4596],[77,-0.4123],[78,-0.3699],[79,-0.3333],[80,-0.3012],[81,-0.2711],[82,-0.2431],[83,-0.2171],[84,-0.193],[85,-0.1708],[86,-0.1503],[87,-0.1316],[109,0],[219,0],[220,-0.0226],[221,-0.0498],[222,-0.0859],[223,-0.1309],[224,-0.1836],[225,-0.2435],[226,-0.31],[227,-0.3822],[228,-0.4599],[229,-0.5423],[230,-0.6291],[231,-0.7196],[232,-0.8134],[233,-0.9095],[235,-1.1079],[240,-1.6141],[243,-1.8982],[244,-1.986],[245,-2.0682],[246,-2.145],[247,-2.2138],[248,-2.2727],[249,-2.3242],[250,-2.3644],[251,-2.3997],[253,-2.4473],[255,-2.4681],[256,-2.469],[258,-2.4522],[260,-2.4145],[261,-2.3865],[262,-2.3532],[264,-2.2738],[265,-2.2249],[267,-2.1143],[269,-1.9829],[271,-1.831],[276,-1.4199],[279,-1.1826],[280,-1.1061],[282,-0.9584],[284,-0.818],[285,-0.7515],[286,-0.6872],[287,-0.6257],[288,-0.5664],[289,-0.5108],[290,-0.4596],[291,-0.4123],[292,-0.3699],[293,-0.3333],[294,-0.3012],[295,-0.2711],[296,-0.2431],[297,-0.2171],[298,-0.193],[299,-0.1708],[300,-0.1503],[323,0],[351,0]]
       }
       clavicle_r_ctrl.Scale.X {type=MovieSceneScriptingFloatChannel, default=0, const [0,171]}
       clavicle_r_ctrl.Scale.Y {type=MovieSceneScriptingFloatChannel, default=0, const [0,171]}
       clavicle_r_ctrl.Scale.Z {type=MovieSceneScriptingFloatChannel, default=0, const [0,171]}
       neck_02_ctrl.Location.X {type=MovieSceneScriptingFloatChannel, default=0, const [0,171]}
       neck_02_ctrl.Location.Y {
        type=MovieSceneScriptingFloatChannel
        default=0
        thinned_from=171
        keys [[0,0],[4,0],[5,0.0004],[6,0.0004],[8,0.0013],[9,0.0047],[10,0.0051],[11,0.006],[12,0.0064],[14,0.0081],[15,0.0094],[16,0.0103],[17,0.0086],[18,0.0124],[19,0.0137],[20,0.012],[21,0.0133],[22,0.0133],[24,0.0119],[25,0.0139],[27,0.0125],[28,0.012],[29,0.0113],[31,0.0104],[32,0.0074],[33,0.0069],[34,0.0011],[35,0.0037],[36,0.0037],[37,0.0062],[38,0.0034],[39,0.006],[40,0.0067],[44,0.0067],[45,0.0115],[46,0.0063],[47,0.0063],[48,0.011],[49,0.0058],[50,0.0105],[51,0.0053],[52,0.01],[53,0.0048],[54,0.0043],[55,0.0091],[56,0.0038],[57,0.0034],[58,0.0015],[60,-0.0005],[61,-0.0003],[62,-0.0028],[63,-0.0038],[64,-0.0036],[65,-0.0061],[67,-0.0088],[68,-0.0098],[75,-0.0191],[76,-0.019],[77,-0.0221],[78,-0.0178],[79,-0.021],[80,-0.0206],[81,-0.0238],[82,-0.0195],[86,-0.018],[87,-0.0142],[109,0],[218,0],[219,0.0004],[220,0.0004],[222,0.0013],[223,0.0047],[224,0.0051],[225,0.006],[226,0.0064],[228,0.0081],[229,0.0094],[230,0.0103],[231,0.0086],[232,0.0124],[233,0.0137],[234,0.012],[235,0.0133],[236,0.0133],[238,0.0119],[239,0.0139],[241,0.0125],[242,0.012],[243,0.0113],[245,0.0104],[246,0.0074],[247,0.0069],[248,0.0011],[249,0.0037],[250,0.0037],[251,0.0062],[252,0.0034],[253,0.006],[254,0.0067],[258,0.0067],[259,0.0115],[260,0.0063],[261,0.0063],[262,0.011],[263,0.0058],[264,0.0105],[265,0.0053],[266,0.01],[267,0.0048],[268,0.0043],[269,0.0091],[270,0.0038],[271,0.0034],[272,0.0015],[274,-0.0005],[275,-0.0003],[276,-0.0028],[277,-0.0038],[278,-0.0036],[279,-0.0061],[281,-0.0088],[282,-0.0098],[289,-0.0191],[290,-0.019],[291,-0.0221],[292,-0.0178],[293,-0.021],[294,-0.0206],[295,-0.0238],[296,-0.0195],[300,-0.018],[323,-0.0029],[351,-0.0029]]
       }
       neck_02_ctrl.Location.Z {
        type=MovieSceneScriptingFloatChannel
        default=0
        thinned_from=171
        keys [[0,0],[4,0],[5,-0.0085],[6,-0.0085],[8,-0.0256],[9,-0.0352],[10,-0.0437],[11,-0.0608],[12,-0.0693],[14,-0.1035],[15,-0.1292],[16,-0.1463],[17,-0.1709],[18,-0.189],[19,-0.2146],[20,-0.2393],[21,-0.2649],[22,-0.2879],[24,-0.3428],[25,-0.3619],[27,-0.4167],[28,-0.435],[29,-0.4624],[31,-0.499],[33,-0.5347],[34,-0.5604],[37,-0.5621],[38,-0.5704],[44,-0.5715],[45,-0.5573],[46,-0.5555],[47,-0.5555],[48,-0.5413],[49,-0.5396],[50,-0.5254],[51,-0.5236],[52,-0.5094],[53,-0.5076],[54,-0.4917],[55,-0.4775],[56,-0.4757],[57,-0.4597],[58,-0.4499],[60,-0.4247],[61,-0.4084],[63,-0.3826],[64,-0.3663],[65,-0.3531],[67,-0.3195],[68,-0.3068],[75,-0.189],[76,-0.1734],[77,-0.1595],[78,-0.1351],[79,-0.1211],[80,-0.1083],[81,-0.0944],[82,-0.07],[86,-0.0189],[87,-0.0072],[109,0],[218,0],[219,-0.0085],[220,-0.0085],[222,-0.0256],[223,-0.0352],[224,-0.0437],[225,-0.0608],[226,-0.0693],[228,-0.1035],[229,-0.1292],[230,-0.1463],[231,-0.1709],[232,-0.189],[233,-0.2146],[234,-0.2393],[235,-0.2649],[236,-0.2879],[238,-0.3428],[239,-0.3619],[241,-0.4167],[242,-0.435],[243,-0.4624],[245,-0.499],[247,-0.5347],[248,-0.5604],[251,-0.5621],[252,-0.5704],[258,-0.5715],[259,-0.5573],[260,-0.5555],[261,-0.5555],[262,-0.5413],[263,-0.5396],[264,-0.5254],[265,-0.5236],[266,-0.5094],[267,-0.5076],[268,-0.4917],[269,-0.4775],[270,-0.4757],[271,-0.4597],[272,-0.4499],[274,-0.4247],[275,-0.4084],[277,-0.3826],[278,-0.3663],[279,-0.3531],[281,-0.3195],[282,-0.3068],[289,-0.189],[290,-0.1734],[291,-0.1595],[292,-0.1351],[293,-0.1211],[294,-0.1083],[295,-0.0944],[296,-0.07],[300,-0.0189],[323,0.0024],[351,0.0024]]
       }
       neck_02_ctrl.Rotation.X {
        type=MovieSceneScriptingFloatChannel
        default=0
        thinned_from=171
        keys [[0,0],[8,0],[9,-0.0013],[15,-0.0013],[16,-0.0026],[19,-0.0026],[20,-0.0038],[26,-0.0038],[27,-0.0093],[33,-0.0093],[34,-0.0147],[37,-0.0147],[38,-0.0201],[39,-0.0201],[40,-0.0213],[41,-0.0213],[42,-0.0259],[43,-0.0259],[44,-0.0306],[45,-0.0306],[46,-0.0352],[48,-0.0352],[49,-0.0398],[50,-0.0398],[51,-0.0444],[52,-0.0444],[53,-0.0491],[55,-0.0491],[56,-0.0537],[57,-0.0537],[58,-0.0542],[59,-0.0566],[60,-0.0566],[61,-0.059],[62,-0.059],[63,-0.0613],[70,-0.0613],[71,-0.059],[72,-0.059],[73,-0.0566],[74,-0.0566],[75,-0.0542],[76,-0.0527],[84,-0.0373],[85,-0.0373],[87,-0.0334],[109,0],[222,0],[223,-0.0013],[229,-0.0013],[230,-0.0026],[233,-0.0026],[234,-0.0038],[240,-0.0038],[241,-0.0093],[247,-0.0093],[248,-0.0147],[251,-0.0147],[252,-0.0201],[253,-0.0201],[254,-0.0213],[255,-0.0213],[256,-0.0259],[257,-0.0259],[258,-0.0306],[259,-0.0306],[260,-0.0352],[262,-0.0352],[263,-0.0398],[264,-0.0398],[265,-0.0444],[266,-0.0444],[267,-0.0491],[269,-0.0491],[270,-0.0537],[271,-0.0537],[272,-0.0542],[273,-0.0566],[274,-0.0566],[275,-0.059],[276,-0.059],[277,-0.0613],[284,-0.0613],[285,-0.059],[286,-0.059],[287,-0.0566],[288,-0.0566],[289,-0.0542],[290,-0.0527],[298,-0.0373],[299,-0.0373],[300,-0.0353],[323,-0.0002],[351,-0.0002]]
       }
       neck_02_ctrl.Rotation.Y {type=MovieSceneScriptingFloatChannel, default=0, const [0,171]}
       neck_02_ctrl.Rotation.Z {type=MovieSceneScriptingFloatChannel, default=0, const [0,171]}
       neck_02_ctrl.Scale.X {type=MovieSceneScriptingFloatChannel, default=0, const [0,171]}
       neck_02_ctrl.Scale.Y {type=MovieSceneScriptingFloatChannel, default=0, const [0,171]}
       neck_02_ctrl.Scale.Z {type=MovieSceneScriptingFloatChannel, default=0, const [0,171]}
       Weight {type=MovieSceneScriptingFloatChannel, default=0.6}
      ]
     }
     {
      type=MovieSceneControlRigParameterSection
      channels [
       body_ctrl.Location.X {type=MovieSceneScriptingFloatChannel, default=0, const [0,2]}
       body_ctrl.Location.Y {type=MovieSceneScriptingFloatChannel, default=0, const [0,2]}
       body_ctrl.Location.Z {type=MovieSceneScriptingFloatChannel, default=0, const [0,2]}
       body_ctrl.Rotation.X {type=MovieSceneScriptingFloatChannel, default=0, const [0,2]}
       body_ctrl.Rotation.Y {type=MovieSceneScriptingFloatChannel, default=0, const [0,2]}
       body_ctrl.Rotation.Z {type=MovieSceneScriptingFloatChannel, default=0, const [0,2]}
       body_ctrl.Scale.X {type=MovieSceneScriptingFloatChannel, default=0, const [0,2]}
       body_ctrl.Scale.Y {type=MovieSceneScriptingFloatChannel, default=0, const [0,2]}
       body_ctrl.Scale.Z {type=MovieSceneScriptingFloatChannel, default=0, const [0,2]}
       Weight {type=MovieSceneScriptingFloatChannel, default=1}
      ]
     }
     {
      type=MovieSceneControlRigParameterSection
      channels [
       foot_l_ik_ctrl.Location.X {type=MovieSceneScriptingFloatChannel, default=0, const [0,5]}
       foot_l_ik_ctrl.Location.Y {type=MovieSceneScriptingFloatChannel, default=0, const [0,5]}
       foot_l_ik_ctrl.Location.Z {type=MovieSceneScriptingFloatChannel, default=0, const [0,5]}
       foot_l_ik_ctrl.Rotation.X {type=MovieSceneScriptingFloatChannel, default=0, keys [[148,0],[159,0],[165,5.5707],[173,0],[198,0]]}
       foot_l_ik_ctrl.Rotation.Y {type=MovieSceneScriptingFloatChannel, default=0, const [0,5]}
       foot_l_ik_ctrl.Rotation.Z {type=MovieSceneScriptingFloatChannel, default=0, const [0,5]}
       foot_l_ik_ctrl.Scale.X {type=MovieSceneScriptingFloatChannel, default=0, const [0,5]}
       foot_l_ik_ctrl.Scale.Y {type=MovieSceneScriptingFloatChannel, default=0, const [0,5]}
       foot_l_ik_ctrl.Scale.Z {type=MovieSceneScriptingFloatChannel, default=0, const [0,5]}
       leg_l_softness {type=MovieSceneScriptingFloatChannel, default=0, const [0,5]}
       foot_r_ik_ctrl.Location.X {type=MovieSceneScriptingFloatChannel, default=0, const [0,5]}
       foot_r_ik_ctrl.Location.Y {type=MovieSceneScriptingFloatChannel, default=0, const [0,5]}
       foot_r_ik_ctrl.Location.Z {type=MovieSceneScriptingFloatChannel, default=0, const [0,5]}
       foot_r_ik_ctrl.Rotation.X {type=MovieSceneScriptingFloatChannel, default=0, keys [[148,0],[183,0],[188,3.3232],[196,0],[198,0]]}
       foot_r_ik_ctrl.Rotation.Y {type=MovieSceneScriptingFloatChannel, default=0, const [0,5]}
       foot_r_ik_ctrl.Rotation.Z {type=MovieSceneScriptingFloatChannel, default=0, const [0,5]}
       foot_r_ik_ctrl.Scale.X {type=MovieSceneScriptingFloatChannel, default=0, const [0,5]}
       foot_r_ik_ctrl.Scale.Y {type=MovieSceneScriptingFloatChannel, default=0, const [0,5]}
       foot_r_ik_ctrl.Scale.Z {type=MovieSceneScriptingFloatChannel, default=0, const [0,5]}
       ball_r_ik_ctrl.Location.X {type=MovieSceneScriptingFloatChannel, default=0, const [0,4]}
       ball_r_ik_ctrl.Location.Y {type=MovieSceneScriptingFloatChannel, default=0, const [0,4]}
       ball_r_ik_ctrl.Location.Z {type=MovieSceneScriptingFloatChannel, default=0, const [0,4]}
       ball_r_ik_ctrl.Rotation.X {type=MovieSceneScriptingFloatChannel, default=0, const [0,4]}
       ball_r_ik_ctrl.Rotation.Y {type=MovieSceneScriptingFloatChannel, default=0, const [0,4]}
       ball_r_ik_ctrl.Rotation.Z {type=MovieSceneScriptingFloatChannel, default=0, const [0,4]}
       ball_r_ik_ctrl.Scale.X {type=MovieSceneScriptingFloatChannel, default=0, const [0,4]}
       ball_r_ik_ctrl.Scale.Y {type=MovieSceneScriptingFloatChannel, default=0, const [0,4]}
       ball_r_ik_ctrl.Scale.Z {type=MovieSceneScriptingFloatChannel, default=0, const [0,4]}
       leg_r_softness {type=MovieSceneScriptingFloatChannel, default=0, const [0,5]}
       Weight {type=MovieSceneScriptingFloatChannel, default=1}
      ]
     }
    ]
   }
   {
    type=MovieSceneControlRigParameterTrack
    display_name=Fortnite_3L_Face_CtrlRig
    sections [{
      type=MovieSceneControlRigParameterSection
      channels [
       CTRL_C_eyesAim.Location.X {type=MovieSceneScriptingFloatChannel, default=0}
       CTRL_C_eyesAim.Location.Y {type=MovieSceneScriptingFloatChannel, default=0}
       CTRL_C_eyesAim.Location.Z {type=MovieSceneScriptingFloatChannel, default=0}
       CTRL_C_eyesAim.Rotation.X {type=MovieSceneScriptingFloatChannel, default=0}
       CTRL_C_eyesAim.Rotation.Y {type=MovieSceneScriptingFloatChannel, default=0}
       CTRL_C_eyesAim.Rotation.Z {type=MovieSceneScriptingFloatChannel, default=0}
       CTRL_C_eyesAim.Scale.X {type=MovieSceneScriptingFloatChannel, default=1}
       CTRL_C_eyesAim.Scale.Y {type=MovieSceneScriptingFloatChannel, default=1}
       CTRL_C_eyesAim.Scale.Z {type=MovieSceneScriptingFloatChannel, default=1}
       CTRL_L_eyeAim.Location.X {type=MovieSceneScriptingFloatChannel, default=0}
       CTRL_L_eyeAim.Location.Y {type=MovieSceneScriptingFloatChannel, default=0}
       CTRL_L_eyeAim.Location.Z {type=MovieSceneScriptingFloatChannel, default=0}
       CTRL_L_eyeAim.Rotation.X {type=MovieSceneScriptingFloatChannel, default=0}
       CTRL_L_eyeAim.Rotation.Y {type=MovieSceneScriptingFloatChannel, default=0}
       CTRL_L_eyeAim.Rotation.Z {type=MovieSceneScriptingFloatChannel, default=0}
       CTRL_L_eyeAim.Scale.X {type=MovieSceneScriptingFloatChannel, default=1}
       CTRL_L_eyeAim.Scale.Y {type=MovieSceneScriptingFloatChannel, default=1}
       CTRL_L_eyeAim.Scale.Z {type=MovieSceneScriptingFloatChannel, default=1}
       CTRL_R_eyeAim.Location.X {type=MovieSceneScriptingFloatChannel, default=0}
       CTRL_R_eyeAim.Location.Y {type=MovieSceneScriptingFloatChannel, default=0}
       CTRL_R_eyeAim.Location.Z {type=MovieSceneScriptingFloatChannel, default=0}
       CTRL_R_eyeAim.Rotation.X {type=MovieSceneScriptingFloatChannel, default=0}
       CTRL_R_eyeAim.Rotation.Y {type=MovieSceneScriptingFloatChannel, default=0}
       CTRL_R_eyeAim.Rotation.Z {type=MovieSceneScriptingFloatChannel, default=0}
       CTRL_R_eyeAim.Scale.X {type=MovieSceneScriptingFloatChannel, default=1}
       CTRL_R_eyeAim.Scale.Y {type=MovieSceneScriptingFloatChannel, default=1}
       CTRL_R_eyeAim.Scale.Z {type=MovieSceneScriptingFloatChannel, default=1}
       CTRL_convergenceSwitch {type=MovieSceneScriptingFloatChannel, default=0}
       CTRL_Master.Location.X {type=MovieSceneScriptingFloatChannel, default=7.8662}
       CTRL_Master.Location.Y {type=MovieSceneScriptingFloatChannel, default=0}
       CTRL_Master.Location.Z {type=MovieSceneScriptingFloatChannel, default=0}
       CTRL_Master.Rotation.X {type=MovieSceneScriptingFloatChannel, default=0}
       CTRL_Master.Rotation.Y {type=MovieSceneScriptingFloatChannel, default=0}
       CTRL_Master.Rotation.Z {type=MovieSceneScriptingFloatChannel, default=0}
       CTRL_Master.Scale.X {type=MovieSceneScriptingFloatChannel, default=2.0763}
       CTRL_Master.Scale.Y {type=MovieSceneScriptingFloatChannel, default=2.0763}
       CTRL_Master.Scale.Z {type=MovieSceneScriptingFloatChannel, default=2.0763}
       CTRL_C_jaw.X {type=MovieSceneScriptingFloatChannel, default=0}
       CTRL_C_jaw.Y {type=MovieSceneScriptingFloatChannel, default=0}
       CTRL_C_jaw_openExtreme {type=MovieSceneScriptingFloatChannel, default=0}
       CTRL_C_jaw_fwdBack {type=MovieSceneScriptingFloatChannel, default=0}
       CTRL_R_jaw_ChinRaiseD {type=MovieSceneScriptingFloatChannel, default=0}
       CTRL_R_jaw_ChinRaiseU {type=MovieSceneScriptingFloatChannel, default=0}
       CTRL_L_jaw_ChinRaiseU {type=MovieSceneScriptingFloatChannel, default=0}
       CTRL_L_jaw_ChinRaiseD {type=MovieSceneScriptingFloatChannel, default=0}
       CTRL_R_mouth_lipsBlow {type=MovieSceneScriptingFloatChannel, default=0}
       CTRL_L_mouth_lipsTogetherU {type=MovieSceneScriptingFloatChannel, default=0}
       CTRL_L_mouth_lipsTogetherD {type=MovieSceneScriptingFloatChannel, default=0}
       CTRL_R_mouth_lipsTogetherU {type=MovieSceneScriptingFloatChannel, default=0}
       CTRL_L_mouth_lipsBlow {type=MovieSceneScriptingFloatChannel, default=0}
       CTRL_R_mouth_lipsTogetherD {type=MovieSceneScriptingFloatChannel, default=0}
       CTRL_C_tongue_roll.X {type=MovieSceneScriptingFloatChannel, default=0}
       CTRL_C_tongue_roll.Y {type=MovieSceneScriptingFloatChannel, default=0}
       CTRL_C_tongue.X {type=MovieSceneScriptingFloatChannel, default=0}
       CTRL_C_tongue.Y {type=MovieSceneScriptingFloatChannel, default=0}
       CTRL_C_tongue_narrowWide {type=MovieSceneScriptingFloatChannel, default=0}
       CTRL_C_tongue_inOut {type=MovieSceneScriptingFloatChannel, default=0}
       CTRL_R_nose.X {type=MovieSceneScriptingFloatChannel, default=0, keys [[0,0],[120,0],[135,0.4511],[152,0.5238],[233,0.5238],[248,0.0023],[352,0]]}
       CTRL_R_nose.Y {type=MovieSceneScriptingFloatChannel, default=0, const [0,7]}
       CTRL_L_nose.X {type=MovieSceneScriptingFloatChannel, default=0, keys [[0,0],[120,0],[135,0.336],[152,0.3939],[233,0.3939],[248,0.0017],[352,0]]}
       CTRL_L_nose.Y {type=MovieSceneScriptingFloatChannel, default=0, const [0,7]}
       CTRL_L_eye.X {type=MovieSceneScriptingFloatChannel, default=0}
       CTRL_L_eye.Y {type=MovieSceneScriptingFloatChannel, default=0}
       CTRL_L_eye_blink {
        type=MovieSceneScriptingFloatChannel
        default=0
        keys [[0,0],[24,0],[26,1],[28,1],[31,0.1053],[34,0],[72,0],[74,1],[77,0.9997],[80,0.0877],[83,0],[122,0],[124,1],[125,1],[128,0.0614],[131,0.0003],[196,0.0001],[199,1],[200,1],[203,0.0965],[206,0.0004],[233,0],[235,1],[243,1],[246,0.0789],[249,0],[293,0],[295,1],[299,1],[302,0.0921],[305,0],[352,0]]
       }
       CTRL_R_eye.X {type=MovieSceneScriptingFloatChannel, default=0}
       CTRL_R_eye.Y {type=MovieSceneScriptingFloatChannel, default=0}
       CTRL_R_eye_blink {
        type=MovieSceneScriptingFloatChannel
        default=0
        keys [[0,0],[24,0],[26,1],[28,1],[31,0.1053],[34,0],[72,0],[74,1],[77,0.9997],[80,0.0877],[83,0],[122,0],[124,1],[125,1],[128,0.0614],[131,0.0003],[196,0.0001],[199,1],[200,1],[203,0.0965],[206,0.0004],[233,0],[235,1],[243,1],[246,0.0789],[249,0],[293,0],[295,1],[299,1],[302,0.0921],[305,0],[352,0]]
       }
       CTRL_C_eye.X {
        type=MovieSceneScriptingFloatChannel
        default=-0.0094
        thinned_from=42
        keys [[0,0.1223],[24,0.1223],[25,0.2474],[47,0.3067],[57,0.3067],[58,0.4792],[72,0.4792],[73,-0.2663],[80,-0.5432],[93,0.1061],[103,0.1475],[104,0.0674],[114,0.0674],[115,0.1584],[120,0.1584],[121,0.1919],[124,0.1919],[127,0.1936],[138,0.4507],[154,0.4488],[168,0.2013],[184,0.044],[210,0.3832],[217,0.3878],[218,0.4629],[233,0.4629],[234,-0.2244],[246,-0.3242],[262,-0.3029],[263,-0.5232],[282,-0.5232],[283,-0.3516],[293,-0.3516],[294,0.1658],[302,0.5697],[312,0.3241],[332,0.1193],[333,0.222],[350,0.222],[351,0.1182],[353,0.1183]]
       }
       CTRL_C_eye.Y {
        type=MovieSceneScriptingFloatChannel
        default=0
        thinned_from=38
        keys [[0,0],[24,0],[25,-0.001],[47,-0.187],[72,-0.187],[73,-0.0008],[93,0.1383],[103,0.1383],[104,0.1274],[114,0.1274],[115,0.2041],[120,0.2041],[121,0.1383],[127,0.1383],[138,-0.5442],[154,-0.9754],[199,-0.9751],[208,-0.9446],[217,-0.6593],[233,-0.6593],[234,-0.447],[246,-0.2236],[262,-0.0577],[263,-0.0969],[282,-0.0969],[283,0.0568],[293,0.0568],[294,0.0594],[302,0.0181],[312,-0.0211],[332,0.0004],[333,-0.0066],[350,-0.0066],[351,0],[353,0]]
       }
       CTRL_C_eyesConvergence {type=MovieSceneScriptingFloatChannel, default=0}
       CTRL_L_eye_cheekRaise {type=MovieSceneScriptingFloatChannel, default=0, keys [[120,0],[135,0.4432],[152,0.5014],[233,0.5014],[248,0]]}
       CTRL_L_eye_squintInner {type=MovieSceneScriptingFloatChannel, default=0}
       CTRL_L_nose_wrinkleUpper {type=MovieSceneScriptingFloatChannel, default=0}
       CTRL_R_nose_wrinkleUpper {type=MovieSceneScriptingFloatChannel, default=0}
       CTRL_R_eye_squintInner {type=MovieSceneScriptingFloatChannel, default=0}
       CTRL_R_eye_cheekRaise {type=MovieSceneScriptingFloatChannel, default=0, keys [[120,0],[135,0.4432],[152,0.5014],[233,0.5014],[248,0]]}
       CTRL_C_eyelidsReduce {type=MovieSceneScriptingFloatChannel, default=0}
       CTRL_R_brow_lateral {type=MovieSceneScriptingFloatChannel, default=0}
       CTRL_L_brow_lateral {type=MovieSceneScriptingFloatChannel, default=0}
       CTRL_R_brow_down {type=MovieSceneScriptingFloatChannel, default=0}
       CTRL_L_brow_down {type=MovieSceneScriptingFloatChannel, default=0}
       CTRL_L_brow_raiseOut {type=MovieSceneScriptingFloatChannel, default=0}
       CTRL_L_brow_raiseIn {type=MovieSceneScriptingFloatChannel, default=0}
       CTRL_R_brow_raiseIn {type=MovieSceneScriptingFloatChannel, default=0}
       CTRL_R_brow_raiseOut {type=MovieSceneScriptingFloatChannel, default=0}
       CTRL_lookAtSwitch {type=MovieSceneScriptingFloatChannel, default=0}
       CTRL_C_teethD.X {type=MovieSceneScriptingFloatChannel, default=0}
       CTRL_C_teethD.Y {type=MovieSceneScriptingFloatChannel, default=0}
       CTRL_C_teethU.X {type=MovieSceneScriptingFloatChannel, default=0}
       CTRL_C_teethU.Y {type=MovieSceneScriptingFloatChannel, default=0}
       CTRL_C_teeth_fwdBackU {type=MovieSceneScriptingFloatChannel, default=0}
       CTRL_L_mouth_corner.X {type=MovieSceneScriptingFloatChannel, default=0}
       CTRL_L_mouth_corner.Y {type=MovieSceneScriptingFloatChannel, default=0}
       CTRL_R_mouth_corner.X {type=MovieSceneScriptingFloatChannel, default=0}
       CTRL_R_mouth_corner.Y {type=MovieSceneScriptingFloatChannel, default=0}
       CTRL_C_teeth_fwdBackD {type=MovieSceneScriptingFloatChannel, default=0}
       CTRL_R_mouth_cornerSharpnessD {type=MovieSceneScriptingFloatChannel, default=0}
       CTRL_R_mouth_cornerSharpnessU {type=MovieSceneScriptingFloatChannel, default=0}
       CTRL_L_eye_eyelidU.X {type=MovieSceneScriptingFloatChannel, default=0}
       CTRL_L_eye_eyelidU.Y {type=MovieSceneScriptingFloatChannel, default=0}
       CTRL_L_mouth_cornerSharpnessD {type=MovieSceneScriptingFloatChannel, default=0}
       CTRL_L_eye_eyelidD.X {type=MovieSceneScriptingFloatChannel, default=0}
       CTRL_L_eye_eyelidD.Y {type=MovieSceneScriptingFloatChannel, default=0}
       CTRL_L_mouth_cornerSharpnessU {type=MovieSceneScriptingFloatChannel, default=0}
       CTRL_R_mouth_lipsTowardsTeethD {type=MovieSceneScriptingFloatChannel, default=0}
       CTRL_L_mouth_lipsTowardsTeethD {type=MovieSceneScriptingFloatChannel, default=0}
       CTRL_R_eye_eyelidU.X {type=MovieSceneScriptingFloatChannel, default=0}
       CTRL_R_eye_eyelidU.Y {type=MovieSceneScriptingFloatChannel, default=0}
       CTRL_L_mouth_lipsTowardsTeethU {type=MovieSceneScriptingFloatChannel, default=0}
       CTRL_R_eye_eyelidD.X {type=MovieSceneScriptingFloatChannel, default=0}
       CTRL_R_eye_eyelidD.Y {type=MovieSceneScriptingFloatChannel, default=0}
       CTRL_R_mouth_lipsTowardsTeethU {type=MovieSceneScriptingFloatChannel, default=0}
       CTRL_faceGUIfollowHead {type=MovieSceneScriptingFloatChannel, default=1}
       CTRL_eyesAimFollowHead {type=MovieSceneScriptingFloatChannel, default=0}
       CTRL_C_nose.X {type=MovieSceneScriptingFloatChannel, default=0}
       CTRL_C_nose.Y {type=MovieSceneScriptingFloatChannel, default=0}
       CTRL_C_teeth_rotateU {type=MovieSceneScriptingFloatChannel, default=0}
       CTRL_C_teeth_rotateD {type=MovieSceneScriptingFloatChannel, default=0}
       CTRL_L_brow_inner {type=MovieSceneScriptingFloatChannel, default=0, keys [[112,0],[125,0],[128,0.176],[131,0.2333],[231,0.2333],[236,0.0284],[240,0]]}
       CTRL_R_brow_inner {type=MovieSceneScriptingFloatChannel, default=0, keys [[112,0],[125,0],[128,0.176],[131,0.2333],[231,0.2333],[236,0.0284],[240,0]]}
       CTRL_L_brow_outer {
        type=MovieSceneScriptingFloatChannel
        default=0
        keys [[0,0],[24,0],[27,-0.4634],[29,-0.488],[32,-0.0364],[35,0],[72,0],[75,-0.464],[78,-0.4886],[81,-0.0407],[84,0],[122,0],[124,-0.4485],[125,-0.4571],[128,0.1806],[131,0.2227],[196,0.2226],[199,-0.1752],[200,-0.1817],[203,0.1781],[206,0.2222],[233,0.2224],[238,-0.4528],[244,-0.5079],[247,-0.0446],[250,0],[293,0],[296,-0.4865],[300,-0.5112],[303,-0.0852],[306,0],[352,0]]
       }
       CTRL_R_brow_outer {
        type=MovieSceneScriptingFloatChannel
        default=0
        keys [[0,0],[24,0],[27,-0.4634],[29,-0.488],[32,-0.0364],[35,0],[72,0],[75,-0.464],[78,-0.4886],[81,-0.0407],[84,0],[122,0],[124,-0.4485],[125,-0.4571],[128,0.1806],[131,0.2227],[196,0.2226],[199,-0.1752],[200,-0.1817],[203,0.1781],[206,0.2222],[233,0.2224],[238,-0.4528],[244,-0.5079],[247,-0.0446],[250,0],[293,0],[296,-0.4865],[300,-0.5112],[303,-0.0852],[306,0],[352,0]]
       }
       CTRL_C_mouth.X {type=MovieSceneScriptingFloatChannel, default=0, const [0,7]}
       CTRL_C_mouth.Y {type=MovieSceneScriptingFloatChannel, default=0, keys [[0,0],[120,0],[133,0.3397],[150,0.3723],[233,0.3723],[248,0.0016],[352,0]]}
       CTRL_L_mouth_upperLipRaise {type=MovieSceneScriptingFloatChannel, default=0, const [0,7]}
       CTRL_L_mouth_cornerPull {type=MovieSceneScriptingFloatChannel, default=0, const [0,7]}
       CTRL_L_mouth_dimple {type=MovieSceneScriptingFloatChannel, default=0, keys [[0,0],[120,0],[133,0.6054],[150,0.6636],[233,0.6636],[248,0.0029],[352,0]]}
       CTRL_L_mouth_cornerDepress {type=MovieSceneScriptingFloatChannel, default=0, keys [[0,0],[120,0],[133,0.1525],[150,0.1671],[233,0.1671],[248,0.0007],[352,0]]}
       CTRL_L_mouth_stretch {type=MovieSceneScriptingFloatChannel, default=0, const [0,7]}
       CTRL_L_mouth_lowerLipDepress {type=MovieSceneScriptingFloatChannel, default=0, const [0,7]}
       CTRL_R_mouth_lowerLipDepress {type=MovieSceneScriptingFloatChannel, default=0, const [0,7]}
       CTRL_R_mouth_stretch {type=MovieSceneScriptingFloatChannel, default=0, const [0,7]}
       CTRL_L_mouth_towardsD {type=MovieSceneScriptingFloatChannel, default=0}
       CTRL_R_mouth_funnelU {type=MovieSceneScriptingFloatChannel, default=0}
       CTRL_L_mouth_towardsU {type=MovieSceneScriptingFloatChannel, default=0}
       CTRL_R_mouth_lipBiteD {type=MovieSceneScriptingFloatChannel, default=0}
       CTRL_R_mouth_funnelD {type=MovieSceneScriptingFloatChannel, default=0}
       CTRL_L_mouth_pressU {type=MovieSceneScriptingFloatChannel, default=0}
       CTRL_L_mouth_funnelD {type=MovieSceneScriptingFloatChannel, default=0}
       CTRL_L_mouth_suckBlow {type=MovieSceneScriptingFloatChannel, default=0}
       CTRL_L_mouth_pressD {type=MovieSceneScriptingFloatChannel, default=0}
       CTRL_L_mouth_funnelU {type=MovieSceneScriptingFloatChannel, default=0}
       CTRL_R_mouth_towardsU {type=MovieSceneScriptingFloatChannel, default=0}
       CTRL_R_mouth_purseD {type=MovieSceneScriptingFloatChannel, default=0}
       CTRL_L_mouth_lipBiteU {type=MovieSceneScriptingFloatChannel, default=0}
       CTRL_R_mouth_pressD {type=MovieSceneScriptingFloatChannel, default=0}
       CTRL_R_mouth_pressU {type=MovieSceneScriptingFloatChannel, default=0}
       CTRL_L_mouth_purseD {type=MovieSceneScriptingFloatChannel, default=0}
       CTRL_R_mouth_lipBiteU {type=MovieSceneScriptingFloatChannel, default=0}
       CTRL_R_mouth_suckBlow {type=MovieSceneScriptingFloatChannel, default=0}
       CTRL_L_mouth_purseU {type=MovieSceneScriptingFloatChannel, default=0}
       CTRL_R_mouth_towardsD {type=MovieSceneScriptingFloatChannel, default=0}
       CTRL_L_mouth_lipBiteD {type=MovieSceneScriptingFloatChannel, default=0}
       CTRL_R_mouth_purseU {type=MovieSceneScriptingFloatChannel, default=0}
       CTRL_R_mouth_cornerDepress {type=MovieSceneScriptingFloatChannel, default=0, keys [[0,0],[120,0],[133,0.2107],[150,0.2309],[233,0.2309],[248,0.001],[352,0]]}
       CTRL_R_mouth_upperLipRaise {type=MovieSceneScriptingFloatChannel, default=0, const [0,7]}
       CTRL_R_mouth_dimple {type=MovieSceneScriptingFloatChannel, default=0, keys [[0,0],[120,0],[133,0.4863],[150,0.533],[233,0.533],[248,0.0023],[352,0]]}
       CTRL_R_mouth_cornerPull {type=MovieSceneScriptingFloatChannel, default=0, const [0,7]}
      ]
     }]
   }
  ]
 }
]
display_rate {denominator=1, numerator=30}
tick_resolution {denominator=1, numerator=24000}
tracks []
